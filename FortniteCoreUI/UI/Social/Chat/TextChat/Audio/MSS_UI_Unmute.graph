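# One-shot stereo MetaSound source for the text chat UI unmute sound, currently using the UI_Notification_Received wave.
# A nested MSS_PlayRandomSound_Stereo graph plays from the Sounds array on play and signals when finished, with both channels scaled by MasterVolume.
# The MinPitch, MaxPitch and MinVolume inputs are declared but not wired into the graph.
# It uses the Fort_UI sound class and sends to the Fortnite_SFX submix.
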MetaSoundSource /FortniteCoreUI/UI/Social/Chat/TextChat/Audio/MSS_UI_Unmute
inputs [
 MasterVolume {type=Float}
 MaxPitch {type=Float, access=Constructor}
 MinPitch {type=Float, access=Constructor}
 MinVolume {type=Float, access=Constructor}
 Sounds {type=WaveAsset:Array, access=Constructor}
 UE.Source.OnPlay {type=Trigger}
]
interfaces [UE.OutputFormat.Stereo:1.0, UE.Source.OneShot:1.0, UE.Source:1.1]
outputs [
 "UE.OutputFormat.Stereo.Audio:0" {type=Audio}
 "UE.OutputFormat.Stereo.Audio:1" {type=Audio}
 UE.Source.OneShot.OnFinished {type=Trigger}
]
pages [Default {
  input_defaults {
   MasterVolume=0.7943
   MaxPitch=1
   MinPitch=1
   MinVolume=0.7
   UE.Source.OnPlay=null
   Sounds [/Game/Athena/Sounds/Notifications/UI_Notification_Received]
  }
  nodes [
   MSS_PlayRandomSound_Stereo_0 = &MSS_PlayRandomSound_Stereo(Sounds=^Sounds, UE.Source.OnPlay=^UE.Source.OnPlay)
   Multiply_0 = UE.Multiply:Audio by Float(AdditionalOperands=^MasterVolume, PrimaryOperand=@MSS_PlayRandomSound_Stereo_0."UE.OutputFormat.Stereo.Audio:1")
   Multiply_1 = UE.Multiply:Audio by Float(AdditionalOperands=^MasterVolume, PrimaryOperand=@MSS_PlayRandomSound_Stereo_0."UE.OutputFormat.Stereo.Audio:0")
  ]
  links [
   {dst="@UE.Source.OneShot.OnFinished", src=MSS_PlayRandomSound_Stereo_0.UE.Source.OneShot.OnFinished}
   {dst="@UE.OutputFormat.Stereo.Audio:1", src=Multiply_0.Out}
   {dst="@UE.OutputFormat.Stereo.Audio:0", src=Multiply_1.Out}
  ]
 }]
properties {
 assetImportData=&self:AssetImportData
 numChannels=2
 outputFormat=Stereo
 sampleRate=48000
 soundClassObject=&Fort_UI
 virtualizationMode=Disabled
 volume=0.5
 concurrencyOverrides {resolutionRule=StopOldest}
 soundSubmixSends [{minSendLevel=1, soundSubmix=&Fortnite_SFX}]
}
