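# Translucent UI material for the text chat background, drawing a rounded backplate with interaction states.
# Slate post buffer color and CoreUI backplate functions provide the regular and disabled looks, routed through named reroutes.
# ModularButton state lerps blend these by interaction, with opacity set mainly for focus and press, plus selectable variants.
# A parameterized rounded mask shapes the result, and the Supports Interactions static switch toggles interaction-driven modulation.
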Material /FortniteCoreUI/UI/Social/Chat/TextChat/Materials/M_UI_TextChat_Background
graph {
 properties {
  bAutomaticallySetUsageInEditor=false
  bUsedWithStaticMesh=false
  blendMode=BLEND_Translucent
  materialDomain=MD_UI
  pixelDepthOffsetMode=PDOM_Legacy
  usedShadingModels=Default Lit
 }
 nodes [
  MaterialFunctionCall!(Post Buffer Blur=NamedRerouteUsage()) {materialFunction=/FortUILibrary/MaterialFunctions/CoreUI/MF_UI_CoreUI_BackplateA}
  MaterialFunctionCall!(Post Buffer Blur=NamedRerouteUsage()) {materialFunction=/FortUILibrary/MaterialFunctions/CoreUI/MF_UI_CoreUI_BackplateDisabled}
  n1 = NamedRerouteDeclaration(None=NamedRerouteUsage()) {name=Regular CTA, nodeColor=<rgba 0 0 0 1>}
  n2 = Constant()
  n3 = MaterialFunctionCall!(
   Idle=@n1
   Idle Opacity=@n2
   Hovered=@n1
   Hovered Opacity=@n2
   Focused=@n1
   Focused Opacity=Constant() {r=1}
   Pressed=@n1
   Pressed Opacity=Constant() {r=0.5}
   Disabled=NamedRerouteUsage()
   Disabled Opacity=@n2
  ) {materialFunction=/FortUILibrary/Systems/ModularButton/MaterialFunctions/MF_UIKit_ButtonStatesBase_Lerp}
  n4 = MaterialFunctionCall!(
   Base States=@n3.Color
   Base States Opacity=@n3.Color
   Selected_Idle
   Selected_Idle Opacity
   Selected_Hovered
   Selected_Hovered Opacity
   Selected_Focused
   Selected_Focused Opacity
   Selected_Pressed
   Selected_Pressed Opacity
   Selected_Disabled
   Selected_Disabled Opacity
  ) {materialFunction=/FortUILibrary/Systems/ModularButton/MaterialFunctions/MF_UIKit_ButtonStatesSelectable_Lerp}
  NamedRerouteDeclaration(None=@n4.Opacity) {name=Interaction Opacity, nodeColor=<rgba 0 0 0 1>}
  NamedRerouteDeclaration(None=MaterialFunctionCall!("UVs (Optional)") {materialFunction=/Engine/Functions/UserInterface/GetSlatePost0}.LinearRGB) {name=Buffer RGB, nodeColor=<rgba 0 0 0 1>}
  NamedRerouteDeclaration(None=Constant3Vector() {constant=<rgba 0.8796 0.8796 0.8796 1>}) {name=Backplate A, nodeColor=<rgba 0 0 0 1>}
  NamedRerouteDeclaration(None=@n4.Color) {name=Interaction Color, nodeColor=<rgba 0 0 0 1>}
  NamedRerouteDeclaration(None=Constant3Vector() {constant=<rgba 0.1046 0.1046 0.1046 1>}) {name=Backplate Disabled, nodeColor=<rgba 0 0 0 1>}
  StaticSwitchParameter(True=NamedRerouteUsage(), False=Constant() {r=1}) {defaultValue=true, parameterName=Supports Interactions}
  n5 = MaterialFunctionCall(
   Corner Radius=ScalarParameter() {defaultValue=12, parameterName=Corner Radius}
   Radius in Pixels=ScalarParameter() {defaultValue=1, parameterName=Radius in Pixels, sliderMax=1}
  ) {materialFunction=&MF_UI_CoreUI_RoundedMask}
  StaticSwitchParameter(True=Multiply(A=@n5.Shape Mask, B=NamedRerouteUsage()), False=@n5.Shape Mask) {defaultValue=true, parameterName=Supports Interactions}
 ]
}
