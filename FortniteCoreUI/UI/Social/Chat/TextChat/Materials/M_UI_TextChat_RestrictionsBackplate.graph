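# Translucent UI material that, going by its name, serves as the backplate for restrictions in text chat.
# The CoreUI rounded-mask function builds a rounded-corner shape with a pixel-based radius and size modulation switched off.
# That shape mask is scaled down to a faint strength, alongside a constant black color, suggesting a subtle dark rounded panel.
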Material /FortniteCoreUI/UI/Social/Chat/TextChat/Materials/M_UI_TextChat_RestrictionsBackplate
graph {
 properties {
  bAutomaticallySetUsageInEditor=false
  bUsedWithStaticMesh=false
  blendMode=BLEND_Translucent
  materialDomain=MD_UI
  usedShadingModels=Default Lit
 }
 nodes [
  Constant3Vector()
  Multiply(A=MaterialFunctionCall(Corner Radius=Constant() {r=27}, Radius in Pixels=Constant() {r=1}, Corners Modulated by Size=Constant()) {materialFunction=&MF_UI_CoreUI_RoundedMask}.Shape Mask) {constB=0.25}
 ]
}
