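# Translucent UI material for a text chat scroll gradient, going by its name.
# It rotates normalized UI UVs and masks the green channel to form a linear gradient.
# The Rotator's 2π speed lets the Rotation parameter act as a fraction of a full turn rather than animating over time.
# A separate CoreUI BackplateD function node supplies a Post Buffer Blur output.
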Material /FortniteCoreUI/UI/Social/Chat/TextChat/Materials/M_UI_TextChat_Scroll_Gradient
graph {
 properties {
  bAutomaticallySetUsageInEditor=false
  bUsedWithStaticMesh=false
  blendMode=BLEND_Translucent
  materialDomain=MD_UI
  pixelDepthOffsetMode=PDOM_Legacy
  usedShadingModels=Default Lit
 }
 nodes [
  ComponentMask(None=Rotator(
    Coordinate=MaterialFunctionCall() {materialFunction=&GetUserInterfaceUV}.Normalized UV
    Time=ScalarParameter() {parameterName=Rotation}
   ) {speed=6.2832}) {g=true}
  MaterialFunctionCall!(Post Buffer Blur) {materialFunction=/FortUILibrary/MaterialFunctions/CoreUI/MF_UI_CoreUI_BackplateD}
 ]
}
