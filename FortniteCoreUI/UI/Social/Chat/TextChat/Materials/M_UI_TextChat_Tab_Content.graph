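# Translucent UI material for text chat tab content, tinting it by button state and optionally masking it to an icon.
# UIKit state-lerp functions blend base and selected states into interaction color and opacity, which multiply vertex color behind the Supports Interactions switch.
# The Display Icon switch, off by default, multiplies alpha by a center-scaled SDF icon mask, thresholded through an anti-aliasing function.
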Material /FortniteCoreUI/UI/Social/Chat/TextChat/Materials/M_UI_TextChat_Tab_Content
graph {
 properties {
  bAutomaticallySetUsageInEditor=false
  bUsedWithStaticMesh=false
  blendMode=BLEND_Translucent
  materialDomain=MD_UI
  pixelDepthOffsetMode=PDOM_Legacy
  usedShadingModels=Default Lit
 }
 nodes [
  n1 = MaterialFunctionCall!(Idle, Idle Opacity, Hovered, Hovered Opacity, Focused, Focused Opacity, Pressed, Pressed Opacity, Disabled, Disabled Opacity) {materialFunction=/FortUILibrary/Systems/ModularButton/MaterialFunctions/MF_UIKit_ButtonStatesBase_Lerp}
  n2 = MaterialFunctionCall!(
   Base States=@n1.Color
   Base States Opacity=@n1.Color
   Selected_Idle
   Selected_Idle Opacity
   Selected_Hovered
   Selected_Hovered Opacity
   Selected_Focused
   Selected_Focused Opacity
   Selected_Pressed
   Selected_Pressed Opacity
   Selected_Disabled
   Selected_Disabled Opacity
  ) {materialFunction=/FortUILibrary/Systems/ModularButton/MaterialFunctions/MF_UIKit_ButtonStatesSelectable_Lerp}
  NamedRerouteDeclaration(None=@n2.Opacity) {name=Interaction Opacity, nodeColor=<rgba 0 0 0 1>}
  n3 = MaterialFunctionCall(
   UVs=MaterialFunctionCall() {materialFunction=&GetUserInterfaceUV}.Normalized UV
   Texture Scale=ScalarParameter() {defaultValue=1, parameterName=Texture Scale}
  ) {materialFunction=&ScaleUVsByCenter}
  n4 = TextureObjectParameter() {parameterName=Icon, samplerType=SAMPLERTYPE_LinearColor, texture=/FortUILibrary/Textures/Icons/T_UI_Icon_UIKit_CurrentPlayerCount_SDF}
  NamedRerouteDeclaration(None=MaterialFunctionCall(
    DistanceField=TextureSample(UVs=@n3.UVs, Tex=@n4).R
    Threshold=ScalarParameter() {defaultValue=0.68, parameterName=AA Threshold, sliderMax=1}
   ) {materialFunction=&MF_UI_AntiAliasing}.Outer) {name=Icon Mask, nodeColor=<rgba 0 0 0 1>}
  NamedRerouteDeclaration(None=@n2.Color) {name=Interaction Color, nodeColor=<rgba 0 0 0 1>}
  n5 = VertexColor()
  n6 = StaticSwitchParameter(True=Multiply(A=@n5.A, B=NamedRerouteUsage()), False=@n5.A) {defaultValue=true, parameterName=Supports Interactions}
  StaticSwitchParameter(True=Multiply(A=NamedRerouteUsage(), B=@n6), False=@n6) {parameterName=Display Icon}
  StaticSwitchParameter(True=Multiply(A=NamedRerouteUsage(), B=@n5) {constA=1}, False=@n5) {defaultValue=true, parameterName=Supports Interactions}
 ]
}
