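# Background block for text chat tabs, built on the UIKit modular button background base, with a Backing image and an overlaid FocusOutline image.
# It applies background and outline materials on init and at design time, caches their dynamic instances, and sets corner radius parameters, with fixed values for the outline.
# Button transition events play hover, unhover and selected animations that drive material scalar parameters and outline opacity.
# Instant selection stops transitions and sets the selected parameter directly.
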WidgetBlueprint /FortniteCoreUI/UI/Social/Chat/TextChat/WBP_TextChat_Block_BackgroundTab
animations [
 Anim_OutlineHover {
  binding_count=1
  length_seconds=1
  object=&self:Anim_OutlineHover
  sequence {
   bindings [FocusOutline {
     id=#1
     possessed_class=Image
     tracks [
      {
       type=MovieSceneFloatTrack
       display_name=Render Opacity
       sections [{type=MovieSceneFloatSection, channels [None {type=MovieSceneScriptingFloatChannel, keys [[0,0],[3,1]]}]}]
      }
      {
       type=MovieSceneWidgetMaterialTrack
       sections [{
         type=MovieSceneParameterSection
         channels [
          Anim_Hover_Fill {type=MovieSceneScriptingFloatChannel, keys [[0,0],[60,1]]}
          Anim_Hover_Energy {type=MovieSceneScriptingFloatChannel, keys [[0,0],[45,1]]}
          Anim_Hover_Outline {type=MovieSceneScriptingFloatChannel, keys [[0,0],[15,1]]}
          Anim_OutlineThickness {type=MovieSceneScriptingFloatChannel, keys [[0,0],[15,1]]}
         ]
        }]
      }
     ]
    }]
   display_rate {numerator=60}
  }
 }
 Anim_OutlineUnhover {
  binding_count=1
  length_seconds=0.5
  object=&self:Anim_OutlineUnhover
  sequence {bindings [FocusOutline {
     id=#2
     possessed_class=Image
     tracks [
      {
       type=MovieSceneFloatTrack
       display_name=Render Opacity
       sections [{type=MovieSceneFloatSection, channels [None {type=MovieSceneScriptingFloatChannel, keys [[0,1],[2,1],[10,0]]}]}]
      }
      {
       type=MovieSceneWidgetMaterialTrack
       sections [{
         type=MovieSceneParameterSection
         channels [
          Anim_Hover_Fill {type=MovieSceneScriptingFloatChannel, const [1,2]}
          Anim_Hover_Energy {type=MovieSceneScriptingFloatChannel, const [1,3]}
          Anim_Hover_Outline {type=MovieSceneScriptingFloatChannel, const [1,2]}
          Anim_OutlineThickness {type=MovieSceneScriptingFloatChannel, keys [[0,1],[4,0]]}
         ]
        }]
      }
     ]
    }]}
 }
 Anim_Selected_Transition {
  binding_count=1
  length_seconds=1.0033
  object=&self:Anim_Selected_Transition
  sequence {bindings [Backing {
     id=#3
     possessed_class=Image
     tracks [{
       type=MovieSceneWidgetMaterialTrack
       sections [{
         type=MovieSceneParameterSection
         end_frame=20
         start_frame=0
         channels [Trans_SelectedIdle {type=MovieSceneScriptingFloatChannel, keys [[0,0],[20.0657,1]]}]
        }]
      }]
    }]}
 }
]
graphs [
 EventGraph {nodes [
   Event OnInstantTransitionToSelected() {is=title K2Node_Event}
   > Stop All Transitions!(In:Exec) -> (Out:Exec) {is=title K2Node_MacroInstance}
   > SetScalarParameterValue(
    self=Get BackgroundDynamicMaterial() -> (BackgroundDynamicMaterial:Material Instance Dynamic Object Reference)
    ParameterName=Trans_SelectedIdle
    Value=1
   )
   n1 = Event OnTransitionSelectedRequested() {is=title K2Node_Event}
   PlayTransition!(
    execute=[@n1.then, @n7.then]
    self:UIKit Block Object Reference
    Animation:Widget Animation Object Reference=Get Anim_Selected_Transition() -> (Anim_Selected_Transition:Widget Animation Object Reference)
    bPlayForward:bool=true
   ) -> (then)
   n2 = Get Anim_OutlineUnhover() -> (Anim_OutlineUnhover:Widget Animation Object Reference)
   > Stop Transition!(In:Exec, Anim:Widget Animation Object Reference=@n2) -> (Out:Exec) {is=title K2Node_MacroInstance}
   n3 = Get Anim_OutlineHover() -> (Anim_OutlineHover:Widget Animation Object Reference)
   > PlayAnimationForward(InAnimation=@n3)
   n4 = Event OnInitialized()
   SetBackgroundMaterial!(
    execute=[@n4.then, @n6.then]
    self:WBP UIKit Block Background Base Object Reference
    Material Instance:"Material Instance Object Reference (by ref)"=Get BackgroundMaterial() -> (BackgroundMaterial:Material Instance Object Reference)
   ) -> (then)
   > SetOutlineMaterial!(
    execute
    self:WBP UIKit Block Background Base Object Reference
    Material Instance:"Material Instance Object Reference (by ref)"=Get OutlineMaterial() -> (OutlineMaterial:Material Instance Object Reference)
   ) -> (then)
   > SetCornerRadiuses!(
    execute
    self
    CornerRadius:float=Get CornerRadius() -> (CornerRadius:double)
    RadiusInPixels:bool=Get RadiusInPixels() -> (RadiusInPixels:bool)
   ) -> (then)
   Event Construct()
   > Set RenderOpacity!(execute, RenderOpacity:float=0, self:Widget Object Reference=Get FocusOutline() -> (FocusOutline:Image Object Reference)) -> (then, Output_Get:float)
   n5 = Event PreConstruct()
   > n6 = Branch(Condition=@n5.IsDesignTime)
   Event OnTransitionFocusedRequested() {is=title K2Node_Event}
   > n7 = Branch(Condition=GetIsSelected!() -> (IsSelected:bool))
   Stop Transition(In=@n7.else, Anim=Get Anim_OutlineUnhover())
   > PlayAnimationForward(InAnimation=Get Anim_OutlineHover())
   Event OnTransitionUnfocusedRequested() {is=title K2Node_Event}
   > Stop Transition(Anim=Get Anim_OutlineHover())
   > PlayAnimationForward(InAnimation=Get Anim_OutlineUnhover())
   Event OnTransitionDeselectedFocusedRequested() {is=title K2Node_Event}
   > PlayTransition(Animation=Get Anim_Selected_Transition(), bPlayForward=false)
   > StopAnimation(InAnimation=Get Anim_OutlineHover())
   > SetRenderOpacity(self=Get FocusOutline(), InOpacity=0)
   Event OnTransitionDeselectedIdleRequested() {is=title K2Node_Event}
   > PlayTransition(Animation=Get Anim_Selected_Transition(), bPlayForward=false)
   > Stop Transition(Anim=@n3)
   > PlayAnimationForward(InAnimation=@n2)
  ]}
 SetBackgroundMaterial {nodes [
   n1 = SetBackgroundMaterial!() -> (then, Material Instance:"Material Instance Object Reference (by ref)") {is=title K2Node_FunctionEntry}
   > n2 = Set BackgroundMaterial(BackgroundMaterial:Material Instance Object Reference=@n1) -> (Output_Get:Material Instance Object Reference)
   n3 = Get Backing() -> (Backing:Image Object Reference)
   > SetBrushFromMaterial(self=@n3, Material=@n2)
   > n4 = GetDynamicMaterial(self=@n3)
   > RefreshBackgroundMID!(execute, self:WBP UIKit Block Background Base Object Reference, Background MID:Material Instance Dynamic Object Reference=@n4) -> (then)
   > Return Node()
  ]}
 SetCornerRadiuses {nodes [
   n1 = SetCornerRadiuses!() -> (then, CornerRadius:float=~, RadiusInPixels:bool=~) {is=title K2Node_FunctionEntry}
   > Set CornerRadius(CornerRadius:double=@n1) -> (Output_Get:double)
   > Set RadiusInPixels(RadiusInPixels:bool=@n1.RadiusInPixels) -> (Output_Get:bool)
   n2 = Get OutlineDynamicMaterial() -> (OutlineDynamicMaterial:Material Instance Dynamic Object Reference)
   > SetScalarParameterValue(self=@n2, ParameterName=Radius in Pixels, Value=1)
   > SetVectorParameterValue(
    self=@n2
    ParameterName=Corner Radius
    Value=Make Linear Color!(R:float=16, G:float=16, B:float=16, A:float=16) -> (LinearColor:Linear Color Structure) {is=title K2Node_MakeStruct}
   )
   n3 = Get BackgroundDynamicMaterial()
   > SetScalarParameterValue(self=@n3, ParameterName=Radius in Pixels, Value="To Float (Boolean)"(InBool=Get RadiusInPixels()))
   > SetScalarParameterValue(self=@n3, ParameterName=Corner Radius, Value=Get CornerRadius())
  ]}
 SetOutlineMaterial {nodes [
   n1 = SetOutlineMaterial!() -> (then, Material Instance:"Material Instance Object Reference (by ref)") {is=title K2Node_FunctionEntry}
   > n2 = Set OutlineMaterial(OutlineMaterial:Material Instance Object Reference=@n1) -> (Output_Get:Material Instance Object Reference)
   n3 = Get FocusOutline()
   > SetBrushFromMaterial(self=@n3, Material=@n2)
   > n4 = GetDynamicMaterial(self=@n3)
   > RefreshOutlineMID!(execute, self:WBP UIKit Block Background Base Object Reference, Outline MID:Material Instance Dynamic Object Reference=@n4) -> (then)
   > Return Node()
  ]}
]
variables [
 CornerRadius {type=number, default=12}
 RadiusInPixels {type=boolean, default=true}
]
widgets {
 info {
  inheritedWidgetCount=0
  parentClass=/FortUILibrary/Systems/ModularButton/Blocks/WBP_UIKit_BlockBackground_Base.WBP_UIKit_BlockBackground_Base_C
  rootWidgetClass=&GridPanel
  widgetCount=3
 }
 widgets [
  Background {widgetClassPath=&GridPanel, widget=Background, properties {columnFill [1], rowFill [1]}}
  Backing {
   widgetClassPath=&Image
   bIsVariable=true
   parent=Background
   widget=Backing
   properties {
    visibility=SelfHitTestInvisible
    brush {
     bIsDynamicallyLoaded=false
     drawAs=Image
     imageType=NoImage
     mirroring=NoMirror
     resourceName=None
     resourceObject=null
     tiling=NoTile
     imageSize=<16 16>
     margin=<ltrb 0.5 0.5 0.5 0.5>
     outlineSettings {
      bUseBrushTransparency=false
      roundingType=HalfHeightRadius
      width=0
      color {colorUseRule=UseColor_Specified, specifiedColor=<rgba 0 0 0 0>}
      cornerRadii=<0 0 0 0>
     }
     tintColor {colorUseRule=UseColor_Specified, specifiedColor=<rgba 1 1 1 1>}
     uVRegion {bIsValid=false, max=<0 0>, min=<0 0>}
    }
    colorAndOpacity=<rgba 0.7913 0.7835 0.7835 1>
   }
   slot {class=GridSlot}
  }
  FocusOutline {
   widgetClassPath=&Image
   bIsVariable=true
   parent=Background
   widget=FocusOutline
   properties {
    pixelSnapping=Disabled
    visibility=SelfHitTestInvisible
    brush {
     bIsDynamicallyLoaded=false
     drawAs=Image
     imageType=NoImage
     mirroring=NoMirror
     resourceName=None
     resourceObject=null
     tiling=NoTile
     imageSize=<16 16>
     margin=<ltrb 0 0 0 0>
     outlineSettings {
      bUseBrushTransparency=false
      roundingType=HalfHeightRadius
      width=0
      color {colorUseRule=UseColor_Specified, specifiedColor=<rgba 0 0 0 0>}
      cornerRadii=<0 0 0 0>
     }
     tintColor {colorUseRule=UseColor_Specified, specifiedColor=<rgba 1 1 1 1>}
     uVRegion {bIsValid=false, max=<0 0>, min=<0 0>}
    }
   }
   slot {class=GridSlot, properties {layer=1, padding=<ltrb -8 -8 -8 -8>}}
  }
 ]
}
