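# A More Options button widget for text chat, derived from the generic UIKit button.
# On hover and unhover it transitions its outline block between the Focused and Unfocused visual states.
# On construct it sets a specific icon size when the compact UI is in use.
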WidgetBlueprint /FortniteCoreUI/UI/Social/Chat/TextChat/WBP_TextChat_Button_MoreOptions
graphs [
 EventGraph {nodes [
   n1 = Event PreConstruct()
   > Parent: PreConstruct(IsDesignTime=@n1.IsDesignTime)
   Event Construct()
   > Parent: Construct()
   > SetCompactUIStyle() {is=title K2Node_CallFunction}
   Event On Hovered()
   > Parent: On Hovered() {is=title K2Node_CallParentFunction}
   n2 = Get WBP_UIKit_Block_Outline() -> (WBP_UIKit_Block_Outline:WBP UIKit Block Outline Object Reference)
   > RequestTransition!(execute, self:UIKit Block Object Reference=@n2, State:EUIKitBlockVisualState Enum=Focused) -> (then)
   Event On Unhovered()
   > Parent: On Unhovered() {is=title K2Node_CallParentFunction}
   > RequestTransition(self=@n2, State=Unfocused)
  ]}
 SetCompactUIStyle {nodes [
   SetCompactUIStyle() {is=title K2Node_FunctionEntry}
   > Branch(Condition=IsUsingCompactUI())
   > SetIconSize!(execute, self:WBP UIKit Button Generic Base Object Reference, Icon:Vector 2D Structure=(X=48,Y=48)) -> (then)
  ]}
]
variables []
widgets {
 info {
  inheritedWidgetCount=0
  namedSlotCount=1
  parentClass=/FortUILibrary/UIKit/Buttons/WBP_UIKit_Button_Generic.WBP_UIKit_Button_Generic_C
  rootWidgetClass=null
  widgetCount=1
 }
 widgets [WBP_UIKit_Block_Outline {widgetClassPath=/FortUILibrary/UIKit/Buttons/Blocks/WBP_UIKit_Block_Outline, bIsVariable=true, widget=WBP_UIKit_Block_Outline}]
}
