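# Base modular-button widget for text chat message entries, with a named slot for message content and a rounded outline block on top.
# PreConstruct registers the outline and broadcasts the compact-UI flag; Construct binds input-method changes, making the entry selectable only for touch.
# Outside preview mode, hover drives the outline's focused state when non-selectable, while selection drives it when selectable.
# Unhover or deselection instantly switches the outline to disabled.
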WidgetBlueprint /FortniteCoreUI/UI/Social/Chat/TextChat/WBP_TextChat_Entry_MessageBase
graphs [
 EventGraph {nodes [
   Event PreConstruct()
   > Set with Broadcast IsUsingCompactUI!(execute, IsUsingCompactUI:bool=IsUsingCompactUI()) -> (then, Output_Get:bool)
   > RegisterModularBlock!(
    execute
    self:UIKit Modular Button Object Reference
    Block:UIKit Block Object Reference=Get WBP_UIKit_Block_Outline() -> (WBP_UIKit_Block_Outline:WBP UIKit Block Outline Object Reference)
   ) -> (then)
   Event Construct()
   n1 = Create Event()
   n2 = Get CommonInputSubsystem()
   > Bind Event to On Input Method Changed(self=@n2, Delegate=@n1)
   > OnHandleInputMethodChanged!(execute, self, NewInputType:ECommonInputType Enum=GetCurrentInputType(self=@n2)) -> (then)
   Event On Hovered()
   n3 = NOR Boolean!(A:bool=Get IsPreviewMode() -> (IsPreviewMode:bool), B:bool=Get bSelectable() -> (bSelectable:bool)) -> (ReturnValue:bool)
   > n4 = Branch(Condition=@n3)
   RequestTransitionOnBlocks!(execute=[@n4.then, @n7.then], self:UIKit Modular Button Object Reference, State:EUIKitBlockVisualState Enum=Focused) -> (then)
   Event Destruct()
   > Unbind Event from On Input Method Changed(self=@n2, Delegate=@n1)
   Event On Unhovered()
   > n5 = Branch(Condition=@n3)
   RequestInstantTransitionOnBlocks!(execute=[@n5.then, @n8.then], self:UIKit Modular Button Object Reference, State:EUIKitBlockInstantTransitionState Enum=InstantDisabled) -> (then)
   Event On Selected()
   n6 = AND Boolean(A=Get bSelectable(), B=NOT Boolean(A=Get IsPreviewMode()))
   > n7 = Branch(Condition=@n6)
   Event On Deselected()
   > n8 = Branch(Condition=@n6)
  ]}
 OnHandleInputMethodChanged {nodes [
   n1 = OnHandleInputMethodChanged!() -> (then, NewInputType:ECommonInputType Enum) {is=title K2Node_FunctionEntry}
   > SetIsSelectable!(
    execute
    self:Common Button Base Object Reference
    bInIsSelectable:bool=Select!(MouseAndKeyboard:bool=false, Gamepad:bool=false, Touch:bool=true, Count:bool=false, Index:ECommonInputType Enum=@n1) -> (ReturnValue:bool)
   ) -> (then)
  ]}
 OnListEntryGenerated {nodes [OnListEntryGenerated() {is=title K2Node_FunctionEntry}]}
 SetEntryIndex {nodes [
   n1 = SetEntryIndex!() -> (then, idx:int) {is=title K2Node_FunctionEntry}
   > Set IndexOfItem(IndexOfItem:int=@n1) -> (Output_Get:int=0)
  ]}
]
variables [
 IndexOfItem {type=integer}
 IsPreviewMode {type=boolean}
 IsUsingCompactUI {type=boolean}
]
widgets {
 info {inheritedWidgetCount=1, parentClass=/Script/FortniteUI.UIKitModularButton, rootWidgetClass=&GridPanel, widgetCount=4}
 widgets [
  Grid_RootMessage {widgetClassPath=&GridPanel, widget=Grid_RootMessage, properties {columnFill [1], rowFill [1]}}
  WBP_UIKit_Block_Outline {
   widgetClassPath=/FortUILibrary/UIKit/Buttons/Blocks/WBP_UIKit_Block_Outline
   bIsVariable=true
   parent=Grid_RootMessage
   widget=WBP_UIKit_Block_Outline
   properties {bottomLeftCornerRadius=20, bottomRightCornerRadius=20, radiusInPixels=true, topLeftCornerRadius=20, topRightCornerRadius=20}
   slot {class=GridSlot, properties {columnSpan=1, layer=10, rowSpan=1}}
  }
  Slot_Message {widgetClassPath=&NamedSlot, bIsVariable=true, parent=Grid_RootMessage, widget=Slot_Message, slot {class=GridSlot}}
  InputActionWidget {widgetClassPath=&CommonActionWidget, bInherited=true, widget=null}
 ]
}
