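# Text chat list entry widget that shows a date divider: centered date text over a backplate image, derived from WBP_TextChat_Entry_MessageBase.
# When the list item is set, it casts it to BP_VM_TextChat_DateMarker_Entry and stores it as the entry's view model.
# On construct and input method changes, it shows the block outline only for gamepad.
# Two helper functions select margin and vertical alignment for compact or default UI.
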WidgetBlueprint /FortniteCoreUI/UI/Social/Chat/TextChat/WBP_TextChat_Entry_DateDivider
graphs [
 CF_CompactUI_DesiredMargin {nodes [
   n1 = CF_CompactUI_DesiredMargin!() -> (then, Margin_CompactUI:Margin Structure, Margin_Default:Margin Structure, IsUsingCompactUI:bool=~) {is=title K2Node_FunctionEntry}
   > n2 = Branch(Condition=@n1.IsUsingCompactUI)
   > Return Node!(execute, Margin_Desired:Margin Structure=@n1)
   Return Node(execute=@n2.else, Margin_Desired=@n1.Margin_Default)
  ]}
 CF_CompactUI_VAlign {nodes [
   n1 = CF_CompactUI_VAlign!() -> (then, VAlign_CompactUI:EVerticalAlignment Enum=VAlign_Fill, VAlign_Default:EVerticalAlignment Enum=VAlign_Fill, IsUsingCompactUI:bool=~) {is=title K2Node_FunctionEntry}
   > n2 = Branch(Condition=@n1.IsUsingCompactUI)
   > Return Node!(execute, VAlign_Desired:EVerticalAlignment Enum=@n1)
   Return Node(execute=@n2.else, VAlign_Desired=@n1.VAlign_Default)
  ]}
 EventGraph {nodes [
   n1 = Event OnListItemObjectSet()
   > n2 = Cast To BP_VM_TextChat_DateMarker_Entry!(execute, Object=@n1.ListItemObject) -> (then, CastFailed, AsBP VM Text Chat Date Marker Entry:BP VM Text Chat Date Marker Entry Object Reference) {is=title K2Node_DynamicCast}
   > Set BP_TextChat_DateMarker_EntryVM(BP_TextChat_DateMarker_EntryVM:BP VM Text Chat Date Marker Entry Object Reference=@n2) -> (Output_Get:BP VM Text Chat Date Marker Entry Object Reference)
   n3 = Event On Input Method Changed!() -> (OutputDelegate, then, CurrentInputType:ECommonInputType Enum) {is=title K2Node_Event}
   > SetVisibility(
    self=Get WBP_UIKit_Block_Outline() -> (WBP_UIKit_Block_Outline:WBP UIKit Block Outline Object Reference)
    InVisibility=Select!(
     MouseAndKeyboard:ESlateVisibility Enum=Collapsed
     Gamepad:ESlateVisibility Enum=SelfHitTestInvisible
     Touch:ESlateVisibility Enum=Collapsed
     Count:ESlateVisibility Enum=Collapsed
     Index:ECommonInputType Enum=@n3.CurrentInputType
    ) -> (ReturnValue)
   )
   Event Construct()
   n4 = Get CommonInputSubsystem()
   > SetVisibility(
    self=Get WBP_UIKit_Block_Outline()
    InVisibility=Select(MouseAndKeyboard=Collapsed, Gamepad=SelfHitTestInvisible, Touch=Collapsed, Count=Collapsed, Index=GetCurrentInputType(self=@n4))
   )
  ]}
]
variables [
 Margin_CompactUI {type=Margin, default=<ltrb 18 2 18 2>}
 Margin_Default {type=Margin, default=<ltrb 9 1 9 1>}
]
widgets {
 info {
  inheritedWidgetCount=0
  namedSlotCount=1
  parentClass=&/WBP_TextChat_Entry_MessageBase.WBP_TextChat_Entry_MessageBase_C
  rootWidgetClass=null
  widgetCount=4
 }
 widgets [
  GridPanel_0 {widgetClassPath=&GridPanel, widget=GridPanel_0, properties {columnFill [1], rowFill [1]}}
  Overlay_0 {
   widgetClassPath=&Overlay
   parent=GridPanel_0
   widget=Overlay_0
   slot {class=GridSlot, properties {horizontalAlignment=HAlign_Center, verticalAlignment=VAlign_Center}}
  }
  DateDividerBackplate {
   widgetClassPath=&Image
   bIsVariable=true
   parent=Overlay_0
   widget=DateDividerBackplate
   properties {brush {
     bIsDynamicallyLoaded=false
     drawAs=Image
     imageType=FullColor
     mirroring=NoMirror
     resourceName=None
     resourceObject=/FortniteCoreUI/UI/Social/Chat/TextChat/Materials/MI_UI_TextChat_DateDividerBackplate
     tiling=NoTile
     imageSize=<24 24>
     margin=<ltrb 0 0 0 0>
     outlineSettings {
      bUseBrushTransparency=false
      roundingType=HalfHeightRadius
      width=0
      color {colorUseRule=UseColor_Specified, specifiedColor=<rgba 0 0 0 0>}
      cornerRadii=<0 0 0 0>
     }
     tintColor {colorUseRule=UseColor_Specified, specifiedColor=<rgba 1 1 1 1>}
     uVRegion {bIsValid=false, max=<0 0>, min=<0 0>}
    }}
   slot {class=OverlaySlot, properties {horizontalAlignment=HAlign_Fill, verticalAlignment=VAlign_Center}}
  }
  DateText {
   widgetClassPath=&CommonTextBlock
   bIsVariable=true
   parent=Overlay_0
   widget=DateText
   properties {
    mobileFontSizeMultiplier=1.41
    text=Date
    colorAndOpacity {colorUseRule=UseColor_Specified, specifiedColor=<rgba 0.5776 0.5776 0.5776 1>}
    font {
     bForceMonospaced=false
     bMaterialIsStencil=false
     fontMaterial=null
     fontObject=&HeadingNow
     letterSpacing=0
     monospacedWidth=1
     size=12.75
     skewAmount=0
     typefaceFontName=74 Regular
     outlineSettings {
      bApplyOutlineToDropShadows=false
      bMiteredCorners=false
      bSeparateFillAlpha=false
      outlineMaterial=null
      outlineSize=0
      outlineColor=<rgba 0 0 0 1>
     }
    }
    shadowOffset=<0 0>
   }
   slot {class=OverlaySlot, properties {horizontalAlignment=HAlign_Center, verticalAlignment=VAlign_Center, padding=<ltrb 9 1 9 1>}}
  }
 ]
}
