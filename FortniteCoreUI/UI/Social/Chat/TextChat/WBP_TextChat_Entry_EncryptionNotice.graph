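# Text chat entry widget, derived from WBP_TextChat_Entry_MessageBase, that shows a notice that messages are end-to-end encrypted.
# A vertical box stacks an icon, the centered notice text and a "Learn more" label, all tinted muted gray.
# The On Clicked event only prints a "Learn more about e2e URL" string; it does not open a link.
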WidgetBlueprint /FortniteCoreUI/UI/Social/Chat/TextChat/WBP_TextChat_Entry_EncryptionNotice
graphs [EventGraph {nodes [
   Event On Clicked()
   > PrintString!(execute, InString=Learn more about e2e URL, bPrintToScreen, bPrintToLog, TextColor, Duration, Key) -> (then)
  ]}]
variables []
widgets {
 info {
  inheritedWidgetCount=0
  namedSlotCount=1
  parentClass=&/WBP_TextChat_Entry_MessageBase.WBP_TextChat_Entry_MessageBase_C
  rootWidgetClass=null
  widgetCount=4
 }
 widgets [
  Grid {widgetClassPath=&VerticalBox, widget=Grid, properties {bOverride_Cursor=true}}
  Icon {
   widgetClassPath=&Image
   bIsVariable=true
   parent=Grid
   widget=Icon
   properties {brush {
     bIsDynamicallyLoaded=false
     drawAs=Image
     imageType=FullColor
     mirroring=NoMirror
     resourceName=None
     resourceObject=/FortUILibrary/Textures/MaterialInstances/MI_UIKit_Icon_Career
     tiling=NoTile
     imageSize=<64 64>
     margin=<ltrb 0 0 0 0>
     outlineSettings {
      bUseBrushTransparency=false
      roundingType=HalfHeightRadius
      width=0
      color {colorUseRule=UseColor_Specified, specifiedColor=<rgba 0 0 0 0>}
      cornerRadii=<0 0 0 0>
     }
     tintColor {colorUseRule=UseColor_Specified, specifiedColor=<rgba 0.402 0.402 0.4233 1>}
     uVRegion {bIsValid=false, max=<0 0>, min=<0 0>}
    }}
   slot {class=VerticalBoxSlot, properties {horizontalAlignment=HAlign_Center, verticalAlignment=VAlign_Center, padding=<ltrb 0 24 0 16>}}
  }
  Message {
   widgetClassPath=&CommonTextBlock
   bIsVariable=true
   parent=Grid
   widget=Message
   properties {
    bIsScrollingEnabled=false
    justification=Center
    minDesiredWidth=300
    mobileFontSizeMultiplier=1.5
    text=Messages are end-to-end encrypted.
    colorAndOpacity {colorUseRule=UseColor_Specified, specifiedColor=<rgba 0.402 0.402 0.4233 1>}
    font {
     bForceMonospaced=false
     bMaterialIsStencil=false
     fontMaterial=null
     fontObject=&HeadingNow
     letterSpacing=0
     monospacedWidth=1
     size=13.5
     skewAmount=0
     typefaceFontName=74 Regular
     outlineSettings {
      bApplyOutlineToDropShadows=false
      bMiteredCorners=false
      bSeparateFillAlpha=false
      outlineMaterial=null
      outlineSize=2
      outlineColor=<rgba 0 0 0 0.15>
     }
    }
    shadowOffset=<0 0>
   }
   slot {class=VerticalBoxSlot, properties {padding=<ltrb 8 0 8 0>}}
  }
  LearnMore {
   widgetClassPath=&CommonTextBlock
   bIsVariable=true
   parent=Grid
   widget=LearnMore
   properties {
    bIsScrollingEnabled=false
    justification=Center
    minDesiredWidth=300
    mobileFontSizeMultiplier=1.5
    text=Learn more
    colorAndOpacity {colorUseRule=UseColor_Specified, specifiedColor=<rgba 0.402 0.402 0.4233 1>}
    font {
     bForceMonospaced=false
     bMaterialIsStencil=false
     fontMaterial=null
     fontObject=&HeadingNow
     letterSpacing=0
     monospacedWidth=1
     size=13.5
     skewAmount=0
     typefaceFontName=75 Medium
     outlineSettings {
      bApplyOutlineToDropShadows=false
      bMiteredCorners=false
      bSeparateFillAlpha=false
      outlineMaterial=null
      outlineSize=2
      outlineColor=<rgba 0 0 0 0.15>
     }
    }
    shadowOffset=<0 0>
   }
   slot {class=VerticalBoxSlot, properties {padding=<ltrb 0 0 0 24>}}
  }
 ]
}
