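# Text chat entry widget for an Epic party info message, showing that someone started a party with a status and time.
# A three-column grid places named slots either side of a text column that fills the remaining width.
# The column stacks a clipped, ellipsized title above a row of green status text, a separator and time.
# The event graph is empty; the widget derives from WBP_TextChat_Entry_MessageBase and holds only layout and placeholder text.
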WidgetBlueprint /FortniteCoreUI/UI/Social/Chat/TextChat/WBP_TextChat_Entry_EpicPartyInfo
graphs [EventGraph {links [], nodes []}]
variables []
widgets {
 info {
  inheritedWidgetCount=0
  namedSlotCount=1
  parentClass=&/WBP_TextChat_Entry_MessageBase.WBP_TextChat_Entry_MessageBase_C
  rootWidgetClass=null
  widgetCount=11
 }
 widgets [
  RootGrid {widgetClassPath=&GridPanel, bIsVariable=true, widget=RootGrid, properties {columnFill [0,1,0], rowFill [1]}}
  NamedSlot {
   widgetClassPath=&NamedSlot
   bIsVariable=true
   parent=RootGrid
   widget=NamedSlot
   slot {class=GridSlot, properties {padding=<ltrb 0 12 12 12>}}
  }
  VB_TextColumn {
   widgetClassPath=&VerticalBox
   parent=RootGrid
   widget=VB_TextColumn
   slot {class=GridSlot, properties {column=1, columnSpan=1, verticalAlignment=VAlign_Center, padding=<ltrb 0 12 12 12>}}
  }
  HorizontalStackBox_Name {
   widgetClassPath=&StackBox
   parent=VB_TextColumn
   widget=HorizontalStackBox_Name
   slot {class=VerticalBoxSlot, properties {verticalAlignment=VAlign_Top, padding=<ltrb 0 0 0 3>}}
  }
  SizeBox_TextClipping {
   widgetClassPath=&SizeBox
   bIsVariable=true
   parent=HorizontalStackBox_Name
   widget=SizeBox_TextClipping
   properties {clipping=ClipToBounds}
   slot {class=StackBoxSlot}
  }
  Text_Group {
   widgetClassPath=&CommonTextBlock
   bIsVariable=true
   parent=SizeBox_TextClipping
   widget=Text_Group
   properties {
    mobileFontSizeMultiplier=1.44
    text="{Name} started a party"
    textOverflowPolicy=Ellipsis
    colorAndOpacity {colorUseRule=UseColor_Specified, specifiedColor=<rgba 0.7913 0.7913 0.8228 1>}
    font {
     bForceMonospaced=false
     bMaterialIsStencil=false
     fontMaterial=null
     fontObject=&HeadingNow
     letterSpacing=10
     monospacedWidth=1
     size=13.5
     skewAmount=0
     typefaceFontName=75 Medium
     outlineSettings {
      bApplyOutlineToDropShadows=false
      bMiteredCorners=false
      bSeparateFillAlpha=false
      outlineMaterial=/FortUILibrary/Systems/ModularButton/MaterialInstances/MI_UIKit_TextOutline
      outlineSize=0
      outlineColor=<rgba 0 0 0 1>
     }
    }
    shadowOffset=<0 0>
   }
   slot {class=SizeBoxSlot, properties {horizontalAlignment=HAlign_Left, verticalAlignment=VAlign_Center}}
  }
  HorizontalBox_0 {
   widgetClassPath=&HorizontalBox
   parent=VB_TextColumn
   widget=HorizontalBox_0
   slot {class=VerticalBoxSlot, properties {verticalAlignment=VAlign_Bottom, size {sizeRule=Fill, value=1}}}
  }
  Text_Status {
   widgetClassPath=&CommonTextBlock
   bIsVariable=true
   parent=HorizontalBox_0
   widget=Text_Status
   properties {
    clipping=ClipToBounds
    mobileFontSizeMultiplier=1.6
    text="{Status}"
    textOverflowPolicy=Ellipsis
    colorAndOpacity {colorUseRule=UseColor_Specified, specifiedColor=<rgba 0.1144 0.7682 0.1356 1>}
    font {
     bForceMonospaced=false
     bMaterialIsStencil=false
     fontMaterial=null
     fontObject=&HeadingNow
     letterSpacing=0
     monospacedWidth=1
     size=11.25
     skewAmount=0
     typefaceFontName=74 Regular
     outlineSettings {
      bApplyOutlineToDropShadows=false
      bMiteredCorners=false
      bSeparateFillAlpha=false
      outlineMaterial=&MI_UIKit_TextOutline
      outlineSize=0
      outlineColor=<rgba 0 0 0 1>
     }
    }
    shadowOffset=<0 0>
   }
   slot {class=HorizontalBoxSlot}
  }
  Text_Separator {
   widgetClassPath=&CommonTextBlock
   parent=HorizontalBox_0
   widget=Text_Separator
   properties {
    clipping=ClipToBounds
    text="•"
    textOverflowPolicy=Ellipsis
    colorAndOpacity {colorUseRule=UseColor_Specified, specifiedColor=<rgba 1 1 1 0.5>}
    font {
     bForceMonospaced=false
     bMaterialIsStencil=false
     fontMaterial=null
     fontObject=&HeadingNow
     letterSpacing=0
     monospacedWidth=1
     size=13.5
     skewAmount=0
     typefaceFontName=76 Bold
     outlineSettings {
      bApplyOutlineToDropShadows=false
      bMiteredCorners=false
      bSeparateFillAlpha=false
      outlineMaterial=&MI_UIKit_TextOutline
      outlineSize=0
      outlineColor=<rgba 0 0 0 1>
     }
    }
    shadowOffset=<0 0>
   }
   slot {class=HorizontalBoxSlot, properties {padding=<ltrb 4 0 4 0>}}
  }
  Text_Time {
   widgetClassPath=&CommonTextBlock
   bIsVariable=true
   parent=HorizontalBox_0
   widget=Text_Time
   properties {
    clipping=ClipToBounds
    mobileFontSizeMultiplier=1.6
    text="{time}"
    textOverflowPolicy=Ellipsis
    colorAndOpacity {colorUseRule=UseColor_Specified, specifiedColor=<rgba 1 1 1 0.5>}
    font {
     bForceMonospaced=false
     bMaterialIsStencil=false
     fontMaterial=null
     fontObject=&HeadingNow
     letterSpacing=0
     monospacedWidth=1
     size=11.25
     skewAmount=0
     typefaceFontName=74 Regular
     outlineSettings {
      bApplyOutlineToDropShadows=false
      bMiteredCorners=false
      bSeparateFillAlpha=false
      outlineMaterial=&MI_UIKit_TextOutline
      outlineSize=0
      outlineColor=<rgba 0 0 0 1>
     }
    }
    shadowOffset=<0 0>
   }
   slot {class=HorizontalBoxSlot}
  }
  NamedSlot_2 {
   widgetClassPath=&NamedSlot
   bIsVariable=true
   parent=RootGrid
   widget=NamedSlot_2
   slot {class=GridSlot, properties {column=2, padding=<ltrb 8 24 8 24>}}
  }
 ]
}
