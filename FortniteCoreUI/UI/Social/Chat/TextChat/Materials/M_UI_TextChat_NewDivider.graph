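# Translucent UI material for the new-message divider in text chat, going by its name and path.
# It defines a constant bright yellow color and a horizontal gradient from a linear interpolation driven by the U channel of normalized UI UVs.
# The gradient rises from zero to a faint maximum across the widget, with no textures or parameters.
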Material /FortniteCoreUI/UI/Social/Chat/TextChat/Materials/M_UI_TextChat_NewDivider
graph {
 properties {
  bAutomaticallySetUsageInEditor=false
  bUsedWithStaticMesh=false
  blendMode=BLEND_Translucent
  materialDomain=MD_UI
  usedShadingModels=Default Lit
 }
 nodes [
  Constant3Vector() {constant=<rgba 0.9301 1 0.0103 1>}
  LinearInterpolate(Alpha=ComponentMask(None=MaterialFunctionCall() {materialFunction=&GetUserInterfaceUV}.Normalized UV) {r=true}) {constB=0.25}
 ]
}
