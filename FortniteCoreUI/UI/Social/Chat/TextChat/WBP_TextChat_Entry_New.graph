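# Text chat list entry widget that marks new messages with a "New" divider, derived from the text chat message entry base widget.
# A grid places a material-driven divider line in the filling left column and a yellow "New" label on the right.
# When its list item is set, it casts the object to the new-marker entry view model and assigns it to its viewmodel property.
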WidgetBlueprint /FortniteCoreUI/UI/Social/Chat/TextChat/WBP_TextChat_Entry_New
graphs [EventGraph {nodes [
   n1 = Event OnListItemObjectSet()
   > n2 = Cast To BP_VM_TextChat_NewMarker_Entry!(execute, Object=@n1.ListItemObject) -> (then, CastFailed, AsBP VM Text Chat New Marker Entry:BP VM Text Chat New Marker Entry Object Reference) {is=title K2Node_DynamicCast}
   > Set BP_TextChat_NewMarker_EntryVM(BP_TextChat_NewMarker_EntryVM:BP VM Text Chat New Marker Entry Object Reference=@n2) -> (Output_Get:BP VM Text Chat New Marker Entry Object Reference)
  ]}]
variables []
widgets {
 info {
  inheritedWidgetCount=0
  namedSlotCount=1
  parentClass=&/WBP_TextChat_Entry_MessageBase.WBP_TextChat_Entry_MessageBase_C
  rootWidgetClass=null
  widgetCount=3
 }
 widgets [
  GridPanel_0 {widgetClassPath=&GridPanel, widget=GridPanel_0, properties {columnFill [1,0], rowFill [1]}}
  Image_LeftSide {
   widgetClassPath=&Image
   bIsVariable=true
   parent=GridPanel_0
   widget=Image_LeftSide
   properties {
    visibility=SelfHitTestInvisible
    brush {
     bIsDynamicallyLoaded=false
     drawAs=Image
     imageType=FullColor
     mirroring=NoMirror
     resourceName=None
     resourceObject=/FortniteCoreUI/UI/Social/Chat/TextChat/Materials/M_UI_TextChat_NewDivider
     tiling=NoTile
     imageSize=<32 2>
     margin=<ltrb 0 0 0 0>
     outlineSettings {
      bUseBrushTransparency=false
      roundingType=HalfHeightRadius
      width=0
      color {colorUseRule=UseColor_Specified, specifiedColor=<rgba 0 0 0 0>}
      cornerRadii=<0 0 0 0>
     }
     tintColor {colorUseRule=UseColor_Specified, specifiedColor=<rgba 1 1 1 1>}
     uVRegion {bIsValid=false, max=<0 0>, min=<0 0>}
    }
   }
   slot {class=GridSlot, properties {verticalAlignment=VAlign_Center, padding=<ltrb 0 0 12 0>}}
  }
  Text_New {
   widgetClassPath=&CommonTextBlock
   bIsVariable=true
   parent=GridPanel_0
   widget=Text_New
   properties {
    applyLineHeightToBottomLine=false
    lineHeightPercentage=1.3
    mobileFontSizeMultiplier=1.4
    text=New
    wrapTextAt=300
    colorAndOpacity {colorUseRule=UseColor_Specified, specifiedColor=<rgba 0.9301 1 0.0103 1>}
    font {
     bForceMonospaced=false
     bMaterialIsStencil=false
     fontMaterial=null
     fontObject=&HeadingNow
     letterSpacing=0
     monospacedWidth=1
     size=11.25
     skewAmount=0
     typefaceFontName=74 Regular
     outlineSettings {
      bApplyOutlineToDropShadows=false
      bMiteredCorners=false
      bSeparateFillAlpha=false
      outlineMaterial=null
      outlineSize=2
      outlineColor=<rgba 0 0 0 0.35>
     }
    }
    shadowColorAndOpacity=<rgba 0 0 0 1>
    shadowOffset=<0 0>
   }
   slot {
    class=GridSlot
    properties {column=1, columnSpan=1, horizontalAlignment=HAlign_Right, verticalAlignment=VAlign_Center, padding=<ltrb 0 0 4 0>}
   }
  }
 ]
}
