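# List entry widget for an NPC message in text chat, showing an avatar, plain "Username: Message" text and a context-menu icon.
# It stores the FortTextChatMessageVM; helper functions set avatar and presence, and restyle avatar and text for blocked or muted senders.
# Non-sent messages show the icon on hover and broadcast OnMessageClicked on click, with touch needing a first tap to reveal it.
# Compact UI adjusts sizes and padding.
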WidgetBlueprint /FortniteCoreUI/UI/Social/Chat/TextChat/WBP_TextChat_Entry_NpcMessage
graphs [
 ConvertMessageDataNotRich {nodes [
   n1 = ConvertMessageDataNotRich!() -> (then, NewNickname:String, NewText:Text) {is=title K2Node_FunctionEntry}
   > Return Node!(
    execute
    FinalText:Text=Format Text!(Format="LOCTEXT(\"{Username}: {Message}\")", Username:Text="To Text (String)"(InString=@n1), Message:Text=@n1.NewText) -> (Result)
   )
  ]}
 EventGraph {nodes [
   n1 = Event PreConstruct()
   > Parent: PreConstruct(IsDesignTime=@n1.IsDesignTime)
   > n2 = Sequence()
   > Branch(Condition=@n1.IsDesignTime)
   > SetWrapTextWidth!(
    execute
    self:Common Text Object Reference=Get TextMessage_NotRich() -> (TextMessage_NotRich:Common Text Object Reference)
    InWrapTextAt:int=600
   ) -> (then)
   Set Message Style!(execute=@n2.then_1, self, IsSubMessage:bool=Get Is Sub Message() -> (Is Sub Message:bool)) -> (then) {is=title K2Node_CallFunction}
   > SetContextMenuIconVisibility!(execute, self, IsVisible:bool=false) -> (then)
   > SetCompactUIStyle() {is=title K2Node_CallFunction}
   n3 = Event OnListItemObjectSet()
   > n4 = Cast To FortTextChatMessageVM!(execute, Object=@n3.ListItemObject) -> (then, CastFailed, AsText Chat Message VM:Text Chat Message VM Object Reference) {is=title K2Node_DynamicCast}
   > Set FortTextChatMessageVM(FortTextChatMessageVM:Text Chat Message VM Object Reference=@n4) -> (Output_Get:Text Chat Message VM Object Reference)
   n5 = Event On Item Selection Changed!() -> (OutputDelegate, then, bIsSelected:bool) {is=title K2Node_Event}
   > n6 = Branch(Condition=@n5.bIsSelected)
   Branch(execute=@n6.else, Condition=IsUsingTouch(PlayerController=GetOwningPlayer()))
   > Set bIsContextIconVisible(bIsContextIconVisible:bool=false) -> (Output_Get:bool)
   > SetContextMenuIconVisibility(IsVisible=false)
   Event On Hovered()
   > n7 = Branch(Condition=IsUsingTouch(PlayerController=GetOwningPlayer()))
   Parent: On Hovered(execute=@n7.else) {is=title K2Node_CallParentFunction}
   n8 = Get IsPreviewMode() -> (IsPreviewMode:bool)
   > n9 = Branch(Condition=@n8)
   n10 = Get MessageType!(self:Text Chat Message VM Object Reference=Get FortTextChatMessageVM() -> (FortTextChatMessageVM:Text Chat Message VM Object Reference)) -> (MessageType:EFortTextChatMessageType Enum=Sent)
   SetContextMenuIconVisibility(execute=@n9.else, IsVisible=NOT Boolean(A="Equal (Enum)"(A:EFortTextChatMessageType Enum=@n10, B:EFortTextChatMessageType Enum=Sent)))
   Event On Unhovered()
   > Parent: On Unhovered() {is=title K2Node_CallParentFunction}
   > n11 = Branch(Condition=@n8)
   n12 = Branch(execute=@n11.else, Condition=Get bIsContextIconVisible() -> (bIsContextIconVisible:bool))
   SetContextMenuIconVisibility(execute=@n12.else, IsVisible=false)
   Event On Clicked()
   n13 = Get MessageType(self=Get FortTextChatMessageVM()) -> (MessageType=Sent)
   > Branch(Condition=NOT Boolean(A="Equal (Enum)"(A=@n13, B=Sent)))
   > n14 = Branch(Condition=IsUsingTouch(PlayerController=GetOwningPlayer()))
   > n15 = Branch(Condition=Get bIsContextIconVisible())
   n16 = Get FortTextChatMessageVM()
   Call On Message Clicked!(
    execute:Exec=[@n17.then, @n15.then]
    self:WBP Text Chat Entry Npc Message Object Reference
    MessageID:Name=Get MessageId!(self:Text Chat Message VM Object Reference=@n16) -> (MessageId:Name=None)
    Username:String=Get DisplayName!(self:Text Chat Message VM Object Reference=@n16) -> (DisplayName:String)
    Button:Common Button Base Object Reference=Self-Reference()
   ) -> (then:Exec) {is=title K2Node_CallDelegate}
   Set bIsContextIconVisible(execute=@n15.else, bIsContextIconVisible=true)
   > SetContextMenuIconVisibility(IsVisible=true)
   n17 = Set bIsContextIconVisible(execute=@n14.else, bIsContextIconVisible=true)
  ]}
 GetPlayerStatus {nodes [
   n1 = GetPlayerStatus!() -> (then, IsPlayerBlocked:bool=~, IsPlayerMuted:bool=~) {is=title K2Node_FunctionEntry}
   > n2 = Branch(Condition=@n1)
   n3 = Get WBP_PlayerAvatar() -> (WBP_PlayerAvatar:WBP Player Avatar Object Reference)
   > SetEmpty!(execute, self:WBP Player Avatar Object Reference=@n3, bEmpty:bool=true) -> (then)
   > SetEmptyBackgroundColor!(
    execute
    self:WBP Player Avatar Object Reference=@n3
    EmptyBackgroundColor:Color Structure=Get Empty Background Color() -> (Empty Background Color:Color Structure)
   ) -> (then)
   > SetEmptyHighlightColor!(
    execute
    self:WBP Player Avatar Object Reference=@n3
    EmptyHighlightColor:Color Structure=Get Empty Highlight Color() -> (Empty Highlight Color:Color Structure)
   ) -> (then)
   > SetIncognitoColor!(
    execute
    self:WBP Player Avatar Object Reference=@n3
    IncognitoColor:Color Structure=Get Incognito Color() -> (Incognito Color:Color Structure)
   ) -> (then)
   > n4 = SetInactiveIncognitoColor!(execute, self:WBP Player Avatar Object Reference=@n3, InactiveIncognitoColor:Color Structure=Get Incognito Color()) -> (then)
   n5 = Get TextMessage_NotRich()
   SetColorAndOpacity(
    execute=[@n8.then, @n4.then]
    self=@n5
    InColorAndOpacity=Get FontColorMutedOrBlockedMessage() -> (FontColorMutedOrBlockedMessage:Slate Color Structure)
   )
   > SetFont(self=@n5, InFontInfo=Get FontStyleMutedOrBlockedMessage() -> (FontStyleMutedOrBlockedMessage:Slate Font Info Structure))
   > Return Node!(execute, IsDisabled:bool=true)
   n6 = Branch(execute=@n2.else, Condition=@n1.IsPlayerMuted)
   n7 = Get WBP_PlayerAvatar()
   > SetIconTexture!(
    execute
    self:WBP Player Avatar Object Reference=@n7
    IconTexture:Texture 2D Object Reference=/FortUILibrary/Textures/Icons/T_UI_Icon_UIKit_TextChatMuted_DF
    UseIcon:bool=true
   ) -> (then)
   > SetIconColor!(execute, self:WBP Player Avatar Object Reference=@n7, IconColor:Color Structure=Get Icon Color() -> (Icon Color:Color Structure)) -> (then)
   > SetBackgroundColor!(
    execute
    self:WBP Player Avatar Object Reference=@n7
    BackgroundColor:Color Structure=Get Background Color() -> (Background Color:Color Structure)
   ) -> (then)
   > n8 = SetHighlightColor!(
    execute
    self:WBP Player Avatar Object Reference=@n7
    HighlightColor:Color Structure=Get Highlight Color() -> (Highlight Color:Color Structure)
   ) -> (then)
   n9 = Get WBP_PlayerAvatar()
   SetIconTexture(execute=@n6.else, self=@n9, IconTexture=&T_UI_Icon_UIKit_TextChatMuted_DF, UseIcon=false)
   > SetEmpty(self=@n9, bEmpty=false)
   n10 = Get TextMessage_NotRich()
   > SetFont(self=@n10, InFontInfo=Get FontStyleNormalMessage() -> (FontStyleNormalMessage:Slate Font Info Structure))
   > SetColorAndOpacity(self=@n10, InColorAndOpacity=Get FontColorNormalMessage() -> (FontColorNormalMessage:Slate Color Structure))
   > Return Node(IsDisabled=false)
  ]}
 OnListEntryGenerated {nodes [
   OnListEntryGenerated() {is=title K2Node_FunctionEntry}
   n1 = Get TextMessage_NotRich()
   n2 = GetLocalSize!(Geometry:"Geometry Structure (by ref)"=GetPaintSpaceGeometry!(self:Widget Object Reference=@n1) -> (ReturnValue:"Geometry Structure (by ref)")) -> (ReturnValue:Vector 2D Structure)
   > Branch(Condition=NOT Boolean(A=IsZero2D!(A:"Vector 2D Structure (by ref)"=@n2) -> (ReturnValue:bool)))
   > SetWrapTextWidth(self=@n1, InWrapTextAt=Truncate(A=BreakVector2D(InVec=@n2)))
  ]}
 OnMessageClicked {nodes [OnMessageClicked!() -> (then, MessageID:Name, Username:String, Button:Common Button Base Object Reference) {is=title K2Node_FunctionEntry}]}
 Set Message Style {nodes [
   n1 = Set Message Style!() -> (then, IsSubMessage:bool=~) {is=title K2Node_FunctionEntry}
   > Set Is Sub Message(Is Sub Message:bool=@n1) -> (Output_Get:bool)
   > Set Visibility(Visibility=Select(Option 0=HitTestInvisible, Option 1=Hidden, Index=Get Is Sub Message()), self=Get WBP_PlayerAvatar())
  ]}
 Set Username Avatar {nodes [
   n1 = Set Username Avatar!() -> (then, FortPlayerAvatarViewModel:Fort Player Avatar Viewmodel Object Reference) {is=title K2Node_FunctionEntry}
   n2 = Get FortTextChatMessageVM()
   > n3 = Branch(Condition=GetPlayerStatus!(
     self
     IsPlayerBlocked:bool=Get bIsSenderBlocked!(self:Text Chat Message VM Object Reference=@n2) -> (bIsSenderBlocked:bool)
     IsPlayerMuted:bool=Get bIsSenderMuted!(self:Text Chat Message VM Object Reference=@n2) -> (bIsSenderMuted:bool)
    ) -> (IsDisabled:bool) {is=title K2Node_CallFunction})
   n4 = Get WBP_PlayerAvatar()
   Set FortPlayerAvatarViewModel!(execute=@n3.else, FortPlayerAvatarViewModel:Fort Player Avatar Viewmodel Object Reference=@n1, self:WBP Player Avatar Object Reference=@n4) -> (then, Output_Get:Fort Player Avatar Viewmodel Object Reference)
   > SetOnlineStatusVisibility!(execute, self:WBP Player Avatar Object Reference=@n4, bHideStatus:bool=false) -> (then)
  ]}
 SetAvatarStatus {nodes [
   n1 = SetAvatarStatus!() -> (then, FortOnlineStatusViewModel:Fort Online Status Viewmodel Object Reference) {is=title K2Node_FunctionEntry}
   > SetOnlineStatus!(
    execute
    self:WBP Player Avatar Object Reference=Get WBP_PlayerAvatar()
    OnlineStatus:EPresenceIndicatorState Enum=Get OnlineStatus!(self:Fort Online Status Viewmodel Object Reference=@n1) -> (OnlineStatus:EPresenceIndicatorState Enum=Online)
   ) -> (then)
  ]}
 SetBackgroundVisibility {nodes [
   n1 = SetBackgroundVisibility!() -> (then, ShouldBeVisible:bool=~) {is=title K2Node_FunctionEntry}
   > SetVisibility(
    self=Get Image_Background() -> (Image_Background:Image Object Reference)
    InVisibility=Select(Option 0=Collapsed, Option 1=SelfHitTestInvisible, Index=@n1)
   )
  ]}
 SetCompactUIStyle {nodes [
   SetCompactUIStyle() {is=title K2Node_FunctionEntry}
   > Branch(Condition=IsUsingCompactUI())
   n1 = Get WBP_PlayerAvatar()
   > Event Set Size!(execute, self:WBP Player Avatar Object Reference=@n1, NewSize:Vector 2D Structure=(X=40,Y=40)) -> (then)
   > SetPadding!(
    execute
    self:Grid Slot Object Reference=SlotAsGridSlot(Widget=@n1)
    InPadding_Left:float=4
    InPadding_Top:float=12
    InPadding_Right:float=12
    InPadding_Bottom:float=12
   ) -> (then)
   n2 = Get TextMessage_NotRich()
   > Set LineHeightPercentage!(execute, LineHeightPercentage:float=1, self:Text Layout Widget Object Reference=@n2) -> (then, Output_Get:float)
   > Set Grid Slot Padding!(execute:Exec, Widget:Widget Object Reference=@n2, Padding_Left:float=0, Padding_Top:float=12, Padding_Right:float=0, Padding_Bottom:float=12) -> (then:Exec) {is=title K2Node_MacroInstance}
   n3 = Get ContextMenuIcon() -> (ContextMenuIcon:Image Object Reference)
   > Set Grid Slot Padding(Widget=@n3, Padding_Left=0, Padding_Top=12, Padding_Right=4, Padding_Bottom=0)
   > SetDesiredSizeOverride!(execute, self:Image Object Reference=@n3, DesiredSize:Vector 2D Structure=(X=40,Y=40)) -> (then)
  ]}
 SetContextMenuIconVisibility {nodes [
   n1 = SetContextMenuIconVisibility!() -> (then, IsVisible:bool=~) {is=title K2Node_FunctionEntry}
   > Set RenderOpacity!(execute, RenderOpacity:float=SelectFloat(A=1, B=0, bPickA=@n1), self:Widget Object Reference=Get ContextMenuIcon()) -> (then, Output_Get:float)
  ]}
 UpdatePlayerStatus {nodes [
   n1 = UpdatePlayerStatus!() -> (then, IsDisabled:bool=~) {is=title K2Node_FunctionEntry}
   > Branch(Condition=@n1)
   n2 = Get WBP_PlayerAvatar()
   > n3 = GetDynamicMaterial(self=Get AvatarImage!(self:WBP Player Avatar Object Reference=@n2) -> (AvatarImage:Image Object Reference))
   > SetVectorParameterValue(self=@n3, ParameterName=BackgroundRGB, Value=(R=0.375,G=0.375,B=0.375,A=1))
   > SetVectorParameterValue(self=@n3, ParameterName=HighlightRGB, Value=(R=0.1354,G=0.1354,B=0.1354,A=1))
   > SetOnlineStatusVisibility(self=@n2, bHideStatus=true)
  ]}
]
variables [
 Background Color {type=Color, default=<rgba 0.4431 0.4431 0.4431 1>}
 bIsContextIconVisible {type=boolean}
 Empty Background Color {type=Color, default=<rgba 0.5725 0.5725 0.5725 1>}
 Empty Highlight Color {type=Color, default=<rgba 0.8275 0.8275 0.8275 1>}
 FontColorMutedOrBlockedMessage {type=SlateColor, default {colorUseRule=UseColor_Specified, specifiedColor=<rgba 0.4125 0.4233 0.4287 1>}}
 FontColorNormalMessage {type=SlateColor, default {colorUseRule=UseColor_Specified, specifiedColor=<rgba 1 1 1 1>}}
 FontStyleMutedOrBlockedMessage {
  type=SlateFontInfo
  default {
   bForceMonospaced=false
   bMaterialIsStencil=false
   fontMaterial=null
   fontObject=&HeadingNow
   letterSpacing=0
   monospacedWidth=1
   size=13.5
   skewAmount=0
   typefaceFontName=74 Regular
   outlineSettings {
    bApplyOutlineToDropShadows=false
    bMiteredCorners=false
    bSeparateFillAlpha=false
    outlineMaterial=null
    outlineSize=2
    outlineColor=<rgba 0 0 0 0.15>
   }
  }
 }
 FontStyleNormalMessage {
  type=SlateFontInfo
  default {
   bForceMonospaced=false
   bMaterialIsStencil=false
   fontMaterial=null
   fontObject=&HeadingNow
   letterSpacing=0
   monospacedWidth=1
   size=13.5
   skewAmount=0
   typefaceFontName=74 Regular
   outlineSettings {
    bApplyOutlineToDropShadows=false
    bMiteredCorners=false
    bSeparateFillAlpha=false
    outlineMaterial=null
    outlineSize=2
    outlineColor=<rgba 0 0 0 0.35>
   }
  }
 }
 Highlight Color {type=Color, default=<rgba 0.8667 0.8667 0.8667 1>}
 Icon Color {type=Color, default=<rgba 0.9059 0.9059 0.9059 1>}
 Incognito Color {type=Color, default=<rgba 0.3608 0.3608 0.3608 1>}
 Is Sub Message {type=boolean}
 OnMessageClicked {}
 Received From PlayerColor {type=LinearColor, default=<rgba 0.5776 0.5776 0.5776 1>}
 SentColor {type=LinearColor, default=<rgba 0.2705 0.491 0.4125 1>}
]
widgets {
 info {
  inheritedWidgetCount=0
  namedSlotCount=1
  parentClass=&/WBP_TextChat_Entry_MessageBase.WBP_TextChat_Entry_MessageBase_C
  rootWidgetClass=null
  widgetCount=6
 }
 widgets [
  GP_UserMessage_Container {widgetClassPath=&GridPanel, widget=GP_UserMessage_Container, properties {columnFill [0,1,0,0], rowFill [1]}}
  WBP_PlayerAvatar {
   widgetClassPath=/Game/UI/Foundation/Profile/WBP_PlayerAvatar
   bIsVariable=true
   parent=GP_UserMessage_Container
   widget=WBP_PlayerAvatar
   properties {avatarSize=<32 32>}
   slot {class=GridSlot, properties {columnSpan=1, horizontalAlignment=HAlign_Left, layer=1, verticalAlignment=VAlign_Top, padding=<ltrb 0 0 8 0>}}
  }
  TextMessage_NotRich {
   widgetClassPath=&CommonTextBlock
   bIsVariable=true
   parent=GP_UserMessage_Container
   widget=TextMessage_NotRich
   properties {
    applyLineHeightToBottomLine=false
    lineHeightPercentage=1.3
    mobileFontSizeMultiplier=1.5
    text=UserName : Message to Secondary UserName
    wrappingPolicy=AllowPerCharacterWrapping
    font {
     bForceMonospaced=false
     bMaterialIsStencil=false
     fontMaterial=null
     fontObject=&HeadingNow
     letterSpacing=0
     monospacedWidth=1
     size=13.5
     skewAmount=0
     typefaceFontName=74 Regular
     outlineSettings {
      bApplyOutlineToDropShadows=false
      bMiteredCorners=false
      bSeparateFillAlpha=false
      outlineMaterial=null
      outlineSize=2
      outlineColor=<rgba 0 0 0 0.35>
     }
    }
    shadowColorAndOpacity=<rgba 0 0 0 1>
    shadowOffset=<0 0>
   }
   slot {class=GridSlot, properties {column=1, columnSpan=1, padding=<ltrb 0 4 0 4>}}
  }
  ContextMenuIcon {
   widgetClassPath=&Image
   bIsVariable=true
   parent=GP_UserMessage_Container
   widget=ContextMenuIcon
   properties {
    visibility=HitTestInvisible
    brush {
     bIsDynamicallyLoaded=false
     drawAs=Image
     imageType=FullColor
     mirroring=NoMirror
     resourceName=None
     resourceObject=/FortniteCoreUI/UI/Social/Chat/TextChat/MaterialInstances/MI_UI_TextChat_MessageContextMenu
     tiling=NoTile
     imageSize=<24 24>
     margin=<ltrb 0 0 0 0>
     outlineSettings {
      bUseBrushTransparency=false
      roundingType=HalfHeightRadius
      width=0
      color {colorUseRule=UseColor_Specified, specifiedColor=<rgba 0 0 0 0>}
      cornerRadii=<0 0 0 0>
     }
     tintColor {colorUseRule=UseColor_Specified, specifiedColor=<rgba 1 1 1 1>}
     uVRegion {bIsValid=false, max=<0 0>, min=<0 0>}
    }
   }
   slot {class=GridSlot, properties {column=2, columnSpan=1, horizontalAlignment=HAlign_Right, verticalAlignment=VAlign_Top, padding=<ltrb 0 4 0 0>}}
  }
  Image_Background {
   widgetClassPath=&Image
   bIsVariable=true
   parent=GP_UserMessage_Container
   widget=Image_Background
   properties {
    visibility=Collapsed
    brush {
     bIsDynamicallyLoaded=false
     drawAs=RoundedBox
     imageType=NoImage
     mirroring=NoMirror
     resourceName=None
     resourceObject=null
     tiling=NoTile
     imageSize=<1 1>
     margin=<ltrb 0 0 0 0>
     outlineSettings {
      bUseBrushTransparency=false
      roundingType=FixedRadius
      width=0
      color {colorUseRule=UseColor_Specified, specifiedColor=<rgba 0 0 0 0>}
      cornerRadii=<18 18 18 18>
     }
     tintColor {colorUseRule=UseColor_Specified, specifiedColor=<rgba 0.0513 0.0513 0.0513 0.5>}
     uVRegion {bIsValid=false, max=<0 0>, min=<0 0>}
    }
   }
   slot {class=GridSlot, properties {columnSpan=3, layer=1}}
  }
  Spacer_RightPadding {
   widgetClassPath=&Spacer
   parent=GP_UserMessage_Container
   widget=Spacer_RightPadding
   properties {size=<4 1>}
   slot {class=GridSlot, properties {column=3, columnSpan=1}}
  }
 ]
}
