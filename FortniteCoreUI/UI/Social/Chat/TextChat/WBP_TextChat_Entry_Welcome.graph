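# Text chat list entry widget that shows a centered welcome message for party, group or one-to-one conversations.
# CF_MessageText picks text by conversation type and compact mode, building group creator and date lines; its reporting-status line is disconnected from the output.
# OnListItemObjectSet caches the welcome view model, while PreConstruct fills a sample group message at design time.
# OnListEntryGenerated sets text wrap width from painted geometry when nonzero.
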WidgetBlueprint /FortniteCoreUI/UI/Social/Chat/TextChat/WBP_TextChat_Entry_Welcome
graphs [
 CF_MessageText {nodes [
   CF_MessageText!() -> (
    then
    ConversationType:EFortTextChatPersistentConversationType Enum=EpicParty
    CreationDate:Date Time Structure
    InviterUser:Fort Social User Object Reference
    IsRecording:bool=~
    IsCompactChat:bool=~
   ) {is=title K2Node_FunctionEntry}
   > n1 = Sequence!(execute) -> (then_0, then_1, then_2:Exec, then_3:Exec)
   > Set dateAsText(dateAsText:Text="As Date (from UTC)"!(
     InDateTime:"Date Time Structure (by ref)"=Get CreationDate() -> (CreationDate:Date Time Structure)
     InTimeZone:String
     InDateStyle:EDateTimeStyle Enum=Medium
    ) -> (ReturnValue:Text)) -> (Output_Get:Text)
   Branch(
    execute=@n1.then_1
    Condition="Equal (Enum)"(
     A:EFortTextChatPersistentConversationType Enum=Get ConversationType() -> (ConversationType:EFortTextChatPersistentConversationType Enum=EpicParty)
     B:EFortTextChatPersistentConversationType Enum=EpicGroup
    )
   )
   > n2 = Get!(execute:Exec) -> (then:Exec, else:Exec, InviterUser:Fort Social User Object Reference)
   n3 = GetNicknameOrDisplayName!(self:Fort Social User Object Reference=Get InviterUser() -> (InviterUser:Fort Social User Object Reference)) -> (ReturnValue:String)
   > Set firstLineOfTextAboutGroupCreatorAndDate(firstLineOfTextAboutGroupCreatorAndDate:Text=Format Text!(
     Format="LOCTEXT(\"{player} created the group on {date}\")"
     player:Text="To Text (String)"(InString=@n3)
     date:Text=Get dateAsText() -> (dateAsText:Text)
    ) -> (Result)) -> (Output_Get:Text)
   Set firstLineOfTextAboutGroupCreatorAndDate(
    execute=@n2.else
    firstLineOfTextAboutGroupCreatorAndDate=Format Text!(Format="LOCTEXT(\"Group was created on {date}\")", date=Get dateAsText()) -> (Result)
   )
   n4 = "Equal (Enum)"(A=Get ConversationType() -> (ConversationType=EpicParty), B=EpicParty)
   n5 = Format Text!(Format="LOCTEXT(\"Open the sidebar to view older messages\")") -> (Result)
   n6 = Format Text(Format="LOCTEXT(\"Send a message to start chatting.\")")
   n7 = Format Text!(
    Format="LOCTEXT(\"{firstLineGroup}\\r\\n\\r\\nMessages may be hidden due to chat permissions\")"
    firstLineGroup:Text=Get firstLineOfTextAboutGroupCreatorAndDate() -> (firstLineOfTextAboutGroupCreatorAndDate:Text)
   ) -> (Result)
   n8 = Get ConversationType() -> (ConversationType=EpicParty)
   n9 = Format Text!(Format="LOCTEXT(\"Send a message to start chatting\")") -> (Result)
   Return Node!(
    execute=@n1.then_3
    MessageText:Text=SelectText(
     A=SelectText(A=@n6, B=@n5, bPickA=@n4)
     B=Select!(EpicParty:Text=@n6, EpicGroup:Text=@n7, Epic121:Text=@n9, Index:EFortTextChatPersistentConversationType Enum=@n8) -> (ReturnValue:Text)
     bPickA=Get IsCompactChat() -> (IsCompactChat:bool)
    )
   )
   n10 = Get IsRecording() -> (IsRecording:bool)
   Set secondLineOfTextAboutReportingStatus(secondLineOfTextAboutReportingStatus:Text=Format Text!(
     Format="LOCTEXT(\"Text reporting is {onoff}. {reportHint}\")"
     onoff:Text=SelectText(A="LOCTEXT(\"on\")", B="LOCTEXT(\"off\")", bPickA=@n10)
     reportHint:Text=SelectText(A="LOCTEXT(\" Report from the ••• menu.\")", bPickA=@n10)
    ) -> (Result)) -> (Output_Get:Text)
  ]}
 EventGraph {nodes [
   n1 = Event OnListItemObjectSet()
   > n2 = Cast To BP_VM_TextChat_Welcome_Entry!(execute, Object=@n1.ListItemObject) -> (then, CastFailed, AsBP VM Text Chat Welcome Entry:BP VM Text Chat Welcome Entry Object Reference) {is=title K2Node_DynamicCast}
   > Set BP_VM_TextChat_Welcome_Entry(BP_VM_TextChat_Welcome_Entry:BP VM Text Chat Welcome Entry Object Reference=@n2) -> (Output_Get:BP VM Text Chat Welcome Entry Object Reference)
   n3 = Event PreConstruct()
   > Parent: PreConstruct(IsDesignTime=@n3.IsDesignTime)
   > Branch(Condition=@n3.IsDesignTime)
   > SetWrapTextWidth!(
    execute
    self:Common Text Object Reference=Get TextMessage_NotRich_WrapAtSetRuntime() -> (TextMessage_NotRich_WrapAtSetRuntime:Common Text Object Reference)
    InWrapTextAt:int=600
   ) -> (then)
   > Set Text(
    Text=CF_MessageText!(
     self
     ConversationType:EFortTextChatPersistentConversationType Enum=EpicGroup
     CreationDate:Date Time Structure=MakeDateTime!(Year:int=2025, Month:int=2, Day:int=24, Hour:int=8, Minute:int=36, Second:int=45, Millisecond:int=0) -> (ReturnValue:Date Time Structure)
     InviterUser:Fort Social User Object Reference
     IsRecording:bool=true
     IsCompactChat:bool=false
    ) -> (MessageText:Text) {is=title K2Node_CallFunction}
    self=Get TextMessage_NotRich_WrapAtSetRuntime()
   )
  ]}
 OnListEntryGenerated {nodes [
   OnListEntryGenerated() {is=title K2Node_FunctionEntry}
   > Parent: OnListEntryGenerated() {is=title K2Node_CallParentFunction}
   n1 = Get TextMessage_NotRich_WrapAtSetRuntime()
   n2 = GetLocalSize!(Geometry:"Geometry Structure (by ref)"=GetPaintSpaceGeometry!(self:Widget Object Reference=@n1) -> (ReturnValue:"Geometry Structure (by ref)")) -> (ReturnValue:Vector 2D Structure)
   > Branch(Condition=NOT Boolean(A=IsZero2D!(A:"Vector 2D Structure (by ref)"=@n2) -> (ReturnValue:bool)))
   > SetWrapTextWidth(self=@n1, InWrapTextAt=Truncate(A=BreakVector2D(InVec=@n2)))
  ]}
]
variables []
widgets {
 info {
  inheritedWidgetCount=0
  namedSlotCount=1
  parentClass=&/WBP_TextChat_Entry_MessageBase.WBP_TextChat_Entry_MessageBase_C
  rootWidgetClass=null
  widgetCount=2
 }
 widgets [
  VerticalBox_28 {widgetClassPath=&VerticalBox, widget=VerticalBox_28}
  TextMessage_NotRich_WrapAtSetRuntime {
   widgetClassPath=&CommonTextBlock
   bIsVariable=true
   parent=VerticalBox_28
   widget=TextMessage_NotRich_WrapAtSetRuntime
   properties {
    applyLineHeightToBottomLine=false
    justification=Center
    lineHeightPercentage=0.9
    mobileFontSizeMultiplier=1.5
    text="{player} created the group on {date}\r\n\r\nText reporting is {on/off}.  Report from the ... menu.\r\n\r\nMessages may be hidden due to chat permissions."
    wrappingPolicy=AllowPerCharacterWrapping
    colorAndOpacity {colorUseRule=UseColor_Specified, specifiedColor=<rgba 0.5776 0.5776 0.5776 1>}
    font {
     bForceMonospaced=false
     bMaterialIsStencil=false
     fontMaterial=null
     fontObject=&HeadingNow
     letterSpacing=0
     monospacedWidth=1
     size=13.5
     skewAmount=0
     typefaceFontName=74 Regular
     outlineSettings {
      bApplyOutlineToDropShadows=false
      bMiteredCorners=false
      bSeparateFillAlpha=false
      outlineMaterial=null
      outlineSize=2
      outlineColor=<rgba 0 0 0 0.35>
     }
    }
    shadowColorAndOpacity=<rgba 0 0 0 1>
    shadowOffset=<0 0>
   }
   slot {class=VerticalBoxSlot, properties {padding=<ltrb 0 4 0 4>}}
  }
 ]
}
