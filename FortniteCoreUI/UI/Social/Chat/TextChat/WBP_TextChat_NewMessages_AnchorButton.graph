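# WBP_TextChat_NewMessages_AnchorButton is a UIKit generic button subclass, named as the new-messages anchor button for text chat.
# It adds no widgets or variables of its own, and its PreConstruct and Construct events only call the parent.
# Hover and unhover events check the CommonInput subsystem and forward to the parent only when the current input type is not gamepad, suppressing that hover handling on gamepad.
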WidgetBlueprint /FortniteCoreUI/UI/Social/Chat/TextChat/WBP_TextChat_NewMessages_AnchorButton
graphs [EventGraph {nodes [
   n1 = Event PreConstruct()
   > Parent: PreConstruct(IsDesignTime=@n1.IsDesignTime)
   Event Construct()
   > Parent: Construct()
   Event On Hovered()
   n2 = Get CommonInputSubsystem()
   n3 = NOT Boolean(A="Equal (Enum)"(A=GetCurrentInputType(self=@n2), B=Gamepad))
   > Branch(Condition=@n3)
   > Parent: On Hovered() {is=title K2Node_CallParentFunction}
   Event On Unhovered()
   > Branch(Condition=@n3)
   > Parent: On Unhovered() {is=title K2Node_CallParentFunction}
  ]}]
variables []
widgets {
 info {
  inheritedWidgetCount=0
  parentClass=/FortUILibrary/UIKit/Buttons/WBP_UIKit_Button_Generic.WBP_UIKit_Button_Generic_C
  rootWidgetClass=null
  widgetCount=0
 }
 widgets []
}
