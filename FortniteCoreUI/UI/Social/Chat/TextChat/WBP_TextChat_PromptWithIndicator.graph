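# HUD widget that shows the text chat prompt: input action glyph, label, touch-only icon and unread-channel count badge.
# On construct, if the player controller uses MessageWidgetV3, it follows input method changes and listens to event-router events.
# Those events collapse it while chat is active and toggle the prompt, whose root starts hidden when auto-hide is set.
# SetUnreadChannelsCount drives the badge, collapsing it at zero.
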WidgetBlueprint /FortniteCoreUI/UI/Social/Chat/TextChat/WBP_TextChat_PromptWithIndicator
animations [Anim_Notification {
  binding_count=1
  length_seconds=0.35
  object=&self:Anim_Notification
  sequence {bindings [UIKit_StatusIndicator {
     id=#1
     possessed_class=WBP_UIKit_StatusIndicator_C
     tracks [{
       type=MovieSceneFloatTrack
       display_name=Render Opacity
       sections [{type=MovieSceneFloatSection, end_frame=7, start_frame=0, channels [None {type=MovieSceneScriptingFloatChannel, keys [[0,0],[7,1]]}]}]
      }]
    }]}
 }]
graphs [
 EventGraph {nodes [
   n1 = OnInputMethodChanged!() -> (OutputDelegate, then, InputType:ECommonInputType Enum) {is=title K2Node_CustomEvent}
   > "SetText (Text)"(
    self=Get ButtonText() -> (ButtonText:Common Text Object Reference)
    InText=Select!(
     MouseAndKeyboard:Text=Get TextForKeyboard() -> (TextForKeyboard:Text)
     Gamepad:Text=Get TextForGamepad() -> (TextForGamepad:Text)
     Touch:Text=Get TextForTouch() -> (TextForTouch:Text)
     Count:Text
     Index:ECommonInputType Enum=@n1.InputType
    ) -> (ReturnValue:Text)
   )
   > SetVisibility(
    self=Get IMG_TextChat() -> (IMG_TextChat:Image Object Reference)
    InVisibility=Select!(
     Option 0:ESlateVisibility Enum=Collapsed
     Option 1:ESlateVisibility Enum=SelfHitTestInvisible
     Index:bool="Equal (Enum)"(A=@n1.InputType, B=Touch)
    ) -> (ReturnValue:ESlateVisibility Enum)
   )
   n2 = Event PreConstruct()
   > n3 = Branch(Condition=@n2.IsDesignTime)
   SetVisibility(execute=@n3.else, self=Self-Reference(), InVisibility=HitTestInvisible)
   > Branch(Condition=Get bAutomaticallyHide() -> (bAutomaticallyHide:bool))
   > SetVisibility(self=Get Root() -> (Root:Grid Panel Object Reference), InVisibility=Collapsed)
   Event Construct()
   > n4 = Cast To FortPlayerController(Object=GetOwningPlayer())
   > Branch(Condition=UseMessageWidgetV3!(self:Fort Player Controller Object Reference=@n4) -> (ReturnValue:bool))
   n5 = Create Event()
   n6 = Get CommonInputSubsystem()
   > Bind Event to On Input Method Changed(self=@n6, Delegate=@n5)
   > OnInputMethodChanged!(execute, self, InputType:ECommonInputType Enum=GetCurrentInputType(self=@n6)) -> (then) {is=title K2Node_CallFunction}
   > n7 = Sequence()
   n8 = StartListeningToEvent!(
    execute:Exec=@n7.then_1
    Target:Gameplay Event Router Component Object Reference=GetGlobalEventRouter()
    EventType:Script Struct Object Reference=&/BPS_SetTextChatPromptButtonVisibility
    Context:Object Reference=GetOwningPlayer()
    EventBubblingRule:EEventBubblingRule Enum
   ) -> (
    then:Exec
    AsyncTaskProxy:Async Action Start Listening To Event Object Reference
    OnEventReceived:Exec
    Payload:BPS Set Text Chat Prompt Button Visibility Structure
   )
   > Set ChatButtonPromptEventReference(ChatButtonPromptEventReference:Async Action Start Listening To Event Object Reference=@n8) -> (Output_Get:Async Action Start Listening To Event Object Reference)
   n9 = Branch(
    execute=@n8.OnEventReceived
    Condition=Break BPS Set Text Chat Prompt Button Visibility!(BPS_SetTextChatPromptButtonVisibility:"BPS Set Text Chat Prompt Button Visibility Structure (by ref)"=@n8.Payload) -> (IsShouldShowChatPromptButton_2_#2:bool=~) {is=title K2Node_BreakStruct}
   )
   > n10 = Branch(Condition=IsVisible(self=Get Root()))
   SetVisibility(execute=@n10.else, self=Get Root(), InVisibility=HitTestInvisible)
   Branch(execute=@n9.else, Condition=Get bAutomaticallyHide())
   > SetVisibility(self=Get Root(), InVisibility=Collapsed)
   n11 = StartListeningToEvent(
    execute=@n7.then_0
    Target=GetGlobalEventRouter()
    EventType=&/BPS_TextChatActiveStateChange
    Context=GetOwningPlayer()
   ) -> (Payload:BPS Text Chat Active State Change Structure)
   > Set ChatOpenCloseEventReference(ChatOpenCloseEventReference:Async Action Start Listening To Event Object Reference=@n11) -> (Output_Get:Async Action Start Listening To Event Object Reference)
   SetVisibility(
    execute=@n11.OnEventReceived
    self=Self-Reference()
    InVisibility=Select(
     Option 0=HitTestInvisible
     Option 1=Collapsed
     Index=Break BPS Text Chat Active State Change!(BPS_TextChatActiveStateChange:"BPS Text Chat Active State Change Structure (by ref)"=@n11.Payload) -> (bActive_4_#3:bool=~) {is=title K2Node_BreakStruct}
    )
   )
   Event Destruct()
   > StopListeningToEvent(self:Async Action Start Listening To Event Object Reference=Get ChatOpenCloseEventReference() -> (ChatOpenCloseEventReference:Async Action Start Listening To Event Object Reference))
   > StopListeningToEvent(self=Get ChatButtonPromptEventReference() -> (ChatButtonPromptEventReference:Async Action Start Listening To Event Object Reference))
   > Unbind Event from On Input Method Changed(self=@n6, Delegate=@n5)
  ]}
 SetUnreadChannelsCount {nodes [
   n1 = SetUnreadChannelsCount!() -> (then, UnreadChannels:int) {is=title K2Node_FunctionEntry}
   > n2 = Branch(Condition="Equal (Integer)"(A=@n1))
   n3 = Get UIKit_StatusIndicator() -> (UIKit_StatusIndicator:WBP UIKit Status Indicator Object Reference)
   > SetVisibility(self=@n3, InVisibility=Collapsed)
   OverrideCount!(execute=@n2.else, self:WBP UIKit Status Indicator Base Object Reference=@n3, Count:int=@n1) -> (then)
   > SetVisibility(self=@n3, InVisibility=HitTestInvisible)
  ]}
]
variables [
 bAutomaticallyHide {type=boolean, default=true}
 ChatButtonPromptEventReference {type=AsyncAction_StartListeningToEvent, default=null}
 ChatOpenCloseEventReference {type=AsyncAction_StartListeningToEvent, default=null}
 KeyForGamepad {type=MaterialInstance, default=null}
 KeyForKeyboard {type=MaterialInstance, default=/FortUILibrary/Input/Keyboard/MaterialInstances/MI_UI_Input_Keyboard_EnterSymbol}
 TextForGamepad {type=string, default="Chat (Double-tap)"}
 TextForKeyboard {type=string, default=Chat}
 TextForTouch {type=string, default=Open Text Chat}
]
widgets {
 info {
  inheritedWidgetCount=0
  namedSlotCount=1
  parentClass=&CommonUserWidget
  rootWidgetClass=/Game/Athena/HUD/FortHUDElementWrapper.FortHUDElementWrapper_C
  widgetCount=9
 }
 widgets [
  FortHUDElementWrapper {
   widgetClassPath=/Game/Athena/HUD/FortHUDElementWrapper
   bIsVariable=true
   widget=FortHUDElementWrapper
   properties {
    elementTag {tagName=Athena.HUD.TextChatPrompt}
    hUDElementTag {gameplayTags [{tagName=Athena.HideForCapture}, {tagName=Athena.Creative.HUDController.HideTextChatPrompt}]}
   }
  }
  Root {
   widgetClassPath=&GridPanel
   bIsVariable=true
   namedSlotHost=FortHUDElementWrapper
   widget=Root
   properties {columnFill [1], rowFill [1]}
  }
  Backplate {
   widgetClassPath=&WBP_UIKit_Backplate
   bIsVariable=true
   parent=Root
   widget=Backplate
   properties {brightness=Obscure, cornerRadius=Circle, renderOpacity=0.7}
   slot {class=GridSlot, properties {layer=1}}
  }
  HorizontalBox_0 {
   widgetClassPath=&HorizontalBox
   parent=Root
   widget=HorizontalBox_0
   slot {class=GridSlot, properties {horizontalAlignment=HAlign_Left, layer=10, verticalAlignment=VAlign_Center}}
  }
  InputAction {
   widgetClassPath=&CommonActionWidget
   bIsVariable=true
   parent=HorizontalBox_0
   widget=InputAction
   properties {
    designTimeKey=Gamepad_Special_Right
    enhancedInputAction=/Game/Input/EnhancedInput/FortPlayerController/IA_OpenTextChat
    visibility=SelfHitTestInvisible
   }
   slot {class=HorizontalBoxSlot, properties {horizontalAlignment=HAlign_Center, verticalAlignment=VAlign_Center, padding=<ltrb 12 0 0 0>}}
  }
  IMG_TextChat {
   widgetClassPath=&Image
   bIsVariable=true
   parent=HorizontalBox_0
   widget=IMG_TextChat
   properties {brush {
     bIsDynamicallyLoaded=false
     drawAs=Image
     imageType=FullColor
     mirroring=NoMirror
     resourceName=None
     resourceObject=/FortniteCoreUI/UI/Social/Chat/TextChat/Textures/T_UI_Mobile_OpenChat
     tiling=NoTile
     imageSize=<32 32>
     margin=<ltrb 0 0 0 0>
     outlineSettings {
      bUseBrushTransparency=false
      roundingType=HalfHeightRadius
      width=0
      color {colorUseRule=UseColor_Specified, specifiedColor=<rgba 0 0 0 0>}
      cornerRadii=<0 0 0 0>
     }
     tintColor {colorUseRule=UseColor_Specified, specifiedColor=<rgba 1 1 1 1>}
     uVRegion {bIsValid=false, max=<0 0>, min=<0 0>}
    }}
   slot {class=HorizontalBoxSlot, properties {horizontalAlignment=HAlign_Left, verticalAlignment=VAlign_Center, padding=<ltrb 12 0 0 0>}}
  }
  ButtonText {
   widgetClassPath=&CommonTextBlock
   bIsVariable=true
   parent=HorizontalBox_0
   widget=ButtonText
   properties {
    bIsScrollingEnabled=false
    text=Chat
    font {
     bForceMonospaced=false
     bMaterialIsStencil=false
     fontMaterial=/FortUILibrary/UIKit/Buttons/MaterialInstances/CTAButtons/MI_UIKit_TextQuiet_Prompt
     fontObject=&HeadingNow
     letterSpacing=10
     monospacedWidth=1
     size=15.75
     skewAmount=0
     typefaceFontName=75 Medium
     outlineSettings {
      bApplyOutlineToDropShadows=false
      bMiteredCorners=false
      bSeparateFillAlpha=false
      outlineMaterial=null
      outlineSize=0
      outlineColor=<rgba 0 0 0 1>
     }
    }
    shadowOffset=<0 0>
   }
   slot {class=HorizontalBoxSlot, properties {horizontalAlignment=HAlign_Left, verticalAlignment=VAlign_Center, padding=<ltrb 16 0 16 0>}}
  }
  UIKit_StatusIndicator {
   widgetClassPath=/FortUILibrary/UIKit/StatusIndicator/WBP_UIKit_StatusIndicator
   bIsVariable=true
   parent=HorizontalBox_0
   widget=UIKit_StatusIndicator
   properties {type=Attention, configuration=Number}
   slot {class=HorizontalBoxSlot, properties {padding=<ltrb 0 0 16 0>}}
  }
  Spacer_96 {widgetClassPath=&Spacer, parent=HorizontalBox_0, widget=Spacer_96, properties {size=<1 48>}, slot {class=HorizontalBoxSlot}}
 ]
}
