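# Fortnite's Settings menu widget, derived from FortOptionsMenu, with a category tab bar, page switcher and keep-video-settings confirmation overlay.
# Once tabs populate, it binds gamepad config tab delegates for apply, reset, back and binding changes, and hands itself to the keyboard bindings tab.
# Unbound gamepad actions trigger warning dialogs.
# Activation handles mobile HUD hiding, title bar and, outside Athena, front-end camera; splitscreen adjusts sizing.
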WidgetBlueprint /FortniteCoreUI/Settings/OptionsMenu
graphs [
 EventGraph {nodes [
   n1 = "On Tab Button Creation (TabList_Categories)"!() -> (OutputDelegate, then, TabId:Name=None, TabButton:Common Button Base Object Reference) {is=title K2Node_ComponentBoundEvent}
   > n2 = Cast To IconTabButton_Legacy!(execute, Object=@n1.TabButton) -> (then, CastFailed, AsIcon Tab Button Legacy:Icon Tab Button Legacy Object Reference) {is=title K2Node_DynamicCast}
   > ShowText(self:Icon Tab Button Legacy Object Reference=@n2)
   > SetStyle!(
    execute
    self:Common Button Base Object Reference=@n2
    InStyle:Common Button Style Class Reference=/FortUILibrary/Foundation/Buttons/ButtonStyles/Tab/ButtonStyle-Tab-Manage.ButtonStyle-Tab-Manage_C
   ) -> (then)
   n3 = On Override Reset Modal!() -> (OutputDelegate, then, bOverride:bool=~) {is=title K2Node_CustomEvent}
   > SetOverrideResetToDefaultModal!(execute, self:Fort Options Menu Object Reference, bOverride:bool=@n3.bOverride) -> (then)
   n4 = HandleTabGamePadActionsNeedBinds!() -> (OutputDelegate, then, ScreenLabels:Array of Texts) {is=title K2Node_CustomEvent}
   > n5 = For Each Loop(Array:Array of Texts=@n4.ScreenLabels) -> (Array Element:Text)
   n6 = Length(TargetArray:Array of Texts=@n4.ScreenLabels)
   > n7 = Branch(Condition="Equal (Integer)"(A=@n5.Array Index, B=int - int(A=@n6, B=1) {is=title K2Node_CallFunction}) {is=title K2Node_CallFunction})
   n8 = Format Text!(
    Format="LOCTEXT(\"Missing Binding {0}/{1}\")"
    0:int=int + int(A=@n5.Array Index, B=1) -> (ReturnValue=0) {is=title K2Node_CommutativeAssociativeBinaryOperator}
    1:int=@n6
   ) -> (Result)
   n9 = Format Text!(
    Format="LOCTEXT(\"** VERY IMPORTANT **\\r\\nPLEASE bind the [{0}] action to a button!\\r\\nNot doing so, will make this action impossible to perform.\")"
    0:Text=@n5
   ) -> (Result)
   > n10 = ShowSimpleConfirmationDialog_NUI!(
    execute:Exec
    Icon:Slate Brush Structure
    Title:Text=@n8
    Message:Text=@n9
    bShowConfirm:bool=true
    bShowDecline:bool=true
    AdditionalContent:Widget Object Reference
    LeftAdditionalContent:Widget Object Reference
    TargetPlayer:Local Player Object Reference
   ) -> (then:Exec, DialogResult:Exec, Result:EFortDialogResult Enum, ResultName:Name) {is=title K2Node_AsyncAction}
   n11 = Switch on EFortDialogResult!(execute=@n10.DialogResult, Selection:EFortDialogResult Enum=@n10) -> (Confirmed:Exec, Declined:Exec, Ignored:Exec, Killed:Exec, TimedOut:Exec) {is=title K2Node_SwitchEnum}
   n12 = Get SettingTabSwitcher() -> (SettingTabSwitcher:Common Widget Switcher Legacy Object Reference)
   > n13 = Cast To TabGamePadConfig_v3!(
    execute
    Object=GetWidgetAtIndex!(
     self:Widget Switcher Object Reference=@n12
     Index:int=GetActiveWidgetIndex!(self:Widget Switcher Object Reference=@n12) -> (ReturnValue:int=0)
    ) -> (ReturnValue:Widget Object Reference)
   ) -> (then, CastFailed, AsTab Game Pad Config V 3:Tab Game Pad Config V 3 Object Reference) {is=title K2Node_DynamicCast}
   > HandleGamepadInfoSelectorFocus(self:Tab Game Pad Config V 3 Object Reference=@n13)
   ShowUnsavedConfirmation(execute=@n11.Declined, self:Fort Options Menu Object Reference)
   ShowSimpleConfirmationDialog_NUI(execute=@n7.else, Title=@n8, Message=@n9, bShowConfirm=true, bShowDecline=false)
   Event HandleTabGamepadConfigActionHandlerStates() {is=title K2Node_Event}
   > Handle Tab Gamepad Config Action Handler States_SAVE()
   Event OnSettingsTabListPopulated() {is=title K2Node_Event}
   n14 = Get SettingTabSwitcher()
   > n15 = Cast To TabVideoOptions_New!(execute, Object=GetWidgetAtIndex(self=@n14, Index=GetActiveWidgetIndex(self=@n14) -> (ReturnValue=0))) -> (then, CastFailed, AsTab Video Options New:Tab Video Options New Object Reference) {is=title K2Node_DynamicCast}
   > n16 = Set bVideoTabSelected(bVideoTabSelected:bool=true) -> (Output_Get:bool)
   n17 = For Loop(
    execute=[@n22.then, @n16.then]
    LastIndex=GetChildrenCount!(self:Panel Widget Object Reference=Get SettingTabSwitcher()) -> (ReturnValue:int=0)
   )
   n18 = GetChildAt!(self:Panel Widget Object Reference=Get SettingTabSwitcher(), Index:int=@n17) -> (ReturnValue:Widget Object Reference)
   > n19 = Cast To TabGamePadConfig_v3(Object=@n18)
   > Set TabGamepadConfig(TabGamepadConfig:Tab Game Pad Config V 3 Object Reference=@n19) -> (Output_Get:Tab Game Pad Config V 3 Object Reference)
   > Bind Event to Gamepad Mapping Info Changed(self:Tab Game Pad Config V 3 Object Reference=@n19, Delegate=Create Event()) {is=title K2Node_AddDelegate}
   > Bind Event to Tab Game Pad Config Back(self:Tab Game Pad Config V 3 Object Reference=@n19, Delegate=Create Event()) {is=title K2Node_AddDelegate}
   > Bind Event to Tab Game Pad Actions That Need Binds(self:Tab Game Pad Config V 3 Object Reference=@n19, Delegate=Create Event()) {is=title K2Node_AddDelegate}
   > Bind Event to Override Reset Modal(self:Tab Game Pad Config V 3 Object Reference=@n19, Delegate=Create Event()) {is=title K2Node_AddDelegate}
   > n20 = Branch(Condition=IsInSubGame!(SubGame:ESubGame Enum=Campaign) -> (ReturnValue:bool))
   > Bind Event to Tab Game Pad Config Apply(self:Tab Game Pad Config V 3 Object Reference=@n19, Delegate=Create Event()) {is=title K2Node_AddDelegate}
   > Bind Event to Gamepad Info Selector Reset(self:Tab Game Pad Config V 3 Object Reference=@n19, Delegate=Create Event()) {is=title K2Node_AddDelegate}
   Bind Event to Tab Game Pad Config Apply(execute=@n20.else, self=@n19, Delegate=Create Event())
   n21 = Cast To TabMouseAndKeyboardBindingOptions_New!(execute=@n19.CastFailed, Object=@n18) -> (then, CastFailed, AsTab Mouse and Keyboard Binding Options New:Tab Mouse And Keyboard Binding Options New Object Reference) {is=title K2Node_DynamicCast}
   > Set OptionsMenu!(execute, OptionsMenu:Options Menu Object Reference=Self-Reference(), self:Tab Mouse And Keyboard Binding Options New Object Reference=@n21) -> (then, Output_Get:Options Menu Object Reference)
   n22 = Set bVideoTabSelected(execute=@n15.CastFailed, bVideoTabSelected=false)
   Event On Activated()
   n23 = IsMobileGame()
   > n24 = Branch(Condition=@n23)
   n25 = MakeLiteralGameplayTagContainer!(Value:Gameplay Tag Container Structure=(GameplayTags=((TagName="UI.ContentOnDemand")))) -> (ReturnValue:Gameplay Tag Container Structure)
   n26 = Get FortHUDContext!() -> (ReturnValue:Fort HUDContext Object Reference) {is=title K2Node_GetBlueprintContextEx}
   > n27 = SetHUDElementVisibility!(
    execute
    self:Fort HUDContext Object Reference=@n26
    HUDElementTags:"Gameplay Tag Container Structure (by ref)"=@n25
    bHideElements:bool=true
    ReasonName:"Name (by ref)"=None
   ) -> (then)
   n28 = Get FortGlobalUIContext!() -> (ReturnValue:Fort Global UIContext Object Reference) {is=title K2Node_GetBlueprintContextEx}
   n29 = Branch(
    execute=[@n27.then, @n24.else]
    Condition="Equal (Enum)"(
     A:ESubGame Enum=GetSubGame!(self:Fort Global UIContext Object Reference=@n28) -> (ReturnValue:ESubGame Enum=Campaign)
     B:ESubGame Enum=Athena
    )
   )
   SetWindowTitleBarState!(
    execute=[@n31.then, @n29.then]
    TitleBarContent:Widget Object Reference
    Mode:EWindowTitleBarMode Enum=Overlay
    bTitleBarDragEnabled:bool=false
    bWindowButtonsVisible:bool=false
    bTitleBarVisible:bool=true
   ) -> (then)
   > n30 = Branch(Condition=IsVisible(self=Get Message() -> (Message:Common Border Object Reference)))
   CenterOnActiveTab(execute=@n30.else, self:Fort Options Menu Object Reference)
   n31 = SetFrontEndCamera!(
    execute=@n29.else
    self:Fort Front End Context Object Reference=Get FortFrontEndContext!() -> (ReturnValue:Fort Front End Context Object Reference) {is=title K2Node_GetBlueprintContextEx}
    NewState:EFrontEndCamera Enum=FrontendDefault
    TransitionParams:View Target Transition Params Structure=(BlendTime=0,BlendFunction=VTBlend_Cubic,BlendExp=2,bLockOutgoing=False)
   ) -> (then)
   Event OnShowResetToDefaultModal() {is=title K2Node_Event}
   n32 = Get TabGamepadConfig() -> (TabGamepadConfig:Tab Game Pad Config V 3 Object Reference)
   > Is Valid(InputObject=@n32)
   > CustomDoResetToDefault(self:Tab Game Pad Config V 3 Object Reference=@n32)
   Event OnSplitscreenUpdateSizing() {is=title K2Node_Event}
   > SetHeightOverride(self=Get SizeBox_ContentLetoSplitscreen() -> (SizeBox_ContentLetoSplitscreen:Size Box Object Reference), InHeightOverride=800)
   > Set MaxDesiredWidth!(
    execute
    MaxDesiredWidth:float=300
    self:Size Box Object Reference=Get SizeBox_HeadingLeto() -> (SizeBox_HeadingLeto:Size Box Object Reference)
   ) -> (then, Output_Get:float)
   Event On Deactivated()
   > Branch(Condition=@n23)
   > SetHUDElementVisibility(self=@n26, HUDElementTags=@n25, bHideElements=false, ReasonName=None)
  ]}
 Handle Tab Gamepad Config Action Handler States_SAVE {nodes [
   Handle Tab Gamepad Config Action Handler States_SAVE() {is=title K2Node_FunctionEntry}
   n1 = Get SettingTabSwitcher()
   > n2 = Cast To TabGamePadConfig_v3(Object=GetWidgetAtIndex(self=@n1, Index=GetActiveWidgetIndex(self=@n1) -> (ReturnValue=0)))
   n3 = Get GamepadInfoSelectorPanel_v3!(self:Tab Game Pad Config V 3 Object Reference=@n2) -> (GamepadInfoSelectorPanel_v3:Gamepad Info Selector Panel V 3 Object Reference)
   n4 = Get!(self:Gamepad Info Selector Panel V 3 Object Reference=@n3) -> (Apply_DataTable:Data Table Object Reference, Apply_RowName:Name)
   > SetActionHandlerState!(
    execute
    self:Common Activatable Panel Legacy Object Reference=@n3
    DataTable:Data Table Object Reference=@n4
    RowName:Name=@n4.Apply_RowName
    State:EInputActionState Enum=Select(
     Option 0:EInputActionState Enum=HiddenAndDisabled
     Option 1:EInputActionState Enum=Enabled
     Index=Get bApplyVisible() -> (bApplyVisible:bool)
    ) -> (ReturnValue:EInputActionState Enum)
   ) -> (then)
  ]}
 HandleGamepadInfoSelectorApply {nodes [
   HandleGamepadInfoSelectorApply() {is=title K2Node_FunctionEntry}
   > HandleApply(self:Fort Options Menu Object Reference)
   n1 = Get SettingTabSwitcher()
   > n2 = Cast To TabGamePadConfig_v3(Object=GetWidgetAtIndex(self=@n1, Index=GetActiveWidgetIndex(self=@n1) -> (ReturnValue=0)))
   n3 = Get GamepadInfoSelectorPanel_v3(self=@n2)
   n4 = Get(self=@n3)
   > SetActionHandlerState(self=@n3, DataTable=@n4, RowName=@n4.Apply_RowName, State=Select(Option 0=HiddenAndDisabled, Option 1=Enabled, Index=Get bApplyVisible()))
  ]}
 HandleGamepadInfoSelectorReset {nodes [
   HandleGamepadInfoSelectorReset() {is=title K2Node_FunctionEntry}
   > OnResetToDefaultInternal!(execute, self:Fort Options Menu Object Reference) -> (then, PassThrough:bool)
   n1 = Get SettingTabSwitcher()
   > n2 = Cast To TabGamePadConfig_v3(Object=GetWidgetAtIndex(self=@n1, Index=GetActiveWidgetIndex(self=@n1) -> (ReturnValue=0)))
   n3 = Get GamepadInfoSelectorPanel_v3(self=@n2)
   n4 = Get(self=@n3)
   > SetActionHandlerState(self=@n3, DataTable=@n4, RowName=@n4.Apply_RowName, State=Select(Option 0=HiddenAndDisabled, Option 1=Enabled, Index=Get bApplyVisible()))
  ]}
 LoadTabClasses {nodes [LoadTabClasses() {is=title K2Node_FunctionEntry}]}
]
variables [TabGamepadConfig {type=TabGamePadConfig_v3_C, default=null}]
widgets {
 info {
  inheritedWidgetCount=9
  namedSlotCount=1
  parentClass=/Script/FortniteUI.FortOptionsMenu
  rootWidgetClass=/Game/Athena/HUD/FortHUDElementWrapper.FortHUDElementWrapper_C
  widgetCount=29
 }
 widgets [
  FortHUDElementWrapper {
   widgetClassPath=/Game/Athena/HUD/FortHUDElementWrapper
   bIsVariable=true
   widget=FortHUDElementWrapper
   properties {elementTag {tagName=HUD}, hUDElementTag {gameplayTags []}}
  }
  Overlay_0 {widgetClassPath=&Overlay, namedSlotHost=FortHUDElementWrapper, widget=Overlay_0, properties {visibility=Visible}}
  Border_Background {
   widgetClassPath=&CommonBorder
   bInherited=true
   bIsVariable=true
   parent=Overlay_0
   widget=Border_Background
   properties {
    style=/FortUILibrary/Styles/Border/Border-MainL.Border-MainL_C
    background {
     bIsDynamicallyLoaded=false
     drawAs=Box
     imageType=NoImage
     mirroring=NoMirror
     resourceName=None
     resourceObject=null
     tiling=NoTile
     imageSize=<128 32>
     margin=<ltrb 0.1 0.45 0.9 0.5>
     outlineSettings {
      bUseBrushTransparency=false
      roundingType=HalfHeightRadius
      width=0
      color {colorUseRule=UseColor_Specified, specifiedColor=<rgba 0 0 0 0>}
      cornerRadii=<0 0 0 0>
     }
     tintColor {colorUseRule=UseColor_Specified, specifiedColor=<rgba 1 1 1 1>}
     uVRegion {bIsValid=false, max=<0 0>, min=<0 0>}
    }
    padding=<ltrb 0 0 0 0>
   }
   slot {class=OverlaySlot, properties {horizontalAlignment=HAlign_Fill, verticalAlignment=VAlign_Fill}}
  }
  BackgroundBlur_0 {
   widgetClassPath=&BackgroundBlur
   parent=Border_Background
   widget=BackgroundBlur_0
   slot {class=BorderSlot, properties {padding=<ltrb 0 0 0 0>}}
  }
  VerticalBox_354 {
   widgetClassPath=&VerticalBox
   parent=BackgroundBlur_0
   widget=VerticalBox_354
   slot {class=BackgroundBlurSlot, properties {padding=<ltrb 0 0 0 0>}}
  }
  CommonBorder_0 {
   widgetClassPath=&CommonBorder
   parent=VerticalBox_354
   widget=CommonBorder_0
   properties {
    style=/FortUILibrary/Styles/Border/Border-SolidBG-ShellBlue.Border-SolidBG-ShellBlue_C
    background {
     bIsDynamicallyLoaded=false
     drawAs=Box
     imageType=NoImage
     mirroring=NoMirror
     resourceName=None
     resourceObject=/FortUILibrary/Styles/Border/Materials/M_UI_Border-SolidBG-ShellBlue
     tiling=NoTile
     imageSize=<240 76>
     margin=<ltrb 0.01 0.05 0.98 0.95>
     outlineSettings {
      bUseBrushTransparency=false
      roundingType=HalfHeightRadius
      width=0
      color {colorUseRule=UseColor_Specified, specifiedColor=<rgba 0 0 0 0>}
      cornerRadii=<0 0 0 0>
     }
     tintColor {colorUseRule=UseColor_Specified, specifiedColor=<rgba 1 1 1 1>}
     uVRegion {bIsValid=false, max=<0 0>, min=<0 0>}
    }
    padding=<ltrb 0 0 0 0>
   }
   slot {class=VerticalBoxSlot, properties {size {sizeRule=Automatic, value=0.2}}}
  }
  WindowTitleBarArea_0 {
   widgetClassPath=/Script/UMG.WindowTitleBarArea
   parent=CommonBorder_0
   widget=WindowTitleBarArea_0
   slot {class=BorderSlot, properties {padding=<ltrb 0 0 0 0>}}
  }
  SafeZone_0 {
   widgetClassPath=&SafeZone
   parent=WindowTitleBarArea_0
   widget=SafeZone_0
   properties {padBottom=false}
   slot {class=WindowTitleBarAreaSlot}
  }
  HorizontalBox_1 {
   widgetClassPath=&HorizontalBox
   parent=SafeZone_0
   widget=HorizontalBox_1
   properties {flowDirectionPreference=Culture}
   slot {class=SafeZoneSlot}
  }
  HorizontalBox_0 {
   widgetClassPath=&HorizontalBox
   parent=HorizontalBox_1
   widget=HorizontalBox_0
   slot {class=HorizontalBoxSlot, properties {size {sizeRule=Fill, value=1}}}
  }
  HorizontalBox_5 {
   widgetClassPath=&HorizontalBox
   parent=HorizontalBox_0
   widget=HorizontalBox_5
   slot {class=HorizontalBoxSlot, properties {size {sizeRule=Fill, value=1}}}
  }
  Image_0 {
   widgetClassPath=&Image
   parent=HorizontalBox_5
   widget=Image_0
   properties {
    visibility=Collapsed
    brush {
     bIsDynamicallyLoaded=false
     drawAs=Image
     imageType=NoImage
     mirroring=NoMirror
     resourceName=None
     resourceObject=/Game/UI/Foundation/Textures/Icons/Social/T-Icon-Config-64
     tiling=NoTile
     imageSize=<52 52>
     margin=<ltrb 0 0 0 0>
     outlineSettings {
      bUseBrushTransparency=false
      roundingType=HalfHeightRadius
      width=0
      color {colorUseRule=UseColor_Specified, specifiedColor=<rgba 0 0 0 0>}
      cornerRadii=<0 0 0 0>
     }
     tintColor {colorUseRule=UseColor_Specified, specifiedColor=<rgba 1 1 1 1>}
     uVRegion {bIsValid=false, max=<0 0>, min=<0 0>}
    }
   }
   slot {class=HorizontalBoxSlot, properties {verticalAlignment=VAlign_Center}}
  }
  SizeBox_HeadingLeto {
   widgetClassPath=&SizeBox
   bIsVariable=true
   parent=HorizontalBox_5
   widget=SizeBox_HeadingLeto
   properties {maxDesiredWidth=300}
   slot {class=HorizontalBoxSlot, properties {verticalAlignment=VAlign_Center, padding=<ltrb 48 0 0 0>, size {sizeRule=Fill, value=1}}}
  }
  Text_Title {
   widgetClassPath=&CommonTextBlock
   bInherited=true
   bIsVariable=true
   parent=SizeBox_HeadingLeto
   widget=Text_Title
   properties {
    style=/FortUILibrary/Styles/Typography/_TextStyle_SubH4._TextStyle_SubH4_C
    text=Settings
    textTransformPolicy=ToUpper
    font {
     bForceMonospaced=false
     bMaterialIsStencil=false
     fontMaterial=null
     fontObject=&BurbankSmall
     letterSpacing=0
     monospacedWidth=1
     size=24
     skewAmount=0
     typefaceFontName=Black
     outlineSettings {
      bApplyOutlineToDropShadows=false
      bMiteredCorners=false
      bSeparateFillAlpha=false
      outlineMaterial=null
      outlineSize=0
      outlineColor=<rgba 0 0 0 1>
     }
    }
    shadowOffset=<0 0>
    strikeBrush {
     bIsDynamicallyLoaded=false
     drawAs=NoDrawType
     imageType=NoImage
     mirroring=NoMirror
     resourceName=None
     resourceObject=null
     tiling=NoTile
     imageSize=<32 32>
     margin=<ltrb 0 0 0 0>
     outlineSettings {
      bUseBrushTransparency=false
      roundingType=HalfHeightRadius
      width=0
      color {colorUseRule=UseColor_Specified, specifiedColor=<rgba 0 0 0 0>}
      cornerRadii=<0 0 0 0>
     }
     tintColor {colorUseRule=UseColor_Specified, specifiedColor=<rgba 1 1 1 1>}
     uVRegion {bIsValid=false, max=<0 0>, min=<0 0>}
    }
   }
   slot {class=SizeBoxSlot}
  }
  TabList_Categories {
   widgetClassPath=/Game/UI/Foundation/TabbedView/HorizontalTabList_Legacy
   bInherited=true
   bIsVariable=true
   parent=HorizontalBox_0
   widget=TabList_Categories
   properties {
    bAutoListenForInput=true
    buttonStyle=&ButtonStyle-Tab-Manage_C
    flowDirectionPreference=LeftToRight
    forceTabLabelsCollapsed=true
    onTabButtonCreation=(&self:WidgetTree.TabList_Categories.HandleOnTabButtonCreated)
    onTabButtonRemoval=(&self:WidgetTree.TabList_Categories.HandleOnTabButtonRemoved)
    useButtonStyle=true
    nextTabInputActionData {dataTable=&InputActionDataTable, rowName=MainTabRight}
    previousTabInputActionData {dataTable=&InputActionDataTable, rowName=MainTabLeft}
   }
   slot {class=HorizontalBoxSlot, properties {horizontalAlignment=HAlign_Center, size {sizeRule=Fill, value=1}}}
  }
  windowchromebuttons {
   widgetClassPath=/Game/UI/Foundation/Shell/windowchromebuttons
   parent=HorizontalBox_0
   widget=windowchromebuttons
   slot {class=HorizontalBoxSlot, properties {size {sizeRule=Fill, value=1}}}
  }
  VisibilityBorder_Touch {
   widgetClassPath=&CommonHardwareVisibilityBorder
   parent=HorizontalBox_0
   widget=VisibilityBorder_Touch
   properties {visibilityQuery {
     autoDescription=" ANY( Input.Touch )"
     tokenStreamVersion=0
     userDescription=""
     queryTokenStream [0,1,1,1,0]
     tagDictionary [{tagName=Input.Touch}]
    }}
   slot {class=HorizontalBoxSlot, properties {padding=<ltrb 0 0 48 0>}}
  }
  Button_CloseTouch {
   widgetClassPath=/FortUILibrary/Foundation/Buttons/CloseButton
   bInherited=true
   bIsVariable=true
   parent=VisibilityBorder_Touch
   widget=Button_CloseTouch
   properties {
    onButtonBaseClicked=(&self:WidgetTree.Button_CloseTouch.HandleOnButtonClicked)
    onButtonBaseDoubleClicked=(&self:WidgetTree.Button_CloseTouch.HandleOnButtonDoubleClicked)
    onButtonBaseHovered=(&self:WidgetTree.Button_CloseTouch.HandleOnButtonHovered)
    onButtonBaseUnhovered=(&self:WidgetTree.Button_CloseTouch.HandleOnButtonUnhovered)
    onSelectedChangedBase=(&self:WidgetTree.Button_CloseTouch.HandleOnSelectedChanged)
   }
   slot {class=BorderSlot, properties {padding=<ltrb 0 0 0 0>}}
  }
  SizeBox_ContentLetoSplitscreen {
   widgetClassPath=&SizeBox
   bIsVariable=true
   parent=VerticalBox_354
   widget=SizeBox_ContentLetoSplitscreen
   slot {class=VerticalBoxSlot, properties {size {sizeRule=Fill, value=1}}}
  }
  SettingTabSwitcher {
   widgetClassPath=&CommonWidgetSwitcherLegacy
   bInherited=true
   bIsVariable=true
   parent=SizeBox_ContentLetoSplitscreen
   widget=SettingTabSwitcher
   properties {navigation=&self:WidgetTree.SettingTabSwitcher.WidgetNavigation_0, transitionDuration=0}
   slot {class=SizeBoxSlot}
  }
  Message {
   widgetClassPath=&CommonBorder
   bInherited=true
   bIsVariable=true
   parent=Overlay_0
   widget=Message
   properties {
    horizontalAlignment=HAlign_Center
    navigation=&self:WidgetTree.Message.WidgetNavigation_1
    verticalAlignment=VAlign_Center
    visibility=Collapsed
    brushColor=<rgba 0 0 0 0.9>
    padding=<ltrb 0 0 0 0>
   }
   slot {class=OverlaySlot, properties {horizontalAlignment=HAlign_Fill, verticalAlignment=VAlign_Fill}}
  }
  BackgroundBorder {
   widgetClassPath=&CommonBorder
   parent=Message
   widget=BackgroundBorder
   properties {
    horizontalAlignment=HAlign_Center
    navigation=&self:WidgetTree.BackgroundBorder.WidgetNavigation_1
    style=/FortUILibrary/Styles/Border/Border-Bottom-Box-Rounded-DkBlue.Border-Bottom-Box-Rounded-DkBlue_C
    verticalAlignment=VAlign_Center
    brushColor=<rgba 0 0 0 0.9>
    padding=<ltrb 25 25 25 25>
   }
   slot {class=BorderSlot, properties {horizontalAlignment=HAlign_Center, verticalAlignment=VAlign_Center, padding=<ltrb 0 0 0 0>}}
  }
  VerticalBox {
   widgetClassPath=&VerticalBox
   parent=BackgroundBorder
   widget=VerticalBox
   slot {class=BorderSlot, properties {horizontalAlignment=HAlign_Center, verticalAlignment=VAlign_Center, padding=<ltrb 25 25 25 25>}}
  }
  SizeBox_537 {
   widgetClassPath=&SizeBox
   parent=VerticalBox
   widget=SizeBox_537
   properties {bOverride_MaxDesiredWidth=true, maxDesiredWidth=1000}
   slot {class=VerticalBoxSlot, properties {padding=<ltrb 0 0 0 12>}}
  }
  OverlayText {
   widgetClassPath=&CommonTextBlock
   parent=SizeBox_537
   widget=OverlayText
   properties {
    autoWrapText=true
    justification=Center
    style=/FortUILibrary/Styles/Typography/Foundation/TextStyle-Base-L-B.TextStyle-Base-L-B_C
    text=Do you want to keep these video settings?
    wrapTextAt=1200
    font {
     bForceMonospaced=false
     bMaterialIsStencil=false
     fontMaterial=null
     fontObject=&NotoSans
     letterSpacing=0
     monospacedWidth=1
     size=32
     skewAmount=0
     typefaceFontName=NotoSans-Bold
     outlineSettings {
      bApplyOutlineToDropShadows=false
      bMiteredCorners=false
      bSeparateFillAlpha=false
      outlineMaterial=null
      outlineSize=0
      outlineColor=<rgba 0 0 0 1>
     }
    }
    shadowOffset=<0 0>
   }
   slot {class=SizeBoxSlot, properties {verticalAlignment=VAlign_Center}}
  }
  HorizontalBox {
   widgetClassPath=&HorizontalBox
   parent=VerticalBox
   widget=HorizontalBox
   properties {navigation=&self:WidgetTree.HorizontalBox.WidgetNavigation_0, visibility=Visible}
   slot {class=VerticalBoxSlot, properties {horizontalAlignment=HAlign_Center}}
  }
  VideoSettingsAccept {
   widgetClassPath=&IconTextButton
   bInherited=true
   bIsVariable=true
   parent=HorizontalBox
   widget=VideoSettingsAccept
   properties {
    bHideInputActionWithKeyboard=true
    bIsFocusable=false
    buttonText=Yes
    navigation=&self:WidgetTree.VideoSettingsAccept.WidgetNavigation_0
    onButtonBaseClicked=(&self:WidgetTree.VideoSettingsAccept.HandleOnButtonClicked)
    onButtonBaseDoubleClicked=(&self:WidgetTree.VideoSettingsAccept.HandleOnButtonDoubleClicked)
    onButtonBaseHovered=(&self:WidgetTree.VideoSettingsAccept.HandleOnButtonHovered)
    onButtonBaseUnhovered=(&self:WidgetTree.VideoSettingsAccept.HandleOnButtonUnhovered)
    onSelectedChangedBase=(&self:WidgetTree.VideoSettingsAccept.HandleOnSelectedChanged)
    visibility=Collapsed
    triggeringInputAction {dataTable=&InputActionDataTable, rowName=GenericForward}
   }
   slot {class=HorizontalBoxSlot, properties {horizontalAlignment=HAlign_Left, padding=<ltrb 0 0 50 0>, size {sizeRule=Fill, value=1}}}
  }
  VideoSettingsCancel {
   widgetClassPath=&IconTextButton
   bInherited=true
   bIsVariable=true
   parent=HorizontalBox
   widget=VideoSettingsCancel
   properties {
    bHideInputActionWithKeyboard=true
    bIsFocusable=false
    buttonText=No
    navigation=&self:WidgetTree.VideoSettingsCancel.WidgetNavigation_1
    onButtonBaseClicked=(&self:WidgetTree.VideoSettingsCancel.HandleOnButtonClicked)
    onButtonBaseDoubleClicked=(&self:WidgetTree.VideoSettingsCancel.HandleOnButtonDoubleClicked)
    onButtonBaseHovered=(&self:WidgetTree.VideoSettingsCancel.HandleOnButtonHovered)
    onButtonBaseUnhovered=(&self:WidgetTree.VideoSettingsCancel.HandleOnButtonUnhovered)
    onSelectedChangedBase=(&self:WidgetTree.VideoSettingsCancel.HandleOnSelectedChanged)
    visibility=Collapsed
    triggeringInputAction {dataTable=&InputActionDataTable, rowName=GenericBack}
   }
   slot {class=HorizontalBoxSlot, properties {horizontalAlignment=HAlign_Right, padding=<ltrb 50 0 0 0>, size {sizeRule=Fill, value=1}}}
  }
  SavingLayer {
   widgetClassPath=&NamedSlot
   bInherited=true
   bIsVariable=true
   parent=Overlay_0
   widget=SavingLayer
   properties {bExposeOnInstanceOnly=true}
   slot {class=OverlaySlot, properties {horizontalAlignment=HAlign_Fill, verticalAlignment=VAlign_Fill}}
  }
 ]
}
