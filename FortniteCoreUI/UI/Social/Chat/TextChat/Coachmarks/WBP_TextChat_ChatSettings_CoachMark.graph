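# Onboarding coach mark for text chat settings, explaining who can chat and how text reporting works.
# It subclasses FortSidebarOnboardTooltipWidget around a UIKit coach mark, setting the description text in PreConstruct.
# Show activates the coach mark; the "Got it" button calls HandleOtherActionClicked, and that click or tooltip end notifies close-animation finished and deactivates both widgets.
# A SetCoachmarkArrowPosition function forwards an arrow position to the coach mark.
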WidgetBlueprint /FortniteCoreUI/UI/Social/Chat/TextChat/Coachmarks/WBP_TextChat_ChatSettings_CoachMark
graphs [
 EventGraph {nodes [
   "Button Other Action Clicked (WBP_UIKit_CoachMark)"() {is=title K2Node_ComponentBoundEvent}
   > n1 = HandleOtherActionClicked(self:Fort Sidebar Onboard Tooltip Widget Object Reference)
   NotifyCloseAnimationFinished(execute=[@n1.then, @n2.then], self:Fort Sidebar Onboard Tooltip Widget Object Reference)
   > DeactivateWidget(self=Get WBP_UIKit_CoachMark() -> (WBP_UIKit_CoachMark:WBP UIKit Coach Mark Object Reference))
   > DeactivateWidget()
   n2 = Event OnEndTooltip() {is=title K2Node_Event}
   Event OnShowTooltip() {is=title K2Node_Event}
   > ActivateWidget(self=Get WBP_UIKit_CoachMark())
   Event PreConstruct()
   > SetTextDescription!(
    execute
    self:WBP UIKit Coach Mark Base Object Reference=Get WBP_UIKit_CoachMark()
    Text:Text="LOCTEXT(\"Decide who you can text chat with and whether text reporting is Always On or Off When Possible.\\r\\n\\r\\nText reporting is Always On for players under the age of 18. You can see if text reporting is currently on or off in the text input field before typing a message.\\r\\n\\r\\nQuickly report a conversation by pressing the Report Button while text chat is open.\")"
   ) -> (then, Empty:bool)
  ]}
 SetCoachmarkArrowPosition {nodes [
   n1 = SetCoachmarkArrowPosition!() -> (then, ArrowPosition:E_UI_OutBoardPosition Enum=NewEnumerator3) {is=title K2Node_FunctionEntry}
   > SetArrowPosition!(execute, self:WBP UIKit Coach Mark Base Object Reference=Get WBP_UIKit_CoachMark(), ArrowPosition:E_UI_OutBoardPosition Enum=@n1) -> (then)
  ]}
]
variables []
widgets {
 info {
  inheritedWidgetCount=2
  parentClass=/Script/FortniteUI.FortSidebarOnboardTooltipWidget
  rootWidgetClass=/FortUILibrary/UIKit/CoachMark/WBP_UIKit_CoachMark.WBP_UIKit_CoachMark_C
  widgetCount=3
 }
 widgets [
  WBP_UIKit_CoachMark {
   widgetClassPath=/FortUILibrary/UIKit/CoachMark/WBP_UIKit_CoachMark
   bIsVariable=true
   widget=WBP_UIKit_CoachMark
   properties {
    arrowPosition=Right_Down
    bIsFocusable=true
    dismissButtonClass=null
    navigation=&self:WidgetTree.WBP_UIKit_CoachMark.WidgetNavigation_0
    shouldShowDismissButton=false
    textDescription="Decide who you can text chat with and whether text reporting is Always On or Off When Possible. Text Reporting is Always On for players under the age of 18.The text reporting status is shown in the text input field before typing a message.  "
    textDismissButton=""
    textHeader=Access text chat settings
    textOtherActionButton=Got it
    visibility=SelfHitTestInvisible
   }
  }
  Dismiss_Button {widgetClassPath=&CommonButtonLegacy, bInherited=true, widget=null}
  OtherAction_Button {widgetClassPath=&CommonButtonLegacy, bInherited=true, widget=null}
 ]
}
