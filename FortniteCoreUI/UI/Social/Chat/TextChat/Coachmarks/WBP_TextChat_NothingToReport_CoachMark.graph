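# Text chat coach mark telling the player that message reporting is unavailable and pointing them to the text reporting setting via a "Take me there" button.
# It extends FortSidebarOnboardTooltipWidget around a WBP_UIKit_CoachMark, which activates when the tooltip shows.
# Dismiss and other-action clicks broadcast dispatchers, the latter after calling HandleOtherActionClicked.
# Both clicks and tooltip end notify close completion and deactivate the widgets; a function sets arrow position.
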WidgetBlueprint /FortniteCoreUI/UI/Social/Chat/TextChat/Coachmarks/WBP_TextChat_NothingToReport_CoachMark
graphs [
 EventGraph {nodes [
   "Button Dismiss Clicked (WBP_UIKit_CoachMark)"() {is=title K2Node_ComponentBoundEvent}
   > n1 = Call On Dismiss Action Button Clicked!(execute:Exec, self:WBP Text Chat Nothing To Report Coach Mark Object Reference) -> (then:Exec) {is=title K2Node_CallDelegate}
   NotifyCloseAnimationFinished(execute=[@n2.then, @n3.then, @n1.then], self:Fort Sidebar Onboard Tooltip Widget Object Reference)
   > DeactivateWidget(self=Get WBP_UIKit_CoachMark() -> (WBP_UIKit_CoachMark:WBP UIKit Coach Mark Object Reference))
   > DeactivateWidget(self=Self-Reference())
   "Button Other Action Clicked (WBP_UIKit_CoachMark)"() {is=title K2Node_ComponentBoundEvent}
   > HandleOtherActionClicked(self:Fort Sidebar Onboard Tooltip Widget Object Reference)
   > n2 = Call On Other Action Button Clicked!(execute:Exec, self:WBP Text Chat Nothing To Report Coach Mark Object Reference) -> (then:Exec) {is=title K2Node_CallDelegate}
   n3 = Event OnEndTooltip() {is=title K2Node_Event}
   Event OnShowTooltip() {is=title K2Node_Event}
   > ActivateWidget(self=Get WBP_UIKit_CoachMark())
  ]}
 OnDismissActionButtonClicked {nodes [OnDismissActionButtonClicked() {is=title K2Node_FunctionEntry}]}
 OnOtherActionButtonClicked {nodes [OnOtherActionButtonClicked() {is=title K2Node_FunctionEntry}]}
 SetCoachmarkArrowPosition {nodes [
   n1 = SetCoachmarkArrowPosition!() -> (then, ArrowPosition:E_UI_OutBoardPosition Enum=NewEnumerator3) {is=title K2Node_FunctionEntry}
   > SetArrowPosition!(execute, self:WBP UIKit Coach Mark Base Object Reference=Get WBP_UIKit_CoachMark(), ArrowPosition:E_UI_OutBoardPosition Enum=@n1) -> (then)
  ]}
]
variables [
 OnDismissActionButtonClicked {}
 OnOtherActionButtonClicked {}
]
widgets {
 info {
  inheritedWidgetCount=2
  parentClass=/Script/FortniteUI.FortSidebarOnboardTooltipWidget
  rootWidgetClass=/FortUILibrary/UIKit/CoachMark/WBP_UIKit_CoachMark.WBP_UIKit_CoachMark_C
  widgetCount=3
 }
 widgets [
  WBP_UIKit_CoachMark {
   widgetClassPath=/FortUILibrary/UIKit/CoachMark/WBP_UIKit_CoachMark
   bIsVariable=true
   widget=WBP_UIKit_CoachMark
   properties {
    arrowPosition=Right_Down
    bIsFocusable=true
    navigation=&self:WidgetTree.WBP_UIKit_CoachMark.WidgetNavigation_0
    textDescription=To report messages ensure that text reporting is set to Always On in your settings and that there are reportable messages in your text chat.
    textHeader=Reporting is not available
    textOtherActionButton=Take me there
    visibility=SelfHitTestInvisible
    widget Width=650
    in Decorator Classes [/FortniteCoreUI/UI/Social/Chat/TextChat/RichTextDecorator/RichText_TextChatDecorator.RichText_TextChatDecorator_C, null]
   }
  }
  Dismiss_Button {widgetClassPath=&CommonButtonLegacy, bInherited=true, widget=null}
  OtherAction_Button {widgetClassPath=&CommonButtonLegacy, bInherited=true, widget=null}
 ]
}
