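# Social text chat restriction message widget showing an alert icon and wrapping text on an optional backplate, with named slots for an extra icon and button.
# Setters control text, backplate, button-area visibility and a tint applied to icon and text through a parent Border, chosen from three message types.
# PreConstruct applies default tint, backplate and compact-UI icon sizing; only at design time it hides the icon slot when empty.
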WidgetBlueprint /FortniteCoreUI/UI/Social/Chat/TextChat/WBP_Social_Restrictions_Internal
graphs [
 EventGraph {nodes [
   n1 = Event PreConstruct()
   > n2 = Sequence()
   > Branch(Condition=@n1.IsDesignTime)
   n3 = Get OptionalIcon_Slot() -> (OptionalIcon_Slot:Named Slot Object Reference)
   > SetVisibility(
    self=@n3
    InVisibility=Select(Option 0=Collapsed, Option 1=SelfHitTestInvisible, Index=HasAnyChildren!(self:Panel Widget Object Reference=@n3) -> (ReturnValue:bool))
   )
   SetWarningColor!(execute=@n2.then_1, self, MessageType:E_UI_SocialRestriction_MessageType Enum=NewEnumerator0) -> (then)
   > SetBackplateVisible!(execute, self, IsBackplateVisible:bool=Get ShowBackplate() -> (ShowBackplate:bool)) -> (then)
   > SetCompactUIStyle() {is=title K2Node_CallFunction}
  ]}
 SetBackplateVisible {nodes [
   n1 = SetBackplateVisible!() -> (then, IsBackplateVisible:bool=~) {is=title K2Node_FunctionEntry}
   > n2 = Set ShowBackplate(ShowBackplate:bool=@n1) -> (Output_Get:bool)
   > SetVisibility(self=Get Backplate() -> (Backplate:Image Object Reference), InVisibility=Select(Option 0=Collapsed, Option 1=HitTestInvisible, Index=@n2))
  ]}
 SetButtonContainerVisibility {nodes [
   n1 = SetButtonContainerVisibility!() -> (then, ContainerVisibility:ESlateVisibility Enum=Visible) {is=title K2Node_FunctionEntry}
   > SetVisibility(
    self=Get SpacerVBox_ToCollapsePaddingBeforeButton() -> (SpacerVBox_ToCollapsePaddingBeforeButton:Vertical Box Object Reference)
    InVisibility=@n1
   )
  ]}
 SetCompactUIStyle {nodes [
   SetCompactUIStyle() {is=title K2Node_FunctionEntry}
   > Branch(Condition=IsUsingCompactUI())
   n1 = Get WarningIcon() -> (WarningIcon:Image Object Reference)
   > SetDesiredSizeOverride!(execute, self:Image Object Reference=@n1, DesiredSize:Vector 2D Structure=(X=32,Y=32)) -> (then)
   n2 = SlotAsHorizontalBoxSlot!(Widget:Widget Object Reference=@n1) -> (ReturnValue:Horizontal Box Slot Object Reference)
   n3 = Get!(self:Horizontal Box Slot Object Reference=@n2) -> (Padding_Left:float=~, Padding_Top:float=~, Padding_Right:float=~, Padding_Bottom:float=~)
   > SetPadding!(
    execute
    self:Horizontal Box Slot Object Reference=@n2
    InPadding_Left:float=@n3
    InPadding_Top:float=0
    InPadding_Right:float=8
    InPadding_Bottom:float=@n3.Padding_Bottom
   ) -> (then)
  ]}
 SetWarningColor {nodes [
   n1 = SetWarningColor!() -> (then, MessageType:E_UI_SocialRestriction_MessageType Enum=NewEnumerator0) {is=title K2Node_FunctionEntry}
   > SetContentColorAndOpacity!(
    execute
    self:Border Object Reference=Get TintedBorder() -> (TintedBorder:Border Object Reference)
    InContentColorAndOpacity:Linear Color Structure=Select!(
     NewEnumerator0:Linear Color Structure=(R=0.5776,G=0.5776,B=0.5776,A=0.851)
     NewEnumerator1:Linear Color Structure=(R=0.1144,G=0.7682,B=0.1356,A=1)
     NewEnumerator2:Linear Color Structure=(R=0.9301,G=1,B=0.0103,A=1)
     Index:E_UI_SocialRestriction_MessageType Enum=@n1
    ) -> (ReturnValue:Linear Color Structure)
   ) -> (then)
  ]}
 SetWarningText {nodes [
   n1 = SetWarningText!() -> (then, WarningText:Text) {is=title K2Node_FunctionEntry}
   > Set Text(Text=@n1, self=Get WarningText() -> (WarningText:Common Text Object Reference))
  ]}
]
variables [
 ShowBackplate {type=boolean, default=true}
 TextColor_Alert {type=LinearColor, default=<rgba 0.9301 1 0.0103 1>}
 TextColor_Default {type=LinearColor, default=<rgba 0.5776 0.5776 0.5776 0.851>}
 TextColor_Secondary {type=LinearColor, default=<rgba 0.1144 0.7682 0.1356 1>}
]
widgets {
 info {inheritedWidgetCount=0, parentClass=&CommonUserWidget, rootWidgetClass=&GridPanel, widgetCount=10}
 widgets [
  Grid_Root {widgetClassPath=&GridPanel, bIsVariable=true, widget=Grid_Root, properties {bOverride_Cursor=true, columnFill [1], rowFill [0,0]}}
  Backplate {
   widgetClassPath=&Image
   bIsVariable=true
   parent=Grid_Root
   widget=Backplate
   properties {brush {
     bIsDynamicallyLoaded=false
     drawAs=Image
     imageType=FullColor
     mirroring=NoMirror
     resourceName=None
     resourceObject=/FortniteCoreUI/UI/Social/Chat/TextChat/Materials/MI_UI_TextChat_RestrictionsBackplate
     tiling=NoTile
     imageSize=<32 32>
     margin=<ltrb 0 0 0 0>
     outlineSettings {
      bUseBrushTransparency=false
      roundingType=HalfHeightRadius
      width=0
      color {colorUseRule=UseColor_Specified, specifiedColor=<rgba 0 0 0 0>}
      cornerRadii=<0 0 0 0>
     }
     tintColor {colorUseRule=UseColor_Specified, specifiedColor=<rgba 1 1 1 1>}
     uVRegion {bIsValid=false, max=<0 0>, min=<0 0>}
    }}
   slot {class=GridSlot, properties {rowSpan=2}}
  }
  TintedBorder {
   widgetClassPath=&Border
   bIsVariable=true
   parent=Grid_Root
   widget=TintedBorder
   properties {
    background {
     bIsDynamicallyLoaded=false
     drawAs=Image
     imageType=NoImage
     mirroring=NoMirror
     resourceName=None
     resourceObject=null
     tiling=NoTile
     imageSize=<32 32>
     margin=<ltrb 0 0 0 0>
     outlineSettings {
      bUseBrushTransparency=false
      roundingType=HalfHeightRadius
      width=0
      color {colorUseRule=UseColor_Specified, specifiedColor=<rgba 0 0 0 0>}
      cornerRadii=<0 0 0 0>
     }
     tintColor {colorUseRule=UseColor_Specified, specifiedColor=<rgba 1 1 1 0>}
     uVRegion {bIsValid=false, max=<0 0>, min=<0 0>}
    }
    padding=<ltrb 0 0 0 0>
   }
   slot {class=GridSlot, properties {layer=10, padding=<ltrb 18 18 18 0>}}
  }
  HorizontalBox_0 {
   widgetClassPath=&HorizontalBox
   parent=TintedBorder
   widget=HorizontalBox_0
   slot {class=BorderSlot, properties {padding=<ltrb 0 0 0 0>}}
  }
  WarningIcon {
   widgetClassPath=&Image
   bIsVariable=true
   parent=HorizontalBox_0
   widget=WarningIcon
   properties {
    visibility=SelfHitTestInvisible
    brush {
     bIsDynamicallyLoaded=false
     drawAs=Image
     imageType=FullColor
     mirroring=NoMirror
     resourceName=None
     resourceObject=/FortUILibrary/Textures/MaterialInstances/MI_UIKit_Icon_Alert
     tiling=NoTile
     imageSize=<28 28>
     margin=<ltrb 0 0 0 0>
     outlineSettings {
      bUseBrushTransparency=false
      roundingType=HalfHeightRadius
      width=0
      color {colorUseRule=UseColor_Specified, specifiedColor=<rgba 0 0 0 0>}
      cornerRadii=<0 0 0 0>
     }
     tintColor {colorUseRule=UseColor_Specified, specifiedColor=<rgba 1 1 1 1>}
     uVRegion {bIsValid=false, max=<0 0>, min=<0 0>}
    }
   }
   slot {class=HorizontalBoxSlot, properties {verticalAlignment=VAlign_Top}}
  }
  OptionalIcon_Slot {widgetClassPath=&NamedSlot, bIsVariable=true, parent=HorizontalBox_0, widget=OptionalIcon_Slot, slot {class=HorizontalBoxSlot}}
  WarningText {
   widgetClassPath=&CommonTextBlock
   bIsVariable=true
   parent=HorizontalBox_0
   widget=WarningText
   properties {
    autoWrapText=true
    lineHeightPercentage=1.1
    mobileFontSizeMultiplier=1.5
    text=Social warning message long enough to wrap to a second line
    font {
     bForceMonospaced=false
     bMaterialIsStencil=false
     fontMaterial=null
     fontObject=&HeadingNow
     letterSpacing=10
     monospacedWidth=1
     size=13.5
     skewAmount=0
     typefaceFontName=74 Regular
     outlineSettings {
      bApplyOutlineToDropShadows=false
      bMiteredCorners=false
      bSeparateFillAlpha=false
      outlineMaterial=null
      outlineSize=0
      outlineColor=<rgba 0 0 0 1>
     }
    }
    shadowOffset=<0 0>
   }
   slot {class=HorizontalBoxSlot, properties {verticalAlignment=VAlign_Center, padding=<ltrb 8 0 0 0>, size {sizeRule=Fill, value=1}}}
  }
  SpacerVBox_ToCollapsePaddingBeforeButton {
   widgetClassPath=&VerticalBox
   bIsVariable=true
   parent=Grid_Root
   widget=SpacerVBox_ToCollapsePaddingBeforeButton
   properties {bOverride_Cursor=true}
   slot {class=GridSlot, properties {layer=10, row=1, padding=<ltrb 18 0 18 18>}}
  }
  Spacer_BetweenButtonAndText {
   widgetClassPath=&Spacer
   parent=SpacerVBox_ToCollapsePaddingBeforeButton
   widget=Spacer_BetweenButtonAndText
   properties {size=<1 18>}
   slot {class=VerticalBoxSlot}
  }
  CustomButton_Slot {
   widgetClassPath=&NamedSlot
   bIsVariable=true
   parent=SpacerVBox_ToCollapsePaddingBeforeButton
   widget=CustomButton_Slot
   slot {class=VerticalBoxSlot}
  }
 ]
}
