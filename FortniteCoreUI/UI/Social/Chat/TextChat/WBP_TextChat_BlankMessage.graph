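# Widget Blueprint for the placeholder shown in an empty text chat: centered, wrapping text over a tinted material background.
# SetBlankMessageText switches on channel type, showing a "quiet in here" prompt, or for a disabled Game channel a notice that chat unlocks upon joining a game.
# PreConstruct defaults to the Party message and swaps style tags to larger body text when NewTextChat.UseLargeChatTextSize is true.
# The Mobile tag array variables are not referenced in these graphs.
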WidgetBlueprint /FortniteCoreUI/UI/Social/Chat/TextChat/WBP_TextChat_BlankMessage
graphs [
 EventGraph {nodes [
   Event PreConstruct()
   > SetBlankMessageText!(execute, self, ChannelType:EFortTextChatChannelType Enum=Party, ChannelEnabled:bool=true) -> (then)
   > n1 = GetConsoleVariableBoolValue!(execute, VariableName:String=NewTextChat.UseLargeChatTextSize) -> (then, ReturnValue:bool)
   > Branch(Condition=@n1)
   > UpdateWidgetStyleTags!(
    execute
    InWidget:Widget Object Reference=Get BlankMessage_Txt() -> (BlankMessage_Txt:Common Text Object Reference)
    InAddedTags:Array of Gameplay Tag Structures=Make Array!("[0]"=(TagName="StyleSheet.Style.20"), "[1]":Gameplay Tag Structure=(TagName="StyleSheet.Style.Body")) -> (Array)
    InRemovedTags:Array of Gameplay Tag Structures=Make Array("[0]"=(TagName="StyleSheet.Style.10"), "[1]"=(TagName="StyleSheet.Style.Footnote"))
   ) -> (then, ReturnValue:bool)
   Event Construct()
   Event Tick()
  ]}
 SetBlankMessageText {nodes [
   n1 = SetBlankMessageText!() -> (then, ChannelType:EFortTextChatChannelType Enum=All, ChannelEnabled:bool=~) {is=title K2Node_FunctionEntry}
   > n2 = Switch on EFortTextChatChannelType!(execute, Selection:EFortTextChatChannelType Enum=@n1) -> (All:Exec, Party:Exec, Game:Exec, Whisper:Exec, Persistent:Exec, AddCmd:Exec, MAX_NUM:Exec) {is=title K2Node_SwitchEnum}
   n3 = Branch(execute=@n2.Game, Condition=@n1.ChannelEnabled)
   "SetText (Text)"(
    execute=[@n3.then, @n2.AddCmd, @n2.All, @n2.MAX_NUM, @n2.Party, @n2.Whisper]
    self=Get BlankMessage_Txt()
    InText="LOCTEXT(\"It’s quiet in here! Chat in any available tab or start a new chat with a friend!\")"
   )
   "SetText (Text)"(
    execute=@n3.else
    self=Get BlankMessage_Txt()
    InText="LOCTEXT(\"This chat will be available once you join a game with game chat enabled.\")"
   )
  ]}
]
variables [
 Mobile Added Tags {type="array<GameplayTag>", default [{tagName=StyleSheet.Style.10}]}
 Mobile Removed Tags {type="array<GameplayTag>", default [{tagName=StyleSheet.Style.20}]}
]
widgets {
 info {inheritedWidgetCount=0, parentClass=&UserWidget, rootWidgetClass=&GridPanel, widgetCount=3}
 widgets [
  GridPanel_17 {widgetClassPath=&GridPanel, widget=GridPanel_17, properties {columnFill [1], rowFill [1]}}
  Bg {
   widgetClassPath=&Image
   bIsVariable=true
   parent=GridPanel_17
   widget=Bg
   properties {
    visibility=SelfHitTestInvisible
    brush {
     bIsDynamicallyLoaded=false
     drawAs=Image
     imageType=FullColor
     mirroring=NoMirror
     resourceName=None
     resourceObject=/FortniteCoreUI/UI/Social/Chat/TextChat/MaterialInstances/MI_UI_TextChat_BlankMessage_Bg
     tiling=NoTile
     imageSize=<32 32>
     margin=<ltrb 0 0 0 0>
     outlineSettings {
      bUseBrushTransparency=false
      roundingType=HalfHeightRadius
      width=0
      color {colorUseRule=UseColor_Specified, specifiedColor=<rgba 0 0 0 0>}
      cornerRadii=<0 0 0 0>
     }
     tintColor {colorUseRule=UseColor_Specified, specifiedColor=<rgba 0.0685 0.0685 0.0685 0.6>}
     uVRegion {bIsValid=false, max=<0 0>, min=<0 0>}
    }
   }
   slot {class=GridSlot}
  }
  BlankMessage_Txt {
   widgetClassPath=&CommonTextBlock
   bIsVariable=true
   parent=GridPanel_17
   widget=BlankMessage_Txt
   properties {
    autoWrapText=true
    justification=Center
    text="It’s quiet in here! Chat in any available tab or start a new chat with a friend!"
    colorAndOpacity {colorUseRule=UseColor_Specified, specifiedColor=<rgba 0.5776 0.5776 0.5776 1>}
    font {
     bForceMonospaced=false
     bMaterialIsStencil=false
     fontMaterial=null
     fontObject=&NotoSans
     letterSpacing=0
     monospacedWidth=1
     size=24
     skewAmount=0
     typefaceFontName=NotoSans-Regular
     outlineSettings {
      bApplyOutlineToDropShadows=false
      bMiteredCorners=false
      bSeparateFillAlpha=false
      outlineMaterial=null
      outlineSize=0
      outlineColor=<rgba 0 0 0 1>
     }
    }
    shadowOffset=<0 0>
   }
   slot {class=GridSlot, properties {layer=1, verticalAlignment=VAlign_Center, padding=<ltrb 32 16 32 16>}}
  }
 ]
}
