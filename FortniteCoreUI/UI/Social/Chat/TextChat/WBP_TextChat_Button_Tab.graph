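# Channel tab button for the text chat UI, built on the UIKit generic button and driven by channel type and unread state.
# PreConstruct caps width from a console variable, sets ellipsis overflow and limits the status indicator count; hover and selection switch text to clipping.
# Helper functions map channel data to label, icon, indicator visibility and indicator configuration, with the AddCmd channel shown as an unlabeled add icon.
# The SetText override still contains a debug PrintString.
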WidgetBlueprint /FortniteCoreUI/UI/Social/Chat/TextChat/WBP_TextChat_Button_Tab
graphs [
 CF_UseUnreadMessagesCountToStatusConfiguration {nodes [
   n1 = CF_UseUnreadMessagesCountToStatusConfiguration!() -> (then, UseUnreadMessagesCount:bool=~) {is=title K2Node_FunctionEntry}
   > Return Node!(
    execute
    Configuration:E_UI_StatusIndicator_Configuration Enum=Select(Option 0:E_UI_StatusIndicator_Configuration Enum=NewEnumerator2, Option 1:E_UI_StatusIndicator_Configuration Enum=NewEnumerator3, Index=@n1) -> (ReturnValue:E_UI_StatusIndicator_Configuration Enum)
   )
  ]}
 EventGraph {nodes [
   n1 = Event PreConstruct()
   > Parent: PreConstruct(IsDesignTime=@n1.IsDesignTime)
   > n2 = GetConsoleVariableIntValue!(execute, VariableName:String=SocialUI.TextChat.SetMaxTabWidth) -> (then, ReturnValue:int=0)
   > SetMaxDimensions!(execute, self:Common Button Base Object Reference, InMaxWidth:int=@n2, InMaxHeight:int=0) -> (then)
   > SetTextOverflowPolicy!(execute, self:WBP UIKit Button Generic Base Object Reference, OverflowPolicy:ETextOverflowPolicy Enum=Ellipsis) -> (then)
   > OverrideCountCeiling!(
    execute
    self:WBP UIKit Status Indicator Base Object Reference=Get StatusIndicator() -> (StatusIndicator:WBP UIKit Status Indicator Object Reference)
    Ceiling:int=9
   ) -> (then)
   Event On Hovered()
   > Parent: On Hovered() {is=title K2Node_CallParentFunction}
   > n3 = Branch(Condition=GetSelected!(self:Common Button Base Object Reference) -> (ReturnValue:bool))
   SetTextOverflowPolicy(execute=[@n5.then, @n3.else], OverflowPolicy=Clip)
   Event On Unhovered()
   > Parent: On Unhovered() {is=title K2Node_CallParentFunction}
   > n4 = Branch(Condition=GetSelected())
   SetTextOverflowPolicy(execute=[@n6.then, @n4.else], OverflowPolicy=Ellipsis)
   Event On Selected()
   > n5 = Parent: On Selected() {is=title K2Node_CallParentFunction}
   Event On Deselected()
   > n6 = Parent: On Deselected() {is=title K2Node_CallParentFunction}
  ]}
 SetChannelName {nodes [
   n1 = SetChannelName!() -> (then, ChannelName:String, ChannelType:EFortTextChatChannelType Enum=All) {is=title K2Node_FunctionEntry}
   > Return Node!(
    execute
    ChannelNameText:Text=Select(
     Option 0:Text="To Text (String)"(InString=@n1)
     Option 1:Text
     Index="Equal (Enum)"(A:EFortTextChatChannelType Enum=@n1.ChannelType, B:EFortTextChatChannelType Enum=AddCmd)
    ) -> (ReturnValue:Text)
   )
  ]}
 SetHasNewMessages {nodes [
   n1 = SetHasNewMessages!() -> (then, HasNewMessages:bool=~, UseUnreadMessagesCount:bool=~, NewMessagesCount:int) {is=title K2Node_FunctionEntry}
   > n2 = Branch(Condition=@n1.UseUnreadMessagesCount)
   n3 = Get NewMessagesCount() -> (NewMessagesCount:int=0)
   > Return Node!(
    execute
    visibility:ESlateVisibility Enum=Select(Option 0:ESlateVisibility Enum=Collapsed, Option 1:ESlateVisibility Enum=SelfHitTestInvisible, Index="integer > integer"(A=@n3)) -> (ReturnValue:ESlateVisibility Enum)
   )
   Return Node(execute=@n2.else, visibility=Select(Option 0=Collapsed, Option 1=SelfHitTestInvisible, Index=Get HasNewMessages() -> (HasNewMessages:bool)))
  ]}
 SetIcon {nodes [
   SetIcon!() -> (then, Icon:Object Reference) {is=title K2Node_FunctionEntry}
   n1 = Get FortTextChatChannelVM() -> (FortTextChatChannelVM:Fort Text Chat Channel VM Object Reference)
   > Branch(Condition=IsValid(Object=@n1))
   > Parent: SetIcon!(
    execute
    Icon:Object Reference=Select!(
     All:Texture 2D Object Reference
     Party:Texture 2D Object Reference
     Game:Texture 2D Object Reference
     Whisper:Texture 2D Object Reference
     Persistent:Texture 2D Object Reference
     AddCmd:Texture 2D Object Reference=/FortUILibrary/Textures/Icons/T_UI_Icon_UIKit_Add_SDF
     MAX_NUM:Texture 2D Object Reference
     Index:EFortTextChatChannelType Enum=Get ChannelType!(self:Fort Text Chat Channel VM Object Reference=@n1) -> (ChannelType:EFortTextChatChannelType Enum=All)
    ) -> (ReturnValue:Texture 2D Object Reference)
   ) -> (then) {is=title K2Node_CallParentFunction}
   Parent: SetIcon(Icon=/FortUILibrary/Textures/Icons/T_UI_Icon_UIKit_TextChatMuted_DF)
  ]}
 SetIconForChannelType {nodes [
   n1 = SetIconForChannelType!() -> (then, ChannelType:EFortTextChatChannelType Enum=All) {is=title K2Node_FunctionEntry}
   > SetIcon!(
    execute
    self:WBP UIKit Button Generic Base Object Reference
    Icon:Object Reference=Select(
     All:Object Reference
     Party:Object Reference=/FortUILibrary/Textures/Icons/T_UI_Icon_UIKit_PartyUp_SDF
     Game:Object Reference=/FortUILibrary/Textures/Icons/T_UI_Icon_UIKit_ControllerMapping_SDF
     Whisper:Object Reference=/FortUILibrary/Textures/Icons/T_UI_Icon_UIKit_ChannelWhisper_SDF
     Persistent:Object Reference
     AddCmd:Object Reference=&T_UI_Icon_UIKit_Add_SDF
     MAX_NUM:Object Reference=/FortUILibrary/Textures/Icons/T_UI_Icon_UIKit_Mic_SDF
     Index=@n1
    ) -> (ReturnValue:Object Reference)
   ) -> (then) {is=title K2Node_CallFunction}
  ]}
 SetText {nodes [
   n1 = SetText!() -> (then, Text:Text) {is=title K2Node_FunctionEntry}
   n2 = Get FortTextChatChannelVM()
   > Branch(Condition=IsValid(Object=@n2))
   > n3 = Branch(Condition="Equal (Enum)"(A=Get ChannelType(self=@n2) -> (ChannelType=All), B=AddCmd))
   > Parent: SetText!(execute, Text:Text) -> (then) {is=title K2Node_CallParentFunction}
   Parent: SetText(execute=@n3.else, Text=@n1)
   > PrintString!(execute, InString=Append(A=SET TEXT, B="To String (Text)"(InText=@n1)), bPrintToScreen, bPrintToLog, TextColor, Duration, Key) -> (then)
  ]}
]
variables []
widgets {
 info {
  inheritedWidgetCount=0
  parentClass=/FortUILibrary/Systems/ModularButton/WBP_UIKit_ButtonGeneric_Base.WBP_UIKit_ButtonGeneric_Base_C
  rootWidgetClass=null
  widgetCount=0
 }
 widgets []
}
