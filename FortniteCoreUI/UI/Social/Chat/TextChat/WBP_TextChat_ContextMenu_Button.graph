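# Text chat context menu button widget, a child of the UIKit context menu anchor button that adds no widgets of its own.
# On construct it runs the parent logic, then, if compact UI is in use, sets a smaller or fixed minimum size with equal width and height.
# When added to the focus path, it calls the parent handler and broadcasts an OnFocused event dispatcher.
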WidgetBlueprint /FortniteCoreUI/UI/Social/Chat/TextChat/WBP_TextChat_ContextMenu_Button
graphs [
 EventGraph {nodes [
   Event Construct()
   > Parent: Construct()
   > SetCompactUIStyle() {is=title K2Node_CallFunction}
   n1 = Event OnAddedToFocusPath!() -> (OutputDelegate, then, InFocusEvent:Focus Event Structure) {is=title K2Node_Event}
   > Parent: OnAddedToFocusPath!(execute, InFocusEvent:Focus Event Structure=@n1.InFocusEvent) -> (then) {is=title K2Node_CallParentFunction}
   > Call On Focused!(execute:Exec, self:WBP Text Chat Context Menu Button Object Reference) -> (then:Exec) {is=title K2Node_CallDelegate}
  ]}
 OnFocused {nodes [OnFocused() {is=title K2Node_FunctionEntry}]}
 SetCompactUIStyle {nodes [
   SetCompactUIStyle() {is=title K2Node_FunctionEntry}
   > Branch(Condition=IsUsingCompactUI())
   > n1 = Set Min Width(Min Width:int=80) -> (Output_Get:int=0)
   > Set Min Height(Min Height:int=@n1) -> (Output_Get:int=0)
  ]}
]
variables [OnFocused {}]
widgets {
 info {
  inheritedWidgetCount=0
  parentClass=/FortUILibrary/UIKit/ContextMenu/WBP_UIKit_ContextMenuAnchorButton.WBP_UIKit_ContextMenuAnchorButton_C
  rootWidgetClass=null
  widgetCount=0
 }
 widgets []
}
