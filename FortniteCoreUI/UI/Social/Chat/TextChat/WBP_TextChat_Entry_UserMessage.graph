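# Text chat list entry widget for one user message, showing avatar, "name: message" text, timestamp, send or text-to-speech error, and context menu.
# It caches its FortTextChatMessageVM, with helper functions formatting restricted text, styling muted, blocked or redacted senders, and showing TTS states.
# Hover swaps timestamp for menu icon; on touch, a first tap reveals it and the next broadcasts OnMessageClicked.
# Compact UI and a console variable adjust layout and font sizes.
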WidgetBlueprint /FortniteCoreUI/UI/Social/Chat/TextChat/WBP_TextChat_Entry_UserMessage
graphs [
 CF_ErrorText {nodes [
   n1 = CF_ErrorText!() -> (then, Has Sent Error:bool=~, TTS State:EFortTextToSpeechState Enum=None) {is=title K2Node_FunctionEntry}
   n2 = NOT Boolean(A=Get Has Sent Error() -> (Has Sent Error:bool))
   n3 = "Equal (Enum)"(A:EFortTextToSpeechState Enum=@n1.TTS State, B:EFortTextToSpeechState Enum=Failed)
   > Return Node!(
    execute
    Text:Text=SelectText(
     A="LOCTEXT(\"Couldn’t read aloud. Select to try again.\")"
     B="LOCTEXT(\"Failed to send. Try again soon.\")"
     bPickA=AND Boolean(A=@n2, B=@n3)
    )
   )
  ]}
 CF_ErrorVisibility {nodes [
   n1 = CF_ErrorVisibility!() -> (then, Has Sent Error:bool=~, TTS State:EFortTextToSpeechState Enum=None) {is=title K2Node_FunctionEntry}
   n2 = "Equal (Enum)"(A=@n1.TTS State, B=Failed)
   n3 = Get Has Sent Error()
   > Return Node!(execute, Visibility:ESlateVisibility Enum=Select(Option 0=Collapsed, Option 1=SelfHitTestInvisible, Index=OR Boolean(A=@n3, B=@n2)))
  ]}
 CF_IconVisibility {nodes [
   n1 = CF_IconVisibility!() -> (then, isSending:bool=~, hasError:bool=~) {is=title K2Node_FunctionEntry}
   > n2 = Branch(Condition=OR Boolean(A=@n1, B=@n1.hasError))
   > n3 = Set isVisible(isVisible:ESlateVisibility Enum=SelfHitTestInvisible) -> (Output_Get:ESlateVisibility Enum=Visible)
   > Return Node(Visibility=@n3)
   n4 = Set isVisible(execute=@n2.else, isVisible=Hidden) -> (Output_Get=Visible)
   > Return Node(Visibility=@n4)
  ]}
 ConvertMessageDataNotRich {nodes [
   n1 = ConvertMessageDataNotRich!() -> (then, NewNickname:String, NewText:Text, MessageRestrictions:ETextChatMessageRestriction Enum=Unrestricted, IsSenderBlocked:bool=~) {is=title K2Node_FunctionEntry}
   n2 = Get MessageRestrictions() -> (MessageRestrictions:ETextChatMessageRestriction Enum=Unrestricted)
   n3 = "To Text (String)"(InString=@n1)
   > Return Node!(
    execute
    FinalText:Text=SelectText(
     A=Format Text!(Format="LOCTEXT(\"{Username}: {Message}\")", Username:Text=@n3, Message:Text=@n1.NewText) -> (Result)
     B="LOCTEXT(\"Hidden player: Message hidden\")"
     bPickA="Equal (Enum)"(A:ETextChatMessageRestriction Enum=@n2, B:ETextChatMessageRestriction Enum=Unrestricted)
    )
   )
  ]}
 EventGraph {nodes [
   n1 = "Init Finished (Message_ContextMenu)"!() -> (OutputDelegate, then, Object:Object Reference) {is=title K2Node_ComponentBoundEvent}
   > n2 = Cast To WBP_UIKit_ContextMenu!(execute, Object=@n1.Object) -> (then, CastFailed, AsWBP UIKit Context Menu:WBP UIKit Context Menu Object Reference) {is=title K2Node_DynamicCast}
   > Bind Event to On Entry Generated(
    self:List View Base Object Reference=Get Content!(self:WBP UIKit Context Menu Object Reference=@n2) -> (Content:Common List View Object Reference)
    Delegate=Create Event()
   ) {is=title K2Node_AddDelegate}
   n3 = OnContextMenuMessageGenerated!() -> (OutputDelegate, then, Widget:User Widget Object Reference) {is=title K2Node_CustomEvent}
   > n4 = Cast To WBP_UIKit_ContextMenuListEntry!(execute, Object=@n3.Widget) -> (then, CastFailed, AsWBP UIKit Context Menu List Entry:WBP UIKit Context Menu List Entry Object Reference) {is=title K2Node_DynamicCast}
   > Branch(Condition=IsValid(Object=Get Icon!(self:WBP UIKit Button Generic Base Object Reference=@n4) -> (Icon:Texture 2D Object Reference)))
   > SetIconMaterial!(
    execute
    self:WBP UIKit Button Generic Base Object Reference=@n4
    Material Instance:Material Instance Object Reference=/FortniteCoreUI/UI/Social/Chat/TextChat/MaterialInstances/MI_UI_TextChat_TTS
   ) -> (then)
   > SetIconPadding!(
    execute
    self:WBP UIKit Button Generic Base Object Reference=@n4
    Padding_Left:float=0
    Padding_Top:float=0
    Padding_Right:float=-8
    Padding_Bottom:float=0
   ) -> (then)
   n5 = Event PreConstruct()
   > Parent: PreConstruct(IsDesignTime=@n5.IsDesignTime)
   > n6 = Sequence()
   > Branch(Condition=@n5.IsDesignTime)
   n7 = Get TextMessage_NotRich_WrapAtSetRuntime() -> (TextMessage_NotRich_WrapAtSetRuntime:Common Text Object Reference)
   > SetWrapTextWidth!(execute, self:Common Text Object Reference=@n7, InWrapTextAt:int=600) -> (then)
   > Set AutoWrapText!(execute, AutoWrapText:bool=false, self:Text Layout Widget Object Reference=@n7) -> (then, Output_Get:bool)
   Set Message Style!(execute=@n6.then_1, self, IsSubMessage:bool=Get Is Sub Message() -> (Is Sub Message:bool)) -> (then) {is=title K2Node_CallFunction}
   > SetContextMenuIconVisibility!(execute, self, IsContextMenuVisible:bool=false) -> (then)
   > SetCompactUIStyle() {is=title K2Node_CallFunction}
   n8 = Event OnListItemObjectSet()
   > n9 = Cast To FortTextChatMessageVM!(execute, Object=@n8.ListItemObject) -> (then, CastFailed, AsText Chat Message VM:Text Chat Message VM Object Reference) {is=title K2Node_DynamicCast}
   > Set FortTextChatMessageVM(FortTextChatMessageVM:Text Chat Message VM Object Reference=@n9) -> (Output_Get:Text Chat Message VM Object Reference)
   n10 = Event On Item Selection Changed!() -> (OutputDelegate, then, bIsSelected:bool) {is=title K2Node_Event}
   > n11 = Branch(Condition=@n10.bIsSelected)
   Branch(execute=@n11.else, Condition=IsUsingTouch(PlayerController=GetOwningPlayer()))
   > Set bIsContextIconVisible(bIsContextIconVisible:bool=false) -> (Output_Get:bool)
   > SetContextMenuIconVisibility(IsContextMenuVisible=false)
   Event On Hovered()
   > n12 = Branch(Condition=IsUsingTouch(PlayerController=GetOwningPlayer()))
   Parent: On Hovered(execute=@n12.else) {is=title K2Node_CallParentFunction}
   n13 = Get IsPreviewMode() -> (IsPreviewMode:bool)
   > n14 = Branch(Condition=@n13)
   n15 = Get TTS_Icon() -> (TTS_Icon:Image Object Reference)
   SetContextMenuIconVisibility(execute=@n14.else, IsContextMenuVisible=NOT Boolean(A=IsVisible(self=@n15)))
   Event On Unhovered()
   > Parent: On Unhovered() {is=title K2Node_CallParentFunction}
   > n16 = Branch(Condition=@n13)
   SetContextMenuIconVisibility(execute=@n16.else, IsContextMenuVisible=false)
   Event On Clicked()
   > n17 = Branch(Condition=IsUsingTouch(PlayerController=GetOwningPlayer()))
   > n18 = Branch(Condition=Get bIsContextIconVisible() -> (bIsContextIconVisible:bool))
   n19 = Get FortTextChatMessageVM() -> (FortTextChatMessageVM:Text Chat Message VM Object Reference)
   Call On Message Clicked!(
    execute:Exec=[@n20.then, @n18.then]
    self:WBP Text Chat Entry User Message Object Reference
    MessageID:Name=Get MessageId!(self:Text Chat Message VM Object Reference=@n19) -> (MessageId:Name=None)
    Username:String=Get DisplayName!(self:Text Chat Message VM Object Reference=@n19) -> (DisplayName:String)
    Button:Common Button Base Object Reference=Self-Reference()
   ) -> (then:Exec) {is=title K2Node_CallDelegate}
   Set bIsContextIconVisible(execute=@n18.else, bIsContextIconVisible=true)
   > SetContextMenuIconVisibility(IsContextMenuVisible=true)
   n20 = Set bIsContextIconVisible(execute=@n17.else, bIsContextIconVisible=true)
  ]}
 GetPlayerStatus {nodes [
   n1 = GetPlayerStatus!() -> (then, IsPlayerBlocked:bool=~, IsPlayerMuted:bool=~) {is=title K2Node_FunctionEntry}
   n2 = Get MessageType!(self:Text Chat Message VM Object Reference=Get FortTextChatMessageVM()) -> (MessageType:EFortTextChatMessageType Enum=Sent)
   > n3 = Branch(Condition=OR Boolean(A=@n1, B="Equal (Enum)"(A:EFortTextChatMessageType Enum=@n2, B:EFortTextChatMessageType Enum=RedactedMessage)))
   n4 = Get WBP_PlayerAvatar() -> (WBP_PlayerAvatar:WBP Player Avatar Object Reference)
   > SetEmpty!(execute, self:WBP Player Avatar Object Reference=@n4, bEmpty:bool=true) -> (then)
   > SetEmptyBackgroundColor!(
    execute
    self:WBP Player Avatar Object Reference=@n4
    EmptyBackgroundColor:Color Structure=Get Empty Background Color() -> (Empty Background Color:Color Structure)
   ) -> (then)
   > SetEmptyHighlightColor!(
    execute
    self:WBP Player Avatar Object Reference=@n4
    EmptyHighlightColor:Color Structure=Get Empty Highlight Color() -> (Empty Highlight Color:Color Structure)
   ) -> (then)
   > SetIncognitoColor!(
    execute
    self:WBP Player Avatar Object Reference=@n4
    IncognitoColor:Color Structure=Get Incognito Color() -> (Incognito Color:Color Structure)
   ) -> (then)
   > n5 = SetInactiveIncognitoColor!(execute, self:WBP Player Avatar Object Reference=@n4, InactiveIncognitoColor:Color Structure=Get Incognito Color()) -> (then)
   n6 = Get TextMessage_NotRich_WrapAtSetRuntime()
   SetColorAndOpacity(
    execute=[@n9.then, @n5.then]
    self=@n6
    InColorAndOpacity=Get FontColorMutedOrBlockedMessage() -> (FontColorMutedOrBlockedMessage:Slate Color Structure)
   )
   > SetFont(self=@n6, InFontInfo=Get FontStyleMutedOrBlockedMessage() -> (FontStyleMutedOrBlockedMessage:Slate Font Info Structure))
   > Return Node!(execute, IsDisabled:bool=true)
   n7 = Branch(execute=@n3.else, Condition=@n1.IsPlayerMuted)
   n8 = Get WBP_PlayerAvatar()
   > SetIconTexture!(
    execute
    self:WBP Player Avatar Object Reference=@n8
    IconTexture:Texture 2D Object Reference=/FortUILibrary/Textures/Icons/T_UI_Icon_UIKit_TextChatMuted_DF
    UseIcon:bool=true
   ) -> (then)
   > SetIconColor!(execute, self:WBP Player Avatar Object Reference=@n8, IconColor:Color Structure=Get Icon Color() -> (Icon Color:Color Structure)) -> (then)
   > SetBackgroundColor!(
    execute
    self:WBP Player Avatar Object Reference=@n8
    BackgroundColor:Color Structure=Get Background Color() -> (Background Color:Color Structure)
   ) -> (then)
   > n9 = SetHighlightColor!(
    execute
    self:WBP Player Avatar Object Reference=@n8
    HighlightColor:Color Structure=Get Highlight Color() -> (Highlight Color:Color Structure)
   ) -> (then)
   n10 = Get WBP_PlayerAvatar()
   SetIconTexture(execute=@n7.else, self=@n10, IconTexture=&T_UI_Icon_UIKit_TextChatMuted_DF, UseIcon=false)
   > SetEmpty(self=@n10, bEmpty=false)
   n11 = Get TextMessage_NotRich_WrapAtSetRuntime()
   > SetFont(self=@n11, InFontInfo=Get FontStyleNormalMessage() -> (FontStyleNormalMessage:Slate Font Info Structure))
   > SetColorAndOpacity(self=@n11, InColorAndOpacity=Get FontColorNormalMessage() -> (FontColorNormalMessage:Slate Color Structure))
   > Return Node(IsDisabled=false)
  ]}
 OnListEntryGenerated {nodes [
   OnListEntryGenerated() {is=title K2Node_FunctionEntry}
   > SetCompactUISize()
  ]}
 OnMessageClicked {nodes [OnMessageClicked!() -> (then, MessageID:Name, Username:String, Button:Common Button Base Object Reference) {is=title K2Node_FunctionEntry}]}
 Set Message Style {nodes [
   n1 = Set Message Style!() -> (then, IsSubMessage:bool=~) {is=title K2Node_FunctionEntry}
   > Set Is Sub Message(Is Sub Message:bool=@n1) -> (Output_Get:bool)
   > Set Visibility(Visibility=Select(Option 0=HitTestInvisible, Option 1=Hidden, Index=Get Is Sub Message()), self=Get WBP_PlayerAvatar())
  ]}
 Set Username Avatar {nodes [
   n1 = Set Username Avatar!() -> (then, FortPlayerAvatarViewModel:Fort Player Avatar Viewmodel Object Reference) {is=title K2Node_FunctionEntry}
   n2 = Get FortTextChatMessageVM()
   > n3 = Branch(Condition=GetPlayerStatus!(
     self
     IsPlayerBlocked:bool=Get bIsSenderBlocked!(self:Text Chat Message VM Object Reference=@n2) -> (bIsSenderBlocked:bool)
     IsPlayerMuted:bool=Get bIsSenderMuted!(self:Text Chat Message VM Object Reference=@n2) -> (bIsSenderMuted:bool)
    ) -> (IsDisabled:bool) {is=title K2Node_CallFunction})
   n4 = Get WBP_PlayerAvatar()
   Set FortPlayerAvatarViewModel!(execute=@n3.else, FortPlayerAvatarViewModel:Fort Player Avatar Viewmodel Object Reference=@n1, self:WBP Player Avatar Object Reference=@n4) -> (then, Output_Get:Fort Player Avatar Viewmodel Object Reference)
   > SetOnlineStatusVisibility!(execute, self:WBP Player Avatar Object Reference=@n4, bHideStatus:bool=false) -> (then)
  ]}
 SetAvatarStatus {nodes [
   n1 = SetAvatarStatus!() -> (then, FortOnlineStatusViewModel:Fort Online Status Viewmodel Object Reference) {is=title K2Node_FunctionEntry}
   > SetOnlineStatus!(
    execute
    self:WBP Player Avatar Object Reference=Get WBP_PlayerAvatar()
    OnlineStatus:EPresenceIndicatorState Enum=Get OnlineStatus!(self:Fort Online Status Viewmodel Object Reference=@n1) -> (OnlineStatus:EPresenceIndicatorState Enum=Online)
   ) -> (then)
  ]}
 SetBackgroundVisibility {nodes [
   n1 = SetBackgroundVisibility!() -> (then, ShouldBeVisible:bool=~) {is=title K2Node_FunctionEntry}
   > SetVisibility(
    self=Get Image_Background() -> (Image_Background:Image Object Reference)
    InVisibility=Select(Option 0=Collapsed, Option 1=SelfHitTestInvisible, Index=@n1)
   )
  ]}
 SetCompactUISize {nodes [
   SetCompactUISize() {is=title K2Node_FunctionEntry}
   > n1 = GetConsoleVariableBoolValue!(execute, VariableName:String=NewTextChat.UseLargeChatTextSize) -> (then, ReturnValue:bool)
   > Branch(Condition=@n1)
   n2 = Get TextMessage_NotRich_WrapAtSetRuntime()
   > n3 = Set members in Slate Font Info!(
    execute:Exec
    StructRef:"Slate Font Info Structure (by ref)"=Get Font!(self:Text Object Reference=@n2) -> (Font:Slate Font Info Structure)
    Size:float=20.25
   ) -> (then:Exec, StructOut:"Slate Font Info Structure (by ref)") {is=title K2Node_SetFieldsInStruct}
   > SetFont(self=@n2, InFontInfo=@n3)
   n4 = Get ErrorMessage() -> (ErrorMessage:Common Text Object Reference)
   > n5 = Set members in Slate Font Info(StructRef=Get Font(self=@n4), Size=18)
   > SetFont(self=@n4, InFontInfo=@n5)
  ]}
 SetCompactUIStyle {nodes [
   SetCompactUIStyle() {is=title K2Node_FunctionEntry}
   > Branch(Condition=IsUsingCompactUI())
   n1 = Get WBP_PlayerAvatar()
   > Event Set Size!(execute, self:WBP Player Avatar Object Reference=@n1, NewSize:Vector 2D Structure=(X=40,Y=40)) -> (then)
   > SetPadding!(
    execute
    self:Horizontal Box Slot Object Reference=SlotAsHorizontalBoxSlot!(Widget:Widget Object Reference=@n1) -> (ReturnValue:Horizontal Box Slot Object Reference)
    InPadding_Left:float=4
    InPadding_Top:float=12
    InPadding_Right:float=12
    InPadding_Bottom:float=12
   ) -> (then)
   n2 = Get TextMessage_NotRich_WrapAtSetRuntime()
   > Set LineHeightPercentage!(execute, LineHeightPercentage:float=1, self:Text Layout Widget Object Reference=@n2) -> (then, Output_Get:float)
   > Set Grid Slot Padding!(execute:Exec, Widget:Widget Object Reference=@n2, Padding_Left:float=0, Padding_Top:float=12, Padding_Right:float=0, Padding_Bottom:float=12) -> (then:Exec) {is=title K2Node_MacroInstance}
   n3 = Get ContextMenuIcon() -> (ContextMenuIcon:Image Object Reference)
   > Set Grid Slot Padding(Widget=@n3, Padding_Left=0, Padding_Top=12, Padding_Right=4, Padding_Bottom=0)
   > SetDesiredSizeOverride!(execute, self:Image Object Reference=@n3, DesiredSize:Vector 2D Structure=(X=40,Y=40)) -> (then)
   n4 = SlotAsGridSlot(Widget=Get Text_TimeStamp() -> (Text_TimeStamp:Common Text Object Reference))
   n5 = Get!(self:Grid Slot Object Reference=@n4) -> (Padding_Left:float=~, Padding_Top:float=~, Padding_Right:float=~, Padding_Bottom:float=~)
   > SetPadding(
    self:Grid Slot Object Reference=@n4
    InPadding_Left=@n5
    InPadding_Top=16
    InPadding_Right=@n5.Padding_Right
    InPadding_Bottom=@n5.Padding_Bottom
   )
   > Set TTSIconSize(TTSIconSize:Vector 2D Structure=(X=48,Y=48)) -> (Output_Get:Vector 2D Structure)
   > Set ThrobberIconSize(ThrobberIconSize:Vector 2D Structure=(X=48,Y=48)) -> (Output_Get:Vector 2D Structure)
  ]}
 SetContextMenuIconVisibility {nodes [
   n1 = SetContextMenuIconVisibility!() -> (then, IsContextMenuVisible:bool=~) {is=title K2Node_FunctionEntry}
   > Set RenderOpacity!(execute, RenderOpacity:float=SelectFloat(A=1, B=0, bPickA=@n1), self:Widget Object Reference=Get ContextMenuIcon()) -> (then, Output_Get:float)
   > Set RenderOpacity(RenderOpacity=SelectFloat(A=0, B=1, bPickA=Get IsContextMenuVisible() -> (IsContextMenuVisible:bool)), self=Get Text_TimeStamp())
  ]}
 SetSpeakingStateVisuals {nodes [
   n1 = SetSpeakingStateVisuals!() -> (then, Message Speaking State:EFortTextToSpeechState Enum=None) {is=title K2Node_FunctionEntry}
   > n2 = Switch on EFortTextToSpeechState!(execute, Selection:EFortTextToSpeechState Enum=@n1) -> (None:Exec, Preparing:Exec, Playing:Exec, Failed:Exec) {is=title K2Node_SwitchEnum}
   SetVisibility(execute=@n2.Failed, self=Get Image_Background(), InVisibility=Collapsed)
   SetVisibility(execute=[@n5.then, @n2.None], self=Get Image_Background(), InVisibility=Collapsed)
   n3 = Get TTS_Icon()
   SetBrushFromMaterial(execute=@n2.Playing, self=@n3, Material=/FortUILibrary/Textures/MaterialInstances/MI_UIKit_Icon_VoiceChatActive)
   n4 = Get TTS_Icon()
   > SetDesiredSizeOverride(self=@n4, DesiredSize=Get TTSIconSize() -> (TTSIconSize:Vector 2D Structure))
   > SetVisibility(self=Get Image_Background(), InVisibility=Visible)
   SetBrushFromMaterial(execute=@n2.Preparing, self=@n3, Material=/FortUILibrary/UIKit/Throbber/MI_UIKit_Monochrome_Throbber)
   > n5 = SetDesiredSizeOverride(self=@n4, DesiredSize=Get ThrobberIconSize() -> (ThrobberIconSize:Vector 2D Structure))
  ]}
 UpdatePlayerStatus {nodes [
   n1 = UpdatePlayerStatus!() -> (then, IsDisabled:bool=~) {is=title K2Node_FunctionEntry}
   > Branch(Condition=@n1)
   n2 = Get WBP_PlayerAvatar()
   > n3 = GetDynamicMaterial(self=Get AvatarImage!(self:WBP Player Avatar Object Reference=@n2) -> (AvatarImage:Image Object Reference))
   > SetVectorParameterValue(self=@n3, ParameterName=BackgroundRGB, Value=(R=0.375,G=0.375,B=0.375,A=1))
   > SetVectorParameterValue(self=@n3, ParameterName=HighlightRGB, Value=(R=0.1354,G=0.1354,B=0.1354,A=1))
   > SetOnlineStatusVisibility(self=@n2, bHideStatus=true)
  ]}
]
variables [
 Background Color {type=Color, default=<rgba 0.4431 0.4431 0.4431 1>}
 bIsContextIconVisible {type=boolean}
 Empty Background Color {type=Color, default=<rgba 0.5725 0.5725 0.5725 1>}
 Empty Highlight Color {type=Color, default=<rgba 0.8275 0.8275 0.8275 1>}
 FontColorMutedOrBlockedMessage {type=SlateColor, default {colorUseRule=UseColor_Specified, specifiedColor=<rgba 0.4125 0.4233 0.4287 1>}}
 FontColorNormalMessage {type=SlateColor, default {colorUseRule=UseColor_Specified, specifiedColor=<rgba 1 1 1 1>}}
 FontStyleMutedOrBlockedMessage {
  type=SlateFontInfo
  default {
   bForceMonospaced=false
   bMaterialIsStencil=false
   fontMaterial=null
   fontObject=&HeadingNow
   letterSpacing=0
   monospacedWidth=1
   size=13.5
   skewAmount=0
   typefaceFontName=74 Regular
   outlineSettings {
    bApplyOutlineToDropShadows=false
    bMiteredCorners=false
    bSeparateFillAlpha=false
    outlineMaterial=null
    outlineSize=2
    outlineColor=<rgba 0 0 0 0.15>
   }
  }
 }
 FontStyleNormalMessage {
  type=SlateFontInfo
  default {
   bForceMonospaced=false
   bMaterialIsStencil=false
   fontMaterial=null
   fontObject=&HeadingNow
   letterSpacing=0
   monospacedWidth=1
   size=13.5
   skewAmount=0
   typefaceFontName=74 Regular
   outlineSettings {
    bApplyOutlineToDropShadows=false
    bMiteredCorners=false
    bSeparateFillAlpha=false
    outlineMaterial=null
    outlineSize=2
    outlineColor=<rgba 0 0 0 0.35>
   }
  }
 }
 Highlight Color {type=Color, default=<rgba 0.8667 0.8667 0.8667 1>}
 Icon Color {type=Color, default=<rgba 0.9059 0.9059 0.9059 1>}
 Incognito Color {type=Color, default=<rgba 0.3608 0.3608 0.3608 1>}
 Is Sub Message {type=boolean}
 OnMessageClicked {}
 Received From PlayerColor {type=LinearColor, default=<rgba 0.5776 0.5776 0.5776 1>}
 SentColor {type=LinearColor, default=<rgba 0.2705 0.491 0.4125 1>}
 ThrobberIconSize {type=Vector2D, default=<20 20>}
 TTSIconSize {type=Vector2D, default=<28 28>}
]
widgets {
 info {
  inheritedWidgetCount=0
  namedSlotCount=1
  parentClass=&/WBP_TextChat_Entry_MessageBase.WBP_TextChat_Entry_MessageBase_C
  rootWidgetClass=null
  widgetCount=12
 }
 widgets [
  GP_UserMessage_Container {widgetClassPath=&GridPanel, widget=GP_UserMessage_Container, properties {columnFill [0,1,0,0], rowFill [1]}}
  HorizontalBox_0 {
   widgetClassPath=&HorizontalBox
   parent=GP_UserMessage_Container
   widget=HorizontalBox_0
   slot {class=GridSlot, properties {columnSpan=1, horizontalAlignment=HAlign_Left, layer=99, verticalAlignment=VAlign_Top, padding=<ltrb 0 0 4 0>}}
  }
  WBP_PlayerAvatar {
   widgetClassPath=/Game/UI/Foundation/Profile/WBP_PlayerAvatar
   bIsVariable=true
   parent=HorizontalBox_0
   widget=WBP_PlayerAvatar
   properties {avatarSize=<32 32>}
   slot {class=HorizontalBoxSlot}
  }
  TTS_Icon {
   widgetClassPath=&Image
   bIsVariable=true
   parent=HorizontalBox_0
   widget=TTS_Icon
   properties {
    visibility=Collapsed
    brush {
     bIsDynamicallyLoaded=false
     drawAs=Image
     imageType=FullColor
     mirroring=NoMirror
     resourceName=None
     resourceObject=&MI_UIKit_Icon_VoiceChatActive
     tiling=NoTile
     imageSize=<32 32>
     margin=<ltrb 0 0 0 0>
     outlineSettings {
      bUseBrushTransparency=false
      roundingType=HalfHeightRadius
      width=0
      color {colorUseRule=UseColor_Specified, specifiedColor=<rgba 0 0 0 0>}
      cornerRadii=<0 0 0 0>
     }
     tintColor {colorUseRule=UseColor_Specified, specifiedColor=<rgba 1 1 1 1>}
     uVRegion {bIsValid=false, max=<0 0>, min=<0 0>}
    }
   }
   slot {class=HorizontalBoxSlot, properties {verticalAlignment=VAlign_Center, padding=<ltrb 3 0 -2 0>}}
  }
  TextMessage_NotRich_WrapAtSetRuntime {
   widgetClassPath=&CommonTextBlock
   bIsVariable=true
   parent=GP_UserMessage_Container
   widget=TextMessage_NotRich_WrapAtSetRuntime
   properties {
    applyLineHeightToBottomLine=false
    autoWrapText=true
    bIsScrollingEnabled=false
    lineHeightPercentage=1.1
    text=UserName : Message to Secondary UserName
    wrappingPolicy=AllowPerCharacterWrapping
    font {
     bForceMonospaced=false
     bMaterialIsStencil=false
     fontMaterial=null
     fontObject=&HeadingNow
     letterSpacing=0
     monospacedWidth=1
     size=13.5
     skewAmount=0
     typefaceFontName=74 Regular
     outlineSettings {
      bApplyOutlineToDropShadows=false
      bMiteredCorners=false
      bSeparateFillAlpha=false
      outlineMaterial=null
      outlineSize=2
      outlineColor=<rgba 0 0 0 0.35>
     }
    }
    shadowColorAndOpacity=<rgba 0 0 0 1>
    shadowOffset=<0 0>
   }
   slot {class=GridSlot, properties {column=1, columnSpan=1, padding=<ltrb 0 4 0 4>}}
  }
  ErrorMessage {
   widgetClassPath=&CommonTextBlock
   bIsVariable=true
   parent=GP_UserMessage_Container
   widget=ErrorMessage
   properties {
    applyLineHeightToBottomLine=false
    lineHeightPercentage=1.3
    text=Failed to send. Try again soon.
    visibility=Collapsed
    wrappingPolicy=AllowPerCharacterWrapping
    colorAndOpacity {colorUseRule=UseColor_Specified, specifiedColor=<rgba 0.8714 0.1046 0.107 1>}
    font {
     bForceMonospaced=false
     bMaterialIsStencil=false
     fontMaterial=null
     fontObject=&HeadingNow
     letterSpacing=0
     monospacedWidth=1
     size=13.5
     skewAmount=0
     typefaceFontName=74 Regular
     outlineSettings {
      bApplyOutlineToDropShadows=false
      bMiteredCorners=false
      bSeparateFillAlpha=false
      outlineMaterial=null
      outlineSize=2
      outlineColor=<rgba 0 0 0 0.35>
     }
    }
    shadowColorAndOpacity=<rgba 0 0 0 1>
    shadowOffset=<0 0>
   }
   slot {class=GridSlot, properties {column=1, columnSpan=1, row=1, padding=<ltrb 0 4 0 4>}}
  }
  Errror_Loading_Icon_SetbyVM {
   widgetClassPath=&Image
   bIsVariable=true
   parent=GP_UserMessage_Container
   widget=Errror_Loading_Icon_SetbyVM
   slot {class=GridSlot, properties {column=2, columnSpan=1, horizontalAlignment=HAlign_Right, verticalAlignment=VAlign_Center}}
  }
  Text_TimeStamp {
   widgetClassPath=&CommonTextBlock
   bIsVariable=true
   parent=GP_UserMessage_Container
   widget=Text_TimeStamp
   properties {
    applyLineHeightToBottomLine=false
    lineHeightPercentage=1.3
    mobileFontSizeMultiplier=1.4
    text=8:24 AM
    wrapTextAt=300
    colorAndOpacity {colorUseRule=UseColor_Specified, specifiedColor=<rgba 1 1 1 0.8>}
    font {
     bForceMonospaced=false
     bMaterialIsStencil=false
     fontMaterial=null
     fontObject=&HeadingNow
     letterSpacing=0
     monospacedWidth=1
     size=11.25
     skewAmount=0
     typefaceFontName=74 Regular
     outlineSettings {
      bApplyOutlineToDropShadows=false
      bMiteredCorners=false
      bSeparateFillAlpha=false
      outlineMaterial=null
      outlineSize=2
      outlineColor=<rgba 0 0 0 0.35>
     }
    }
    shadowColorAndOpacity=<rgba 0 0 0 1>
    shadowOffset=<0 0>
   }
   slot {class=GridSlot, properties {column=2, columnSpan=1, horizontalAlignment=HAlign_Right, verticalAlignment=VAlign_Top, padding=<ltrb 4 5 0 4>}}
  }
  ContextMenuIcon {
   widgetClassPath=&Image
   bIsVariable=true
   parent=GP_UserMessage_Container
   widget=ContextMenuIcon
   properties {
    visibility=HitTestInvisible
    brush {
     bIsDynamicallyLoaded=false
     drawAs=Image
     imageType=FullColor
     mirroring=NoMirror
     resourceName=None
     resourceObject=/FortniteCoreUI/UI/Social/Chat/TextChat/MaterialInstances/MI_UI_TextChat_MessageContextMenu
     tiling=NoTile
     imageSize=<24 24>
     margin=<ltrb 0 0 0 0>
     outlineSettings {
      bUseBrushTransparency=false
      roundingType=HalfHeightRadius
      width=0
      color {colorUseRule=UseColor_Specified, specifiedColor=<rgba 0 0 0 0>}
      cornerRadii=<0 0 0 0>
     }
     tintColor {colorUseRule=UseColor_Specified, specifiedColor=<rgba 1 1 1 1>}
     uVRegion {bIsValid=false, max=<0 0>, min=<0 0>}
    }
   }
   slot {class=GridSlot, properties {column=2, columnSpan=1, horizontalAlignment=HAlign_Right, verticalAlignment=VAlign_Top, padding=<ltrb 0 4 0 0>}}
  }
  Image_Background {
   widgetClassPath=&Image
   bIsVariable=true
   parent=GP_UserMessage_Container
   widget=Image_Background
   properties {
    visibility=Collapsed
    brush {
     bIsDynamicallyLoaded=false
     drawAs=RoundedBox
     imageType=NoImage
     mirroring=NoMirror
     resourceName=None
     resourceObject=null
     tiling=NoTile
     imageSize=<1 1>
     margin=<ltrb 0 0 0 0>
     outlineSettings {
      bUseBrushTransparency=false
      roundingType=FixedRadius
      width=0
      color {colorUseRule=UseColor_Specified, specifiedColor=<rgba 0 0 0 0>}
      cornerRadii=<8 8 8 8>
     }
     tintColor {colorUseRule=UseColor_Specified, specifiedColor=<rgba 1 1 1 0.08>}
     uVRegion {bIsValid=false, max=<0 0>, min=<0 0>}
    }
   }
   slot {class=GridSlot, properties {columnSpan=3, layer=1, padding=<ltrb -2 0 -2 2>}}
  }
  Spacer_RightPadding {
   widgetClassPath=&Spacer
   parent=GP_UserMessage_Container
   widget=Spacer_RightPadding
   properties {size=<4 1>}
   slot {class=GridSlot, properties {column=3, columnSpan=1}}
  }
  Message_ContextMenu {
   widgetClassPath=/FortUILibrary/UIKit/ContextMenu/WBP_UIKit_ContextMenuAnchorButton
   bIsVariable=true
   parent=GP_UserMessage_Container
   widget=Message_ContextMenu
   properties {
    button Background Material=null
    button Outline Material=null
    buttonClass=/Game/UI/Frontend/Lobby/Nameplates/WBP_EmptyButton_ContextMenu.WBP_EmptyButton_ContextMenu_C
    buttonIcon=null
    closeMenuOnContextClicked=true
    menuPlacement=MenuPlacement_BelowRightAnchor
    min Height=1
    button Padding=<ltrb 0 0 0 0>
   }
   slot {class=GridSlot, properties {column=3, horizontalAlignment=HAlign_Right, layer=99, verticalAlignment=VAlign_Center}}
  }
 ]
}
