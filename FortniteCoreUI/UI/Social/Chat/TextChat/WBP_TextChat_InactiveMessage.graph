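# Text chat widget that shows why a conversation is inactive, built as a CommonUserWidget around WBP_Social_Restrictions_Internal.
# CF_InactiveMessageText maps the conversation type (party, group or one-to-one) to a localized explanation.
# PreConstruct collapses the inner widget's button container and applies the RestrictionWidgetInternalPadding variable.
# It sets the warning text only at design time, as an EpicGroup preview; this graph does not set it at runtime.
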WidgetBlueprint /FortniteCoreUI/UI/Social/Chat/TextChat/WBP_TextChat_InactiveMessage
graphs [
 CF_InactiveMessageText {nodes [
   n1 = CF_InactiveMessageText!() -> (then, ConversationType:EFortTextChatPersistentConversationType Enum=EpicParty) {is=title K2Node_FunctionEntry}
   > Return Node!(
    execute
    InactiveMessageText:Text=Select!(
     EpicParty:Text="LOCTEXT(\"Party text chat will be available once you\\'re in a party\")"
     EpicGroup:Text="LOCTEXT(\"To continue chatting, you must have an Epic friend in the group, or someone else must also join\")"
     Epic121:Text="LOCTEXT(\"You can\\'t reply in this chat\")"
     Index:EFortTextChatPersistentConversationType Enum=@n1
    ) -> (ReturnValue:Text)
   )
  ]}
 EventGraph {nodes [
   n1 = Event PreConstruct()
   > n2 = Sequence()
   > Branch(Condition=@n1.IsDesignTime)
   > SetWarningText!(
    execute
    self:WBP Social Restrictions Internal Object Reference=Get WBP_Social_Restrictions_Internal() -> (WBP_Social_Restrictions_Internal:WBP Social Restrictions Internal Object Reference)
    WarningText:Text=CF_InactiveMessageText!(self, ConversationType:EFortTextChatPersistentConversationType Enum=EpicGroup) -> (InactiveMessageText:Text) {is=title K2Node_CallFunction}
   ) -> (then)
   SetButtonContainerVisibility!(
    execute=@n2.then_1
    self:WBP Social Restrictions Internal Object Reference=Get WBP_Social_Restrictions_Internal()
    ContainerVisibility:ESlateVisibility Enum=Collapsed
   ) -> (then)
   > SetPadding!(
    execute
    self:User Widget Object Reference=Get WBP_Social_Restrictions_Internal()
    InPadding:Margin Structure=Get RestrictionWidgetInternalPadding() -> (RestrictionWidgetInternalPadding:Margin Structure)
   ) -> (then)
  ]}
]
variables [RestrictionWidgetInternalPadding {type=Margin, default=<ltrb 0 0 0 0>}]
widgets {
 info {
  inheritedWidgetCount=0
  parentClass=&CommonUserWidget
  rootWidgetClass=&/WBP_Social_Restrictions_Internal.WBP_Social_Restrictions_Internal_C
  widgetCount=1
 }
 widgets [WBP_Social_Restrictions_Internal {
   widgetClassPath=&/WBP_Social_Restrictions_Internal
   bIsVariable=true
   widget=WBP_Social_Restrictions_Internal
  }]
}
