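# WBP_TextChat_Input is the text chat input field widget, a child of WBP_UIKit_InputField that adds no widgets of its own.
# On Construct, it checks the NewTextChat.UseLargeChatTextSize console variable and, when true, adds the StyleSheet.Style.Mobile style tag to itself.
# When added to the focus path, it calls the parent handler and then broadcasts its OnFocused dispatcher.
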WidgetBlueprint /FortniteCoreUI/UI/Social/Chat/TextChat/WBP_TextChat_Input
graphs [
 EventGraph {nodes [
   n1 = Event PreConstruct()
   > Parent: PreConstruct(IsDesignTime=@n1.IsDesignTime)
   Event Construct()
   > Parent: Construct()
   > n2 = GetConsoleVariableBoolValue!(execute, VariableName:String=NewTextChat.UseLargeChatTextSize) -> (then, ReturnValue:bool)
   > Branch(Condition=@n2)
   > UpdateWidgetStyleTags!(
    execute
    InWidget:Widget Object Reference=Self-Reference()
    InAddedTags:Array of Gameplay Tag Structures=Make Array("[0]"=(TagName="StyleSheet.Style.Mobile"))
    InRemovedTags:Array of Gameplay Tag Structures=Make Array("[0]"=(TagName=""))
   ) -> (then, ReturnValue:bool)
   n3 = Event OnAddedToFocusPath!() -> (OutputDelegate, then, InFocusEvent:Focus Event Structure) {is=title K2Node_Event}
   > Parent: OnAddedToFocusPath!(execute, InFocusEvent:Focus Event Structure=@n3.InFocusEvent) -> (then) {is=title K2Node_CallParentFunction}
   > Call On Focused!(execute:Exec, self:WBP Text Chat Input Object Reference) -> (then:Exec) {is=title K2Node_CallDelegate}
  ]}
 OnFocused {nodes [OnFocused() {is=title K2Node_FunctionEntry}]}
]
variables [OnFocused {}]
widgets {
 info {
  inheritedWidgetCount=0
  parentClass=/FortUILibrary/UIKit/InputField/WBP_UIKit_InputField.WBP_UIKit_InputField_C
  rootWidgetClass=null
  widgetCount=0
 }
 widgets []
}
