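# Text chat list entry widget for system-style messages (system, game, killfeed, welcome, info), extending WBP_TextChat_Entry_MessageBase.
# Binds the list item's Text Chat Message VM, colors text by message type, and shows side lines for RecordingInfo.
# Text-to-speech state drives a throbber or speaking icon, background highlight and error text.
# Hover reveals the context menu icon; on touch, a first tap reveals it and a second fires OnMessageClicked.
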WidgetBlueprint /FortniteCoreUI/UI/Social/Chat/TextChat/WBP_TextChat_MessageSystem
graphs [
 CF_ErrorVisibility {nodes [
   n1 = CF_ErrorVisibility!() -> (then, TTS State:EFortTextToSpeechState Enum=None) {is=title K2Node_FunctionEntry}
   > Return Node!(
    execute
    Visibility:ESlateVisibility Enum=Select(
     Option 0=Collapsed
     Option 1=SelfHitTestInvisible
     Index="Equal (Enum)"(A:EFortTextToSpeechState Enum=@n1, B:EFortTextToSpeechState Enum=Failed)
    )
   )
  ]}
 EventGraph {nodes [
   n1 = "Init Finished (Message_ContextMenu)"!() -> (OutputDelegate, then, Object:Object Reference) {is=title K2Node_ComponentBoundEvent}
   > n2 = Cast To WBP_UIKit_ContextMenu!(execute, Object=@n1.Object) -> (then, CastFailed, AsWBP UIKit Context Menu:WBP UIKit Context Menu Object Reference) {is=title K2Node_DynamicCast}
   > Bind Event to On Entry Generated(
    self:List View Base Object Reference=Get Content!(self:WBP UIKit Context Menu Object Reference=@n2) -> (Content:Common List View Object Reference)
    Delegate=Create Event()
   ) {is=title K2Node_AddDelegate}
   n3 = OnContextMenuMessageGenerated!() -> (OutputDelegate, then, Widget:User Widget Object Reference) {is=title K2Node_CustomEvent}
   > n4 = Cast To WBP_UIKit_ContextMenuListEntry!(execute, Object=@n3.Widget) -> (then, CastFailed, AsWBP UIKit Context Menu List Entry:WBP UIKit Context Menu List Entry Object Reference) {is=title K2Node_DynamicCast}
   > Branch(Condition=IsValid(Object=Get Icon!(self:WBP UIKit Button Generic Base Object Reference=@n4) -> (Icon:Texture 2D Object Reference)))
   > SetIconMaterial!(
    execute
    self:WBP UIKit Button Generic Base Object Reference=@n4
    Material Instance:Material Instance Object Reference=/FortniteCoreUI/UI/Social/Chat/TextChat/MaterialInstances/MI_UI_TextChat_TTS
   ) -> (then)
   > SetIconPadding!(
    execute
    self:WBP UIKit Button Generic Base Object Reference=@n4
    Padding_Left:float=0
    Padding_Top:float=0
    Padding_Right:float=-8
    Padding_Bottom:float=0
   ) -> (then)
   n5 = Event PreConstruct()
   > Parent: PreConstruct(IsDesignTime=@n5.IsDesignTime)
   > SetVisibility(
    self=[Get Image_LeftSide() -> (Image_LeftSide:Image Object Reference), Get Image_RightSide() -> (Image_RightSide:Image Object Reference)]
    InVisibility=Collapsed
   )
   Event On Clicked()
   > n6 = Branch(Condition=IsUsingTouch(PlayerController=GetOwningPlayer()))
   > n7 = Branch(Condition=Get bIsContextIconVisible() -> (bIsContextIconVisible:bool))
   Call On Message Clicked!(
    execute:Exec=[@n8.then, @n7.then]
    self:WBP Text Chat Message System Object Reference
    Message ID:Name=Get MessageId!(self:Text Chat Message VM Object Reference=Get FortTextChatMessageVM() -> (FortTextChatMessageVM:Text Chat Message VM Object Reference)) -> (MessageId:Name=None)
    Button:Common Button Base Object Reference=Self-Reference()
   ) -> (then:Exec) {is=title K2Node_CallDelegate}
   Set bIsContextIconVisible(execute=@n7.else, bIsContextIconVisible:bool=true) -> (Output_Get:bool)
   > SetContextMenuIconVisibility!(execute, self, IsContextMenuVisible:bool=true) -> (then)
   n8 = Set bIsContextIconVisible(execute=@n6.else, bIsContextIconVisible=true)
   Event On Hovered()
   > n9 = Branch(Condition=IsUsingTouch(PlayerController=GetOwningPlayer()))
   Parent: On Hovered(execute=@n9.else) {is=title K2Node_CallParentFunction}
   n10 = Get IsPreviewMode() -> (IsPreviewMode:bool)
   > n11 = Branch(Condition=@n10)
   SetContextMenuIconVisibility(execute=@n11.else, IsContextMenuVisible=true)
   n12 = Event OnListItemObjectSet()
   > n13 = Cast To FortTextChatMessageVM!(execute, Object=@n12.ListItemObject) -> (then, CastFailed, AsText Chat Message VM:Text Chat Message VM Object Reference) {is=title K2Node_DynamicCast}
   > Set FortTextChatMessageVM(FortTextChatMessageVM:Text Chat Message VM Object Reference=@n13) -> (Output_Get:Text Chat Message VM Object Reference)
   Event On Unhovered()
   > Parent: On Unhovered() {is=title K2Node_CallParentFunction}
   > n14 = Branch(Condition=@n10)
   SetContextMenuIconVisibility(execute=@n14.else, IsContextMenuVisible=false)
   n15 = Event On Item Selection Changed!() -> (OutputDelegate, then, bIsSelected:bool) {is=title K2Node_Event}
   > n16 = Branch(Condition=@n15.bIsSelected)
   Branch(execute=@n16.else, Condition=IsUsingTouch(PlayerController=GetOwningPlayer()))
   > Set bIsContextIconVisible(bIsContextIconVisible=false)
   > SetContextMenuIconVisibility(IsContextMenuVisible=false)
  ]}
 OnMessageClicked {nodes [OnMessageClicked!() -> (then, Message ID:Name, Button:Common Button Base Object Reference) {is=title K2Node_FunctionEntry}]}
 SetBackgroundVisibility {nodes [
   n1 = SetBackgroundVisibility!() -> (then, ShouldBeVisible:bool=~) {is=title K2Node_FunctionEntry}
   > SetVisibility(
    self=Get Image_Background() -> (Image_Background:Image Object Reference)
    InVisibility=Select(Option 0=Collapsed, Option 1=SelfHitTestInvisible, Index=@n1)
   )
  ]}
 SetContextMenuIconVisibility {nodes [
   n1 = SetContextMenuIconVisibility!() -> (then, IsContextMenuVisible:bool=~) {is=title K2Node_FunctionEntry}
   > Set RenderOpacity!(
    execute
    RenderOpacity:float=SelectFloat(A=1, B=0, bPickA=@n1)
    self:Widget Object Reference=Get ContextMenuIcon() -> (ContextMenuIcon:Image Object Reference)
   ) -> (then, Output_Get:float)
  ]}
 SetImagesVisibility {nodes [
   n1 = SetImagesVisibility!() -> (then, isSpeaking:bool=~) {is=title K2Node_FunctionEntry}
   > n2 = Branch(Condition=@n1)
   > Set RenderOpacity(RenderOpacity=0, self=Get Image_RightSide())
   > Set RenderOpacity(RenderOpacity=0, self=Get Image_LeftSide())
   n3 = Get FortTextChatMessageVM()
   n4 = SelectFloat(
    A=1
    B=0
    bPickA="Equal (Enum)"(
     A:EFortTextChatMessageType Enum=Get MessageType!(self:Text Chat Message VM Object Reference=@n3) -> (MessageType:EFortTextChatMessageType Enum=Sent)
     B:EFortTextChatMessageType Enum=RecordingInfo
    )
   )
   Set RenderOpacity(execute=@n2.else, RenderOpacity=@n4, self=Get Image_RightSide())
   > Set RenderOpacity(RenderOpacity=@n4, self=Get Image_LeftSide())
  ]}
 SetMessageText {nodes [
   n1 = SetMessageText!() -> (then, Text:Text) {is=title K2Node_FunctionEntry}
   > "SetText (Text)"(self=Get SystemMessage() -> (SystemMessage:Common Text Object Reference), InText=@n1)
  ]}
 SetSpeakingStateVisuals {nodes [
   n1 = SetSpeakingStateVisuals!() -> (then, Message Speaking State:EFortTextToSpeechState Enum=None) {is=title K2Node_FunctionEntry}
   > n2 = Switch on EFortTextToSpeechState!(execute, Selection:EFortTextToSpeechState Enum=@n1) -> (None:Exec, Preparing:Exec, Playing:Exec, Failed:Exec) {is=title K2Node_SwitchEnum}
   SetVisibility(execute=@n2.Failed, self=Get Image_Background(), InVisibility=Collapsed)
   SetVisibility(execute=[@n5.then, @n2.None], self=Get Image_Background(), InVisibility=Collapsed)
   n3 = Get TTS_Icon() -> (TTS_Icon:Image Object Reference)
   SetBrushFromMaterial(execute=@n2.Playing, self=@n3, Material=/FortUILibrary/Textures/MaterialInstances/MI_UIKit_Icon_VoiceChatActive)
   n4 = Get TTS_Icon()
   > SetDesiredSizeOverride!(execute, self:Image Object Reference=@n4, DesiredSize:Vector 2D Structure=Get TTSIconSize() -> (TTSIconSize:Vector 2D Structure)) -> (then)
   > SetVisibility(self=Get Image_Background(), InVisibility=Visible)
   SetBrushFromMaterial(execute=@n2.Preparing, self=@n3, Material=/FortUILibrary/UIKit/Throbber/MI_UIKit_Monochrome_Throbber)
   > n5 = SetDesiredSizeOverride(self=@n4, DesiredSize=Get ThrobberIconSize() -> (ThrobberIconSize:Vector 2D Structure))
  ]}
 SetSystemMessageColor {nodes [
   n1 = SetSystemMessageColor!() -> (then, MessageType:EFortTextChatMessageType Enum=Sent) {is=title K2Node_FunctionEntry}
   n2 = Get SystemColor() -> (SystemColor:Linear Color Structure)
   n3 = Get Welcome Color() -> (Welcome Color:Linear Color Structure)
   n4 = Get EpicPartyInfoColor() -> (EpicPartyInfoColor:Linear Color Structure)
   n5 = Get GameColor() -> (GameColor:Linear Color Structure)
   n6 = Get KillfeedColor() -> (KillfeedColor:Linear Color Structure)
   n7 = Get ChatInfoColor() -> (ChatInfoColor:Linear Color Structure)
   n8 = Get Player InfoColor() -> (Player InfoColor:Linear Color Structure)
   n9 = Get Recording InfoColor() -> (Recording InfoColor:Linear Color Structure)
   > SetColorAndOpacity(
    self=Get SystemMessage()
    InColorAndOpacity=Make Slate Color(SpecifiedColor=Select!(
      Sent:Linear Color Structure
      ReceivedFromPlayer:Linear Color Structure
      RedactedMessage:Linear Color Structure
      ChatInfo:Linear Color Structure=@n7
      RecordingInfo:Linear Color Structure=@n9
      PlayerInfo:Linear Color Structure=@n8
      System:Linear Color Structure=@n2
      Game:Linear Color Structure=@n5
      GameNPC:Linear Color Structure
      Killfeed:Linear Color Structure=@n6
      Welcome:Linear Color Structure=@n3
      EncryptionNotice:Linear Color Structure
      EpicPartyInfo:Linear Color Structure=@n4
      MAX_NUM:Linear Color Structure
      Index:EFortTextChatMessageType Enum=@n1
     ) -> (ReturnValue:Linear Color Structure))
   )
   n10 = Get MessageType!() -> (MessageType=Sent)
   > SetVisibility(
    self=[Get Image_LeftSide(), Get Image_RightSide()]
    InVisibility=Select!(
     Option 0:ESlateVisibility Enum=Collapsed
     Option 1:ESlateVisibility Enum=SelfHitTestInvisible
     Index:bool="Equal (Enum)"(A=@n10, B=RecordingInfo)
    ) -> (ReturnValue:ESlateVisibility Enum)
   )
  ]}
]
variables [
 bIsContextIconVisible {type=boolean}
 ChatInfoColor {type=LinearColor, default=<rgba 0.5765 0.5765 0.5765 1>}
 EpicPartyInfoColor {type=LinearColor, default=<rgba 0.5765 0.5765 0.5765 1>}
 GameColor {type=LinearColor, default=<rgba 1 1 1 1>}
 KillfeedColor {type=LinearColor, default=<rgba 1 0.147 0.1329 1>}
 OnMessageClicked {}
 Player InfoColor {type=LinearColor, default=<rgba 0.9301 1 0.0103 1>}
 Recording InfoColor {type=LinearColor, default=<rgba 0.0395 1 0.2307 1>}
 SystemColor {type=LinearColor, default=<rgba 0.5776 0.5776 0.5776 1>}
 ThrobberIconSize {type=Vector2D, default=<20 20>}
 TTSIconSize {type=Vector2D, default=<32 32>}
 Welcome Color {type=LinearColor, default=<rgba 1 0.7257 0 1>}
]
widgets {
 info {
  inheritedWidgetCount=0
  namedSlotCount=1
  parentClass=&/WBP_TextChat_Entry_MessageBase.WBP_TextChat_Entry_MessageBase_C
  rootWidgetClass=null
  widgetCount=10
 }
 widgets [
  GridPanel_0 {widgetClassPath=&GridPanel, widget=GridPanel_0, properties {columnFill [0,1,0], rowFill [1]}}
  HorizontalBox_0 {
   widgetClassPath=&HorizontalBox
   parent=GridPanel_0
   widget=HorizontalBox_0
   slot {class=GridSlot, properties {horizontalAlignment=HAlign_Left, verticalAlignment=VAlign_Center, padding=<ltrb 3 0 -2 0>}}
  }
  TTS_Icon {
   widgetClassPath=&Image
   bIsVariable=true
   parent=HorizontalBox_0
   widget=TTS_Icon
   properties {
    visibility=Collapsed
    brush {
     bIsDynamicallyLoaded=false
     drawAs=Image
     imageType=FullColor
     mirroring=NoMirror
     resourceName=None
     resourceObject=/FortniteCoreUI/UI/Icons/Icon-Settings/T-Icon-Audio-64
     tiling=NoTile
     imageSize=<32 32>
     margin=<ltrb 0 0 0 0>
     outlineSettings {
      bUseBrushTransparency=false
      roundingType=HalfHeightRadius
      width=0
      color {colorUseRule=UseColor_Specified, specifiedColor=<rgba 0 0 0 0>}
      cornerRadii=<0 0 0 0>
     }
     tintColor {colorUseRule=UseColor_Specified, specifiedColor=<rgba 1 1 1 1>}
     uVRegion {bIsValid=false, max=<0 0>, min=<0 0>}
    }
   }
   slot {class=HorizontalBoxSlot, properties {horizontalAlignment=HAlign_Left, verticalAlignment=VAlign_Center, padding=<ltrb 0 0 4 0>}}
  }
  Image_LeftSide {
   widgetClassPath=&Image
   bIsVariable=true
   parent=HorizontalBox_0
   widget=Image_LeftSide
   properties {
    brush {
     bIsDynamicallyLoaded=false
     drawAs=Image
     imageType=NoImage
     mirroring=NoMirror
     resourceName=None
     resourceObject=null
     tiling=NoTile
     imageSize=<32 2>
     margin=<ltrb 0 0 0 0>
     outlineSettings {
      bUseBrushTransparency=false
      roundingType=HalfHeightRadius
      width=0
      color {colorUseRule=UseColor_Specified, specifiedColor=<rgba 0 0 0 0>}
      cornerRadii=<0 0 0 0>
     }
     tintColor {colorUseRule=UseColor_Specified, specifiedColor=<rgba 1 1 1 1>}
     uVRegion {bIsValid=false, max=<0 0>, min=<0 0>}
    }
    colorAndOpacity=<rgba 1 1 1 0.102>
   }
   slot {class=HorizontalBoxSlot, properties {horizontalAlignment=HAlign_Left, verticalAlignment=VAlign_Center}}
  }
  SystemMessage {
   widgetClassPath=&CommonTextBlock
   bIsVariable=true
   parent=GridPanel_0
   widget=SystemMessage
   properties {
    bIsScrollingEnabled=false
    justification=Center
    minDesiredWidth=300
    mobileFontSizeMultiplier=1.5
    text="{SystemMessage} {SystemMessage}"
    wrapTextAt=500
    colorAndOpacity {colorUseRule=UseColor_Specified, specifiedColor=<rgba 0.5776 0.5776 0.5776 1>}
    font {
     bForceMonospaced=false
     bMaterialIsStencil=false
     fontMaterial=null
     fontObject=&HeadingNow
     letterSpacing=0
     monospacedWidth=1
     size=13.5
     skewAmount=0
     typefaceFontName=74 Regular
     outlineSettings {
      bApplyOutlineToDropShadows=false
      bMiteredCorners=false
      bSeparateFillAlpha=false
      outlineMaterial=null
      outlineSize=2
      outlineColor=<rgba 0 0 0 0.15>
     }
    }
    shadowOffset=<0 0>
   }
   slot {class=GridSlot, properties {column=1, padding=<ltrb 24 10 24 10>}}
  }
  ErrorMessage {
   widgetClassPath=&CommonTextBlock
   bIsVariable=true
   parent=GridPanel_0
   widget=ErrorMessage
   properties {
    applyLineHeightToBottomLine=false
    lineHeightPercentage=1.3
    text="Couldn’t read aloud. Select to try again."
    visibility=Collapsed
    wrappingPolicy=AllowPerCharacterWrapping
    colorAndOpacity {colorUseRule=UseColor_Specified, specifiedColor=<rgba 0.8714 0.1046 0.107 1>}
    font {
     bForceMonospaced=false
     bMaterialIsStencil=false
     fontMaterial=null
     fontObject=&HeadingNow
     letterSpacing=0
     monospacedWidth=1
     size=13.5
     skewAmount=0
     typefaceFontName=74 Regular
     outlineSettings {
      bApplyOutlineToDropShadows=false
      bMiteredCorners=false
      bSeparateFillAlpha=false
      outlineMaterial=null
      outlineSize=2
      outlineColor=<rgba 0 0 0 0.35>
     }
    }
    shadowColorAndOpacity=<rgba 0 0 0 1>
    shadowOffset=<0 0>
   }
   slot {class=GridSlot, properties {column=1, columnSpan=1, horizontalAlignment=HAlign_Center, row=1, padding=<ltrb 0 4 0 4>}}
  }
  Image_RightSide {
   widgetClassPath=&Image
   bIsVariable=true
   parent=GridPanel_0
   widget=Image_RightSide
   properties {
    brush {
     bIsDynamicallyLoaded=false
     drawAs=Image
     imageType=NoImage
     mirroring=NoMirror
     resourceName=None
     resourceObject=null
     tiling=NoTile
     imageSize=<32 2>
     margin=<ltrb 0 0 0 0>
     outlineSettings {
      bUseBrushTransparency=false
      roundingType=HalfHeightRadius
      width=0
      color {colorUseRule=UseColor_Specified, specifiedColor=<rgba 0 0 0 0>}
      cornerRadii=<0 0 0 0>
     }
     tintColor {colorUseRule=UseColor_Specified, specifiedColor=<rgba 1 1 1 1>}
     uVRegion {bIsValid=false, max=<0 0>, min=<0 0>}
    }
    colorAndOpacity=<rgba 1 1 1 0.102>
   }
   slot {class=GridSlot, properties {column=2, verticalAlignment=VAlign_Center, padding=<ltrb 5 0 0 0>}}
  }
  ContextMenuIcon {
   widgetClassPath=&Image
   bIsVariable=true
   parent=GridPanel_0
   widget=ContextMenuIcon
   properties {
    renderOpacity=0
    visibility=HitTestInvisible
    brush {
     bIsDynamicallyLoaded=false
     drawAs=Image
     imageType=FullColor
     mirroring=NoMirror
     resourceName=None
     resourceObject=/FortniteCoreUI/UI/Social/Chat/TextChat/MaterialInstances/MI_UI_TextChat_MessageContextMenu
     tiling=NoTile
     imageSize=<24 24>
     margin=<ltrb 0 0 0 0>
     outlineSettings {
      bUseBrushTransparency=false
      roundingType=HalfHeightRadius
      width=0
      color {colorUseRule=UseColor_Specified, specifiedColor=<rgba 0 0 0 0>}
      cornerRadii=<0 0 0 0>
     }
     tintColor {colorUseRule=UseColor_Specified, specifiedColor=<rgba 1 1 1 1>}
     uVRegion {bIsValid=false, max=<0 0>, min=<0 0>}
    }
   }
   slot {
    class=GridSlot
    properties {column=2, columnSpan=1, horizontalAlignment=HAlign_Right, verticalAlignment=VAlign_Center, padding=<ltrb 0 0 4 0>}
   }
  }
  Message_ContextMenu {
   widgetClassPath=/FortUILibrary/UIKit/ContextMenu/WBP_UIKit_ContextMenuAnchorButton
   bIsVariable=true
   parent=GridPanel_0
   widget=Message_ContextMenu
   properties {
    button Background Material=null
    button Outline Material=null
    buttonClass=/Game/UI/Frontend/Lobby/Nameplates/WBP_EmptyButton_ContextMenu.WBP_EmptyButton_ContextMenu_C
    buttonIcon=null
    closeMenuOnContextClicked=true
    menuPlacement=MenuPlacement_BelowRightAnchor
    min Height=1
    button Padding=<ltrb 0 0 0 0>
   }
   slot {class=GridSlot, properties {column=3, horizontalAlignment=HAlign_Right, layer=99, verticalAlignment=VAlign_Center}}
  }
  Image_Background {
   widgetClassPath=&Image
   bIsVariable=true
   parent=GridPanel_0
   widget=Image_Background
   properties {
    visibility=Collapsed
    brush {
     bIsDynamicallyLoaded=false
     drawAs=RoundedBox
     imageType=NoImage
     mirroring=NoMirror
     resourceName=None
     resourceObject=null
     tiling=NoTile
     imageSize=<1 1>
     margin=<ltrb 0 0 0 0>
     outlineSettings {
      bUseBrushTransparency=false
      roundingType=FixedRadius
      width=0
      color {colorUseRule=UseColor_Specified, specifiedColor=<rgba 0 0 0 0>}
      cornerRadii=<8 8 8 8>
     }
     tintColor {colorUseRule=UseColor_Specified, specifiedColor=<rgba 1 1 1 0.08>}
     uVRegion {bIsValid=false, max=<0 0>, min=<0 0>}
    }
   }
   slot {class=GridSlot, properties {columnSpan=3, layer=1, padding=<ltrb -2 0 -2 2>}}
  }
 ]
}
