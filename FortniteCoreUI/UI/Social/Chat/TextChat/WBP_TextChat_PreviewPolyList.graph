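# Polymorphic list widget for the text chat preview, showing messages in a non-interactive, non-focusable list with a hidden scrollbar.
# GetDesiredEntryClassForItem maps each message view model's type to a user-message or system-message entry, leaving some types unassigned.
# On entry init, it passes the item index to the entry, enables preview mode and broadcasts OnListViewInitialized.
# On entry generation, it forwards OnListEntryGenerated to the entry widget.
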WidgetBlueprint /FortniteCoreUI/UI/Social/Chat/TextChat/WBP_TextChat_PreviewPolyList
graphs [
 EventGraph {nodes [
   n1 = "On Entry Initialized (ListView)"!() -> (OutputDelegate, then, Item:Object Reference, Widget:User Widget Object Reference) {is=title K2Node_ComponentBoundEvent}
   > n2 = Cast To WBP_TextChat_Entry_MessageBase!(execute, Object=@n1.Widget) -> (then, CastFailed, AsWBP Text Chat Entry Message Base:WBP Text Chat Entry Message Base Object Reference) {is=title K2Node_DynamicCast}
   > SetEntryIndex!(
    execute
    self:WBP Text Chat Entry Message Base Object Reference=@n2
    idx:int=GetIndexForItem!(self:List View Object Reference=Get ListView() -> (ListView:Common List View Object Reference), Item:Object Reference=@n1.Item) -> (ReturnValue:int=0)
   ) -> (then)
   > Set IsPreviewMode!(execute, IsPreviewMode:bool=true, self:WBP Text Chat Entry Message Base Object Reference=@n2) -> (then, Output_Get:bool)
   > Call On List View Initialized!(
    execute:Exec
    self:WBP Text Chat Preview Poly List Object Reference
    Item:Object Reference=@n1.Item
    Widget:User Widget Object Reference=@n1.Widget
   ) -> (then:Exec) {is=title K2Node_CallDelegate}
   n3 = "On Entry Generated (ListView)"!() -> (OutputDelegate, then, Widget:User Widget Object Reference) {is=title K2Node_ComponentBoundEvent}
   > n4 = Cast To WBP_TextChat_Entry_MessageBase(Object=@n3.Widget)
   > OnListEntryGenerated(self:WBP Text Chat Entry Message Base Object Reference=@n4)
   Event OnInitialized()
   > SetScrollbarVisibility!(execute, self:List View Base Object Reference=Get ListView(), InVisibility:ESlateVisibility Enum=Collapsed) -> (then)
  ]}
 GetDesiredEntryClassForItem {nodes [
   n1 = GetDesiredEntryClassForItem!() -> (then, Item:Object Reference) {is=title K2Node_FunctionEntry}
   > n2 = Cast To FortTextChatMessageVM!(execute, Object=@n1) -> (then, CastFailed, AsText Chat Message VM:Text Chat Message VM Object Reference) {is=title K2Node_DynamicCast}
   > Return Node!(
    execute
    ReturnValue:User Widget Class Reference=Select!(
     Sent:User Widget Class Reference=&/WBP_TextChat_Entry_UserMessage.WBP_TextChat_Entry_UserMessage_C
     ReceivedFromPlayer:User Widget Class Reference=&WBP_TextChat_Entry_UserMessage_C
     RedactedMessage:User Widget Class Reference
     ChatInfo:User Widget Class Reference=&/WBP_TextChat_MessageSystem.WBP_TextChat_MessageSystem_C
     RecordingInfo:User Widget Class Reference=&WBP_TextChat_MessageSystem_C
     PlayerInfo:User Widget Class Reference=&WBP_TextChat_MessageSystem_C
     System:User Widget Class Reference=&WBP_TextChat_MessageSystem_C
     Game:User Widget Class Reference=&WBP_TextChat_MessageSystem_C
     GameNPC:User Widget Class Reference
     Killfeed:User Widget Class Reference=&WBP_TextChat_MessageSystem_C
     Welcome:User Widget Class Reference=&WBP_TextChat_MessageSystem_C
     EncryptionNotice:User Widget Class Reference
     EpicPartyInfo:User Widget Class Reference
     MAX_NUM:User Widget Class Reference
     Index:EFortTextChatMessageType Enum=Get MessageType!(self:Text Chat Message VM Object Reference=@n2) -> (MessageType:EFortTextChatMessageType Enum=Sent)
    ) -> (ReturnValue:User Widget Class Reference)
   )
  ]}
 OnListViewInitialized {nodes [OnListViewInitialized!() -> (then, Item:Object Reference, Widget:User Widget Object Reference) {is=title K2Node_FunctionEntry}]}
]
variables [OnListViewInitialized {}]
widgets {
 info {
  inheritedWidgetCount=1
  parentClass=/Script/FortniteUI.FortPolymorphicListView
  rootWidgetClass=&CommonListView
  widgetCount=1
 }
 widgets [ListView {
   widgetClassPath=&CommonListView
   bInherited=true
   bIsVariable=true
   widget=ListView
   properties {
    allowOverscroll=false
    bAllowDragging=false
    bEnableRightClickScrolling=false
    bEnableTouchScrolling=false
    bIsFocusable=false
    clipping=Inherit
    consumeMouseWheel=Never
    entryWidgetClass=&WBP_TextChat_Entry_UserMessage_C
    navigation=&self:WidgetTree.ListView.WidgetNavigation_0
    numDesignerPreviewEntries=12
    selectionMode=None
    verticalEntrySpacing=4
    wheelScrollMultiplier=0
    scrollBarPadding=<ltrb 7 0 0 0>
    scrollBarStyle {
     thickness=0
     draggedThumbImage {
      bIsDynamicallyLoaded=false
      drawAs=Box
      imageType=FullColor
      mirroring=NoMirror
      resourceName=../../../Engine/Content/Slate/Common/Scrollbar_Thumb.png
      resourceObject=null
      tiling=NoTile
      imageSize=<0 0>
      margin=<ltrb 0.25 0.25 0.25 0.25>
      outlineSettings {
       bUseBrushTransparency=false
       roundingType=HalfHeightRadius
       width=0
       color {colorUseRule=UseColor_Specified, specifiedColor=<rgba 0 0 0 0>}
       cornerRadii=<0 0 0 0>
      }
      tintColor {colorUseRule=UseColor_Specified, specifiedColor=<rgba 1 1 1 1>}
      uVRegion {bIsValid=false, max=<0 0>, min=<0 0>}
     }
     horizontalBackgroundImage {
      bIsDynamicallyLoaded=false
      drawAs=NoDrawType
      imageType=NoImage
      mirroring=NoMirror
      resourceName=None
      resourceObject=null
      tiling=NoTile
      imageSize=<0 0>
      margin=<ltrb 0 0 0 0>
      outlineSettings {
       bUseBrushTransparency=false
       roundingType=HalfHeightRadius
       width=0
       color {colorUseRule=UseColor_Specified, specifiedColor=<rgba 0 0 0 0>}
       cornerRadii=<0 0 0 0>
      }
      tintColor {colorUseRule=UseColor_Specified, specifiedColor=<rgba 1 1 1 1>}
      uVRegion {bIsValid=false, max=<0 0>, min=<0 0>}
     }
     horizontalBottomSlotImage {
      bIsDynamicallyLoaded=false
      drawAs=Image
      imageType=FullColor
      mirroring=NoMirror
      resourceName=../../../Engine/Content/Slate/Common/Scrollbar_Background_Horizontal.png
      resourceObject=null
      tiling=NoTile
      imageSize=<8 8>
      margin=<ltrb 0 0 0 0>
      outlineSettings {
       bUseBrushTransparency=false
       roundingType=HalfHeightRadius
       width=0
       color {colorUseRule=UseColor_Specified, specifiedColor=<rgba 0 0 0 0>}
       cornerRadii=<0 0 0 0>
      }
      tintColor {colorUseRule=UseColor_Specified, specifiedColor=<rgba 1 1 1 1>}
      uVRegion {bIsValid=false, max=<0 0>, min=<0 0>}
     }
     horizontalTopSlotImage {
      bIsDynamicallyLoaded=false
      drawAs=Image
      imageType=FullColor
      mirroring=NoMirror
      resourceName=../../../Engine/Content/Slate/Common/Scrollbar_Background_Horizontal.png
      resourceObject=null
      tiling=NoTile
      imageSize=<8 8>
      margin=<ltrb 0 0 0 0>
      outlineSettings {
       bUseBrushTransparency=false
       roundingType=HalfHeightRadius
       width=0
       color {colorUseRule=UseColor_Specified, specifiedColor=<rgba 0 0 0 0>}
       cornerRadii=<0 0 0 0>
      }
      tintColor {colorUseRule=UseColor_Specified, specifiedColor=<rgba 1 1 1 1>}
      uVRegion {bIsValid=false, max=<0 0>, min=<0 0>}
     }
     hoveredThumbImage {
      bIsDynamicallyLoaded=false
      drawAs=Box
      imageType=FullColor
      mirroring=NoMirror
      resourceName=../../../Engine/Content/Slate/Common/Scrollbar_Thumb.png
      resourceObject=null
      tiling=NoTile
      imageSize=<0 0>
      margin=<ltrb 0.25 0.25 0.25 0.25>
      outlineSettings {
       bUseBrushTransparency=false
       roundingType=HalfHeightRadius
       width=0
       color {colorUseRule=UseColor_Specified, specifiedColor=<rgba 0 0 0 0>}
       cornerRadii=<0 0 0 0>
      }
      tintColor {colorUseRule=UseColor_Specified, specifiedColor=<rgba 1 1 1 1>}
      uVRegion {bIsValid=false, max=<0 0>, min=<0 0>}
     }
     normalThumbImage {
      bIsDynamicallyLoaded=false
      drawAs=Box
      imageType=FullColor
      mirroring=NoMirror
      resourceName=../../../Engine/Content/Slate/Common/Scrollbar_Thumb.png
      resourceObject=null
      tiling=NoTile
      imageSize=<0 0>
      margin=<ltrb 0.25 0.25 0.25 0.25>
      outlineSettings {
       bUseBrushTransparency=false
       roundingType=HalfHeightRadius
       width=0
       color {colorUseRule=UseColor_Specified, specifiedColor=<rgba 0 0 0 0>}
       cornerRadii=<0 0 0 0>
      }
      tintColor {colorUseRule=UseColor_Specified, specifiedColor=<rgba 1 1 1 1>}
      uVRegion {bIsValid=false, max=<0 0>, min=<0 0>}
     }
     verticalBackgroundImage {
      bIsDynamicallyLoaded=false
      drawAs=NoDrawType
      imageType=NoImage
      mirroring=NoMirror
      resourceName=None
      resourceObject=null
      tiling=NoTile
      imageSize=<0 0>
      margin=<ltrb 0 0 0 0>
      outlineSettings {
       bUseBrushTransparency=false
       roundingType=HalfHeightRadius
       width=0
       color {colorUseRule=UseColor_Specified, specifiedColor=<rgba 0 0 0 0>}
       cornerRadii=<0 0 0 0>
      }
      tintColor {colorUseRule=UseColor_Specified, specifiedColor=<rgba 1 1 1 1>}
      uVRegion {bIsValid=false, max=<0 0>, min=<0 0>}
     }
     verticalBottomSlotImage {
      bIsDynamicallyLoaded=false
      drawAs=Image
      imageType=FullColor
      mirroring=NoMirror
      resourceName=../../../Engine/Content/Slate/Common/Scrollbar_Background_Vertical.png
      resourceObject=null
      tiling=NoTile
      imageSize=<8 8>
      margin=<ltrb 0 0 0 0>
      outlineSettings {
       bUseBrushTransparency=false
       roundingType=HalfHeightRadius
       width=0
       color {colorUseRule=UseColor_Specified, specifiedColor=<rgba 0 0 0 0>}
       cornerRadii=<0 0 0 0>
      }
      tintColor {colorUseRule=UseColor_Specified, specifiedColor=<rgba 1 1 1 1>}
      uVRegion {bIsValid=false, max=<0 0>, min=<0 0>}
     }
     verticalTopSlotImage {
      bIsDynamicallyLoaded=false
      drawAs=Image
      imageType=FullColor
      mirroring=NoMirror
      resourceName=../../../Engine/Content/Slate/Common/Scrollbar_Background_Vertical.png
      resourceObject=null
      tiling=NoTile
      imageSize=<8 8>
      margin=<ltrb 0 0 0 0>
      outlineSettings {
       bUseBrushTransparency=false
       roundingType=HalfHeightRadius
       width=0
       color {colorUseRule=UseColor_Specified, specifiedColor=<rgba 0 0 0 0>}
       cornerRadii=<0 0 0 0>
      }
      tintColor {colorUseRule=UseColor_Specified, specifiedColor=<rgba 1 1 1 1>}
      uVRegion {bIsValid=false, max=<0 0>, min=<0 0>}
     }
    }
   }
  }]
}
