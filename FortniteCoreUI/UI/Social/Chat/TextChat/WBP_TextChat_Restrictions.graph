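# Text chat restriction notice that explains why chat is unavailable, such as settings, parental controls, platform privacy or social ban.
# A switcher picks the standard or generic (PlasterBeach/updated visuals) inner widget.
# Functions map the restriction enum to text, color, visibility and action button state, some field-notify broadcast.
# Generic-widget clicks query the social user viewmodel to open chat settings, parental controls or the PlasterBeach modal, then fire OnButtonClicked.
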WidgetBlueprint /FortniteCoreUI/UI/Social/Chat/TextChat/WBP_TextChat_Restrictions
graphs [
 EventGraph {nodes [
   "On Button Clicked (WBP_Social_Restriction_Generic_Internal)"() {is=title K2Node_ComponentBoundEvent}
   > HandleRestrictionActionButtonClicked()
   > Call On Button Clicked!(execute:Exec, self:WBP Text Chat Restrictions Object Reference) -> (then:Exec) {is=title K2Node_CallDelegate}
   DoPlasterBeach() {is=title K2Node_CustomEvent}
   > Push Plaster Beach Modal!(execute:Exec) -> (then:Exec) {is=title K2Node_MacroInstance}
   n1 = Event PreConstruct()
   > n2 = Sequence()
   > Branch(Condition=@n1.IsDesignTime)
   > SetWarningColor!(execute, self, TextChatRestriction:EFortTextChatRestriction Enum=ParentalPermission) -> (then)
   > SetWarningText!(
    execute
    self:WBP Social Restrictions Internal Object Reference=Get WBP_Social_Restrictions_Internal() -> (WBP_Social_Restrictions_Internal:WBP Social Restrictions Internal Object Reference)
    WarningText:Text=GetWarningText!(
     self
     TextChatRestriction:EFortTextChatRestriction Enum=ParentalPermission
     SocialBanTimeLeft:Timespan Structure=MakeTimespan!(Days:int=5, Hours:int=0, Minutes:int=0, Seconds:int=0, Milliseconds:int=0) -> (ReturnValue:Timespan Structure)
    ) -> (RestrictionWarningText:Text)
   ) -> (then)
   SetBackplateVisible!(execute=@n2.then_1, self, IsBackplateVisible:bool=Get ShowBackplate() -> (ShowBackplate:bool)) -> (then)
   > SetPadding!(
    execute
    self:User Widget Object Reference=[
     Get WBP_Social_Restrictions_Internal()
     Get WBP_Social_Restriction_Generic_Internal() -> (WBP_Social_Restriction_Generic_Internal:WBP Social Restriction Generic Internal Object Reference)
    ]
    InPadding:Margin Structure=Get RestrictionWidgetInternalPadding() -> (RestrictionWidgetInternalPadding:Margin Structure)
   ) -> (then)
   > SetPlasterBeach!(execute, self, IsActive:bool=false) -> (then)
   Event Construct()
   n3 = GetOwningPlayer()
   > SetSocialUser!(
    execute
    self:Fort Social User Viewmodel Object Reference=Get FortSocialUserInfoViewModel() -> (FortSocialUserInfoViewModel:Fort Social User Info View Model Object Reference)
    InSocialUser:Fort Social User Object Reference=GetSocialUserFromController!(FortPC:Fort Player Controller Object Reference=GetFortPlayerControllerFromActor!(Actor:Actor Object Reference=@n3) -> (ReturnValue:Fort Player Controller Object Reference)) -> (ReturnValue:Fort Social User Object Reference)
   ) -> (then)
  ]}
 GetWarningText {nodes [
   n1 = GetWarningText!() -> (then, TextChatRestriction:EFortTextChatRestriction Enum=Unrestricted, SocialBanTimeLeft:Timespan Structure) {is=title K2Node_FunctionEntry}
   n2 = GetTimespanAsText!(
    Timespan:"Timespan Structure (by ref)"=@n1.SocialBanTimeLeft
    bShowHoursWhenZero:bool=false
    bShowSeconds:bool=false
    FormatType:ETimespanAsTextFormat Enum=Approximate
    UnitDisplayType:ETimespanUnitDisplayFormat Enum=Full
    bUseSentenceCase:bool=true
   ) -> (ReturnValue:Text)
   > Return Node!(
    execute
    RestrictionWarningText:Text=Select!(
     Unrestricted:Text
     UserPermission:Text="LOCTEXT(\"Text chat is turned off in your settings\")"
     ParentalPermission:Text="LOCTEXT(\"Your parent or guardian has turned off text chat\")"
     SplitscreenParentalPermission:Text="LOCTEXT(\"Text chat is turned off due to the other player\\'s settings\")"
     NotAcceptedModeration:Text
     SplitscreenNotAcceptedModeration:Text
     PlatformPermission:Text="LOCTEXT(\"You can\\'t use text chat due to your console\\'s privacy settings\")"
     Cabined:Text="LOCTEXT(\"You need your parent\\'s permission to use text chat\")"
     SocialBan:Text=Format Text!(Format="LOCTEXT(\"You can\\'t use chat for {timeleft} because of your ban for breaking Community Rules\")", timeleft:Text=@n2) -> (Result)
     NonLocalReason:Text="LOCTEXT(\"Something went wrong\")"
     NotAcceptedBrigadeiro:Text="INVTEXT(\"Lorem ipsum dolor sit amet, consectetur adipiscing elit. Curabitur quis nunc sollicitudin, vestibulum mauris at, pulvinar ante.\")"
     SplitscreenNotAcceptedBrigadeiro:Text="INVTEXT(\"Lorem ipsum dolor sit amet, consectetur adipiscing elit. Curabitur quis nunc sollicitudin, vestibulum mauris at, pulvinar ante. \")"
     Index:EFortTextChatRestriction Enum=Get TextChatRestriction() -> (TextChatRestriction:EFortTextChatRestriction Enum=Unrestricted)
    ) -> (ReturnValue:Text)
   )
  ]}
 HandleRestrictionActionButtonClicked {nodes [
   HandleRestrictionActionButtonClicked() {is=title K2Node_FunctionEntry}
   > n1 = Switch on EFortTextChatRestriction!(
    execute
    Selection:EFortTextChatRestriction Enum=GetTextChatSettingRestriction!(self:Fort Social User Info View Model Object Reference=Get FortSocialUserInfoViewModel()) -> (ReturnValue:EFortTextChatRestriction Enum=Unrestricted)
   ) -> (
    Unrestricted:Exec
    UserPermission:Exec
    ParentalPermission:Exec
    SplitscreenParentalPermission:Exec
    NotAcceptedModeration:Exec
    SplitscreenNotAcceptedModeration:Exec
    PlatformPermission:Exec
    Cabined:Exec
    SocialBan:Exec
    NonLocalReason:Exec
    NotAcceptedBrigadeiro:Exec
    SplitscreenNotAcceptedBrigadeiro:Exec
   ) {is=title K2Node_SwitchEnum}
   RequestToOpenParentalControls!(execute=[@n1.Cabined, @n1.ParentalPermission]) -> (then)
   DoPlasterBeach(execute=[@n1.NotAcceptedBrigadeiro, @n1.SplitscreenNotAcceptedBrigadeiro]) {is=title K2Node_CallFunction}
   OpenSettingsToTextChat!(execute=@n1.UserPermission, OwningWidget:Common User Widget Object Reference=Self-Reference()) -> (then)
  ]}
 OnButtonClicked {nodes [OnButtonClicked() {is=title K2Node_FunctionEntry}]}
 SetActionButtonText {nodes [
   n1 = SetActionButtonText!() -> (then, TextChatRestriction:EFortTextChatRestriction Enum=Unrestricted) {is=title K2Node_FunctionEntry}
   > Set with Broadcast NotifyButtonActionText!(
    execute
    NotifyButtonActionText:Text=Select(
     UserPermission="LOCTEXT(\"Open communication settings\")"
     ParentalPermission="LOCTEXT(\"Open parental controls\")"
     Cabined="LOCTEXT(\"Open parental controls\")"
     Index=@n1
    )
   ) -> (then, Output_Get:Text)
  ]}
 SetActionButtonVisibility {nodes [
   SetActionButtonVisibility!() -> (then, TextChatRestriction:EFortTextChatRestriction Enum=Unrestricted) {is=title K2Node_FunctionEntry}
   n1 = Get FortGlobalUIContext() -> (ReturnValue:Fort Global UIContext Object Reference) {is=title K2Node_GetSubsystem}
   n2 = AND Boolean(
    A="Equal (Enum)"(
     A:EFortTextChatRestriction Enum=Get TextChatRestriction() -> (TextChatRestriction=Unrestricted)
     B:EFortTextChatRestriction Enum=ParentalPermission
    )
    B=NOT Boolean(A=IsInZone!(self:Fort Global UIContext Object Reference=@n1) -> (ReturnValue:bool))
   )
   n3 = "Equal (Enum)"(A=Get TextChatRestriction() -> (TextChatRestriction=Unrestricted), B=UserPermission)
   n4 = "Equal (Enum)"(A=Get TextChatRestriction() -> (TextChatRestriction=Unrestricted), B=Cabined)
   > n5 = Set with Broadcast NotifyButtonActionVisibility!(
    execute
    NotifyButtonActionVisibility:ESlateVisibility Enum=Select!(
     Option 0:ESlateVisibility Enum=Collapsed
     Option 1:ESlateVisibility Enum=SelfHitTestInvisible
     Index:bool=OR Boolean!(A=@n3, B=@n2, C:bool=@n4) -> (ReturnValue)
    ) -> (ReturnValue:ESlateVisibility Enum)
   ) -> (then, Output_Get:ESlateVisibility Enum=Visible)
   > SetButtonContainerVisibility!(
    execute
    self:WBP Social Restrictions Internal Object Reference=Get WBP_Social_Restrictions_Internal()
    ContainerVisibility:ESlateVisibility Enum=@n5
   ) -> (then)
  ]}
 SetBackplateVisible {nodes [
   n1 = SetBackplateVisible!() -> (then, IsBackplateVisible:bool=~) {is=title K2Node_FunctionEntry}
   > n2 = Set ShowBackplate(ShowBackplate:bool=@n1) -> (Output_Get:bool)
   > SetBackplateVisible(self:WBP Social Restrictions Internal Object Reference=Get WBP_Social_Restrictions_Internal(), IsBackplateVisible=@n2) {is=title K2Node_CallFunction}
   > SetBackplateVisible(self:WBP Social Restriction Generic Internal Object Reference=Get WBP_Social_Restriction_Generic_Internal(), IsBackplateVisible=@n2)
  ]}
 SetPlasterBeach {nodes [
   n1 = SetPlasterBeach!() -> (then, IsActive:bool=~) {is=title K2Node_FunctionEntry}
   n2 = Get Use Updated Visuals() -> (Use Updated Visuals:bool)
   > SetActiveWidget(
    self:Widget Switcher Object Reference=Get CommonActivatableWidgetSwitcher_0() -> (CommonActivatableWidgetSwitcher_0:Common Activatable Widget Switcher Object Reference)
    Widget=Select(
     Option 0:Widget Object Reference=Get WBP_Social_Restrictions_Internal()
     Option 1:Widget Object Reference=Get WBP_Social_Restriction_Generic_Internal()
     Index=OR Boolean!(A=@n1, B=@n2) -> (ReturnValue)
    ) -> (ReturnValue:Widget Object Reference)
   )
  ]}
 SetPlasterBeachVisibility {nodes [
   SetPlasterBeachVisibility!() -> (then, TextChatRestriction:EFortTextChatRestriction Enum=Unrestricted) {is=title K2Node_FunctionEntry}
   n1 = IsInZone(self=Get FortGlobalUIContext())
   n2 = Get TextChatRestriction() -> (TextChatRestriction=Unrestricted)
   n3 = Get TextChatRestriction() -> (TextChatRestriction=Unrestricted)
   n4 = OR Boolean!(
    A="Equal (Enum)"(A=Get TextChatRestriction() -> (TextChatRestriction=Unrestricted), B=UserPermission)
    B=AND Boolean(A="Equal (Enum)"(A=@n2, B=ParentalPermission), B=NOT Boolean(A=@n1))
    C:bool="Equal (Enum)"(A=Get TextChatRestriction() -> (TextChatRestriction=Unrestricted), B=Cabined)
    D:bool="Equal (Enum)"(A=@n3, B=NotAcceptedBrigadeiro)
    E:bool="Equal (Enum)"(A=@n3, B=SplitscreenNotAcceptedBrigadeiro)
   ) -> (ReturnValue)
   > SetIsClickable!(execute, self:WBP Social Restriction Generic Internal Object Reference=Get WBP_Social_Restriction_Generic_Internal(), IsClickable:bool=@n4) -> (then)
   Select(Option 0:ESlateVisibility Enum=Collapsed, Option 1:ESlateVisibility Enum=SelfHitTestInvisible, Index=@n4) -> (ReturnValue:ESlateVisibility Enum)
  ]}
 SetSelfVisibility {nodes [
   SetSelfVisibility!() -> (then, TextChatRestriction:EFortTextChatRestriction Enum=Unrestricted) {is=title K2Node_FunctionEntry}
   n1 = Self-Reference()
   > SetVisibility(
    self=@n1
    InVisibility=Select!(
     Unrestricted:ESlateVisibility Enum=Collapsed
     UserPermission:ESlateVisibility Enum=SelfHitTestInvisible
     ParentalPermission:ESlateVisibility Enum=SelfHitTestInvisible
     SplitscreenParentalPermission:ESlateVisibility Enum=SelfHitTestInvisible
     NotAcceptedModeration:ESlateVisibility Enum=Collapsed
     SplitscreenNotAcceptedModeration:ESlateVisibility Enum=Collapsed
     PlatformPermission:ESlateVisibility Enum=SelfHitTestInvisible
     Cabined:ESlateVisibility Enum=SelfHitTestInvisible
     SocialBan:ESlateVisibility Enum=SelfHitTestInvisible
     NonLocalReason:ESlateVisibility Enum=SelfHitTestInvisible
     NotAcceptedBrigadeiro:ESlateVisibility Enum=SelfHitTestInvisible
     SplitscreenNotAcceptedBrigadeiro:ESlateVisibility Enum=SelfHitTestInvisible
     Index:EFortTextChatRestriction Enum=Get TextChatRestriction() -> (TextChatRestriction=Unrestricted)
    ) -> (ReturnValue)
   )
   > Set with Broadcast NotifyHasRestrictionState!(execute, NotifyHasRestrictionState:bool=IsVisible(self=@n1)) -> (then, Output_Get:bool)
  ]}
 SetWarningColor {nodes [
   n1 = SetWarningColor!() -> (then, TextChatRestriction:EFortTextChatRestriction Enum=Unrestricted) {is=title K2Node_FunctionEntry}
   n2 = Select!(
    Option 0:E_UI_SocialRestriction_MessageType Enum=NewEnumerator0
    Option 1:E_UI_SocialRestriction_MessageType Enum=NewEnumerator2
    Index:bool="Equal (Enum)"(A=@n1, B=SocialBan)
   ) -> (ReturnValue:E_UI_SocialRestriction_MessageType Enum)
   > SetWarningColor!(
    execute
    self:WBP Social Restrictions Internal Object Reference=Get WBP_Social_Restrictions_Internal()
    MessageType:E_UI_SocialRestriction_MessageType Enum=@n2
   ) -> (then) {is=title K2Node_CallFunction}
   > SetWarningColor(self:WBP Social Restriction Generic Internal Object Reference=Get WBP_Social_Restriction_Generic_Internal(), MessageType=@n2)
  ]}
]
variables [
 NotifyButtonActionText {type=string, default=""}
 NotifyButtonActionVisibility {type=ESlateVisibility, default=Visible}
 NotifyHasRestrictionState {type=boolean}
 OnButtonClicked {}
 RestrictionWidgetInternalPadding {type=Margin, default=<ltrb 0 0 0 0>}
 ShowBackplate {type=boolean, default=true}
 Use Updated Visuals {type=boolean}
]
widgets {
 info {
  inheritedWidgetCount=0
  namedSlotCount=1
  parentClass=&CommonUserWidget
  rootWidgetClass=&CommonActivatableWidgetSwitcher
  widgetCount=4
 }
 widgets [
  CommonActivatableWidgetSwitcher_0 {
   widgetClassPath=&CommonActivatableWidgetSwitcher
   bIsVariable=true
   widget=CommonActivatableWidgetSwitcher_0
   properties {bOverride_Cursor=true, transitionDuration=0}
  }
  WBP_Social_Restrictions_Internal {
   widgetClassPath=&/WBP_Social_Restrictions_Internal
   bIsVariable=true
   parent=CommonActivatableWidgetSwitcher_0
   widget=WBP_Social_Restrictions_Internal
   slot {class=WidgetSwitcherSlot}
  }
  Slot_AddButtonHere {widgetClassPath=&NamedSlot, bIsVariable=true, namedSlotHost=WBP_Social_Restrictions_Internal, widget=Slot_AddButtonHere}
  WBP_Social_Restriction_Generic_Internal {
   widgetClassPath=/FortniteCoreUI/UI/Social/Restrictions/WBP_Social_Restriction_Generic_Internal
   bIsVariable=true
   parent=CommonActivatableWidgetSwitcher_0
   widget=WBP_Social_Restriction_Generic_Internal
   properties {visibility=Collapsed}
   slot {class=WidgetSwitcherSlot, properties {padding=<ltrb 8 8 8 8>}}
  }
 ]
}
