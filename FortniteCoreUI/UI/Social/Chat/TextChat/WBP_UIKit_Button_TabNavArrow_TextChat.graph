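# Tab navigation arrow button variant for text chat, derived from the UIKit TabNavArrow button with no widgets of its own.
# On Construct it restyles the inherited background block's dynamic material, setting the backplate switch and uniform rounded-corner parameters.
# It then overrides the icon size when the compact UI style is in use.
# PreConstruct and Tick only forward to the parent.
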WidgetBlueprint /FortniteCoreUI/UI/Social/Chat/TextChat/WBP_UIKit_Button_TabNavArrow_TextChat
graphs [
 EventGraph {nodes [
   n1 = Event PreConstruct()
   > Parent: PreConstruct(IsDesignTime=@n1.IsDesignTime)
   Event Construct()
   > Parent: Construct()
   > SetBackplateVisuals()
   > SetCompactUIStyle() {is=title K2Node_CallFunction}
   n2 = Event Tick()
   > Parent: Tick!(execute, MyGeometry:Geometry Structure=@n2.MyGeometry, InDeltaTime:float=@n2.InDeltaTime) -> (then) {is=title K2Node_CallParentFunction}
  ]}
 SetBackplateVisuals {nodes [
   SetBackplateVisuals() {is=title K2Node_FunctionEntry}
   n1 = Get BackgroundDynamicMaterial!(self:WBP UIKit Block Background Base Object Reference=Get Block_Background() -> (Block_Background:WBP UIKit Block Background Base Object Reference)) -> (BackgroundDynamicMaterial:Material Instance Dynamic Object Reference)
   > SetScalarParameterValue(self=@n1, ParameterName=Backplate Switch, Value=0)
   > SetScalarParameterValue(self=@n1, ParameterName=Radius In Pixels, Value=1)
   > SetVectorParameterValue(
    self=@n1
    ParameterName=Corner Radius
    Value=MakeColor!(R:float=12, G:float=12, B:float=12, A:float=12) -> (ReturnValue:Linear Color Structure)
   )
  ]}
 SetCompactUIStyle {nodes [
   SetCompactUIStyle() {is=title K2Node_FunctionEntry}
   > Branch(Condition=IsUsingCompactUI())
   > SetIconSize!(execute, self:WBP UIKit Button Generic Base Object Reference, Icon:Vector 2D Structure=(X=48,Y=48)) -> (then)
  ]}
]
variables []
widgets {
 info {
  inheritedWidgetCount=0
  parentClass=/FortUILibrary/UIKit/Buttons/TabButtons/WBP_UIKit_Button_TabNavArrow.WBP_UIKit_Button_TabNavArrow_C
  rootWidgetClass=null
  widgetCount=0
 }
 widgets []
}
